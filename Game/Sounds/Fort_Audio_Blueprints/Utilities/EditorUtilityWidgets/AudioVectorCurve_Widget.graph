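# Editor utility widget for sound designers to build and audition Fort audio curve vectors.
# Imports Close/Medium/Distant attenuation curves or three generic float curves into X/Y/Z, then exports and activates original or modified versions.
# Colored progress bars show curve values at the slider's normalized distance.
# Each tick, a valid audition actor drives the slider using its distance from the PIE player pawn, or the editor camera without PIE.
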EditorUtilityWidgetBlueprint /Game/Sounds/Fort_Audio_Blueprints/Utilities/EditorUtilityWidgets/AudioVectorCurve_Widget
graphs [
 AssignAttenuationFloatCurves {nodes [
   n1 = AssignAttenuationFloatCurves!() -> (then, AudioCurveVector:Fort Audio Curve Vector Object Reference) {is=title K2Node_FunctionEntry}
   > n2 = Get!(execute:Exec) -> (then:Exec, else:Exec, CloseAttenuation:Sound Attenuation Object Reference)
   n3 = Break Sound Attenuation Settings!(SoundAttenuationSettings:"Sound Attenuation Settings Structure (by ref)"=Get Attenuation!(self:Sound Attenuation Object Reference=@n2) -> (Attenuation:Sound Attenuation Settings Structure)) -> (
    bEnableSourceDataOverride:bool=~
    bEnableSendToAudioLink:bool=~
    AudioLinkSettingsOverride:Audio Link Settings Abstract Object Reference
    NonSpatializedRadiusStart:float=~
    NonSpatializedRadiusEnd:float=~
    NonSpatializedRadiusMode:ENonSpatializedRadiusSpeakerMapMode Enum
    FalloffDistance:float=~
    ConeSphereRadius:float=~
    ConeSphereFalloffDistance:float=~
    CustomAttenuationCurve:Runtime Float Curve Structure
   ) {is=title K2Node_BreakStruct}
   > n4 = ImportAttenuationCurve!(
    execute
    self:Fort Audio Curve Vector Object Reference=@n1
    Element:EVectorCurveFloat Enum=VectorCurve_X
    Curve:"Runtime Float Curve Structure (by ref)"=@n3.CustomAttenuationCurve
    MaxDistance:float=@n3.FalloffDistance
   ) -> (then)
   n5 = Get!(execute=[@n16.then, @n4.then]) -> (then, else, MediumAttenuation:Sound Attenuation Object Reference)
   n6 = Break Sound Attenuation Settings(SoundAttenuationSettings=Get Attenuation(self=@n5)) -> (
    bEnableSourceDataOverride=~
    bEnableSendToAudioLink=~
    NonSpatializedRadiusStart=~
    NonSpatializedRadiusEnd=~
    FalloffDistance=~
    ConeSphereRadius=~
    ConeSphereFalloffDistance=~
   )
   > n7 = ImportAttenuationCurve(self=@n1, Element=VectorCurve_Y, Curve=@n6.CustomAttenuationCurve, MaxDistance=@n6.FalloffDistance)
   n8 = Get!(execute=[@n7.then, @n14.then]) -> (then, else, DistantAttenuation:Sound Attenuation Object Reference)
   n9 = Break Sound Attenuation Settings(SoundAttenuationSettings=Get Attenuation(self=@n8)) -> (
    bEnableSourceDataOverride=~
    bEnableSendToAudioLink=~
    NonSpatializedRadiusStart=~
    NonSpatializedRadiusEnd=~
    FalloffDistance=~
    ConeSphereRadius=~
    ConeSphereFalloffDistance=~
   )
   > n10 = ImportAttenuationCurve(self=@n1, Element=VectorCurve_Z, Curve=@n9.CustomAttenuationCurve, MaxDistance=@n9.FalloffDistance)
   InitializeAudioCurveVector(execute=[@n10.then, @n12.then], self:Fort Audio Curve Vector Object Reference=@n1)
   n11 = Break Sound Attenuation Settings(SoundAttenuationSettings=Make Sound Attenuation Settings!(
     bAttenuate:bool=True
     bSpatialize:bool=True
     bAttenuateWithLPF:bool=False
     bEnableListenerFocus:bool=False
     bEnableFocusInterpolation:bool=False
     bEnableOcclusion:bool=False
     bUseComplexCollisionForOcclusion:bool=False
     bEnableReverbSend:bool=True
     bEnablePriorityAttenuation:bool=False
     bApplyNormalizationToStereoSounds:bool=False
     bEnableLogFrequencyScaling:bool=False
     bEnableSubmixSends:bool=False
     bEnableSourceDataOverride:bool=False
     bEnableSendToAudioLink:bool=True
     SpatializationAlgorithm:ESoundSpatializationAlgorithm Enum=SPATIALIZATION_Default
     AudioLinkSettingsOverride:Audio Link Settings Abstract Object Reference
     BinauralRadius:float=0
     CustomLowpassAirAbsorptionCurve:Runtime Float Curve Structure=(EditorCurveData=(Keys=,DefaultValue=340282346638528859811704183484516925440,PreInfinityExtrap=RCCE_Constant,PostInfinityExtrap=RCCE_Constant),ExternalCurve=None)
     CustomHighpassAirAbsorptionCurve:Runtime Float Curve Structure=(EditorCurveData=(Keys=,DefaultValue=340282346638528859811704183484516925440,PreInfinityExtrap=RCCE_Constant,PostInfinityExtrap=RCCE_Constant),ExternalCurve=None)
     AbsorptionMethod:EAirAbsorptionMethod Enum=Linear
     OcclusionTraceChannel:ECollisionChannel Enum=ECC_Visibility
     ReverbSendMethod:EReverbSendMethod Enum=Linear
     PriorityAttenuationMethod:EPriorityAttenuationMethod Enum=Linear
     NonSpatializedRadiusStart:float=0
     NonSpatializedRadiusEnd:float=0
     NonSpatializedRadiusMode:ENonSpatializedRadiusSpeakerMapMode Enum=OmniDirectional
     StereoSpread:float=200
     LPFRadiusMin:float=3000
     LPFRadiusMax:float=6000
     LPFFrequencyAtMin:float=20000
     LPFFrequencyAtMax:float=20000
     HPFFrequencyAtMin:float=0
     HPFFrequencyAtMax:float=0
     FocusAzimuth:float=30
     NonFocusAzimuth:float=60
     FocusDistanceScale:float=1
     NonFocusDistanceScale:float=1
     FocusPriorityScale:float=1
     NonFocusPriorityScale:float=1
     FocusVolumeAttenuation:float=1
     NonFocusVolumeAttenuation:float=1
     FocusAttackInterpSpeed:float=1
     FocusReleaseInterpSpeed:float=1
     OcclusionLowPassFilterFrequency:float=20000
     OcclusionVolumeAttenuation:float=1
     OcclusionInterpolationTime:float=0.1
     ReverbWetLevelMin:float=0.3
     ReverbWetLevelMax:float=0.95
     ReverbDistanceMin:float=400
     ReverbDistanceMax:float=4000
     ManualReverbSendLevel:float=0
     PriorityAttenuationMin:float=1
     PriorityAttenuationMax:float=1
     PriorityAttenuationDistanceMin:float=400
     PriorityAttenuationDistanceMax:float=4000
     ManualPriorityAttenuation:float=1
     CustomReverbSendCurve:Runtime Float Curve Structure=(EditorCurveData=(Keys=,DefaultValue=340282346638528859811704183484516925440,PreInfinityExtrap=RCCE_Constant,PostInfinityExtrap=RCCE_Constant),ExternalCurve=None)
     SubmixSendSettings:Array of Attenuation Submix Send Settings Structures
     CustomPriorityAttenuationCurve:Runtime Float Curve Structure=(EditorCurveData=(Keys=,DefaultValue=340282346638528859811704183484516925440,PreInfinityExtrap=RCCE_Constant,PostInfinityExtrap=RCCE_Constant),ExternalCurve=None)
     PluginSettings:Sound Attenuation Plugin Settings Structure=(SpatializationPluginSettingsArray=,OcclusionPluginSettingsArray=,ReverbPluginSettingsArray=,SourceDataOverridePluginSettingsArray=)
     DistanceAlgorithm:EAttenuationDistanceModel Enum=Linear
     AttenuationShape:EAttenuationShape Enum=Sphere
     FalloffMode:ENaturalSoundFalloffMode Enum=Continues
     dBAttenuationAtMax:float=-60
     AttenuationShapeExtents:Vector="400,0,0"
     ConeOffset:float=0
     FalloffDistance:float=0
     ConeSphereRadius:float=0
     ConeSphereFalloffDistance:float=0
     CustomAttenuationCurve:Runtime Float Curve Structure=(EditorCurveData=(Keys=,DefaultValue=340282346638528859811704183484516925440,PreInfinityExtrap=RCCE_Constant,PostInfinityExtrap=RCCE_Constant),ExternalCurve=None)
    ) -> (SoundAttenuationSettings:Sound Attenuation Settings Structure) {is=title K2Node_MakeStruct}) -> (
    bEnableSourceDataOverride=~
    bEnableSendToAudioLink=~
    NonSpatializedRadiusStart=~
    NonSpatializedRadiusEnd=~
    FalloffDistance=~
    ConeSphereRadius=~
    ConeSphereFalloffDistance=~
   )
   n12 = ImportAttenuationCurve(execute=@n8.else, self=@n1, Element=VectorCurve_Z, Curve=@n11.CustomAttenuationCurve, MaxDistance=@n11.FalloffDistance)
   n13 = Break Sound Attenuation Settings(SoundAttenuationSettings=Make Sound Attenuation Settings(
     bAttenuate=True
     bSpatialize=True
     bAttenuateWithLPF=False
     bEnableListenerFocus=False
     bEnableFocusInterpolation=False
     bEnableOcclusion=False
     bUseComplexCollisionForOcclusion=False
     bEnableReverbSend=True
     bEnablePriorityAttenuation=False
     bApplyNormalizationToStereoSounds=False
     bEnableLogFrequencyScaling=False
     bEnableSubmixSends=False
     bEnableSourceDataOverride=False
     bEnableSendToAudioLink=True
     SpatializationAlgorithm=SPATIALIZATION_Default
     BinauralRadius=0
     CustomLowpassAirAbsorptionCurve=(EditorCurveData=(Keys=,DefaultValue=340282346638528859811704183484516925440,PreInfinityExtrap=RCCE_Constant,PostInfinityExtrap=RCCE_Constant),ExternalCurve=None)
     CustomHighpassAirAbsorptionCurve=(EditorCurveData=(Keys=,DefaultValue=340282346638528859811704183484516925440,PreInfinityExtrap=RCCE_Constant,PostInfinityExtrap=RCCE_Constant),ExternalCurve=None)
     AbsorptionMethod=Linear
     OcclusionTraceChannel=ECC_Visibility
     ReverbSendMethod=Linear
     PriorityAttenuationMethod=Linear
     NonSpatializedRadiusStart=0
     NonSpatializedRadiusEnd=0
     NonSpatializedRadiusMode=OmniDirectional
     StereoSpread=200
     LPFRadiusMin=3000
     LPFRadiusMax=6000
     LPFFrequencyAtMin=20000
     LPFFrequencyAtMax=20000
     HPFFrequencyAtMin=0
     HPFFrequencyAtMax=0
     FocusAzimuth=30
     NonFocusAzimuth=60
     FocusDistanceScale=1
     NonFocusDistanceScale=1
     FocusPriorityScale=1
     NonFocusPriorityScale=1
     FocusVolumeAttenuation=1
     NonFocusVolumeAttenuation=1
     FocusAttackInterpSpeed=1
     FocusReleaseInterpSpeed=1
     OcclusionLowPassFilterFrequency=20000
     OcclusionVolumeAttenuation=1
     OcclusionInterpolationTime=0.1
     ReverbWetLevelMin=0.3
     ReverbWetLevelMax=0.95
     ReverbDistanceMin=400
     ReverbDistanceMax=4000
     ManualReverbSendLevel=0
     PriorityAttenuationMin=1
     PriorityAttenuationMax=1
     PriorityAttenuationDistanceMin=400
     PriorityAttenuationDistanceMax=4000
     ManualPriorityAttenuation=1
     CustomReverbSendCurve=(EditorCurveData=(Keys=,DefaultValue=340282346638528859811704183484516925440,PreInfinityExtrap=RCCE_Constant,PostInfinityExtrap=RCCE_Constant),ExternalCurve=None)
     CustomPriorityAttenuationCurve=(EditorCurveData=(Keys=,DefaultValue=340282346638528859811704183484516925440,PreInfinityExtrap=RCCE_Constant,PostInfinityExtrap=RCCE_Constant),ExternalCurve=None)
     PluginSettings=(SpatializationPluginSettingsArray=,OcclusionPluginSettingsArray=,ReverbPluginSettingsArray=,SourceDataOverridePluginSettingsArray=)
     DistanceAlgorithm=Linear
     AttenuationShape=Sphere
     FalloffMode=Continues
     dBAttenuationAtMax=-60
     AttenuationShapeExtents="400,0,0"
     ConeOffset=0
     FalloffDistance=0
     ConeSphereRadius=0
     ConeSphereFalloffDistance=0
     CustomAttenuationCurve=(EditorCurveData=(Keys=,DefaultValue=340282346638528859811704183484516925440,PreInfinityExtrap=RCCE_Constant,PostInfinityExtrap=RCCE_Constant),ExternalCurve=None)
    )) -> (
    bEnableSourceDataOverride=~
    bEnableSendToAudioLink=~
    NonSpatializedRadiusStart=~
    NonSpatializedRadiusEnd=~
    FalloffDistance=~
    ConeSphereRadius=~
    ConeSphereFalloffDistance=~
   )
   n14 = ImportAttenuationCurve(execute=@n5.else, self=@n1, Element=VectorCurve_Y, Curve=@n13.CustomAttenuationCurve, MaxDistance=@n13.FalloffDistance)
   n15 = Break Sound Attenuation Settings(SoundAttenuationSettings=Make Sound Attenuation Settings(
     bAttenuate=True
     bSpatialize=True
     bAttenuateWithLPF=False
     bEnableListenerFocus=False
     bEnableFocusInterpolation=False
     bEnableOcclusion=False
     bUseComplexCollisionForOcclusion=False
     bEnableReverbSend=True
     bEnablePriorityAttenuation=False
     bApplyNormalizationToStereoSounds=False
     bEnableLogFrequencyScaling=False
     bEnableSubmixSends=False
     bEnableSourceDataOverride=False
     bEnableSendToAudioLink=True
     SpatializationAlgorithm=SPATIALIZATION_Default
     BinauralRadius=0
     CustomLowpassAirAbsorptionCurve=(EditorCurveData=(Keys=,DefaultValue=340282346638528859811704183484516925440,PreInfinityExtrap=RCCE_Constant,PostInfinityExtrap=RCCE_Constant),ExternalCurve=None)
     CustomHighpassAirAbsorptionCurve=(EditorCurveData=(Keys=,DefaultValue=340282346638528859811704183484516925440,PreInfinityExtrap=RCCE_Constant,PostInfinityExtrap=RCCE_Constant),ExternalCurve=None)
     AbsorptionMethod=Linear
     OcclusionTraceChannel=ECC_Visibility
     ReverbSendMethod=Linear
     PriorityAttenuationMethod=Linear
     NonSpatializedRadiusStart=0
     NonSpatializedRadiusEnd=0
     NonSpatializedRadiusMode=OmniDirectional
     StereoSpread=200
     LPFRadiusMin=3000
     LPFRadiusMax=6000
     LPFFrequencyAtMin=20000
     LPFFrequencyAtMax=20000
     HPFFrequencyAtMin=0
     HPFFrequencyAtMax=0
     FocusAzimuth=30
     NonFocusAzimuth=60
     FocusDistanceScale=1
     NonFocusDistanceScale=1
     FocusPriorityScale=1
     NonFocusPriorityScale=1
     FocusVolumeAttenuation=1
     NonFocusVolumeAttenuation=1
     FocusAttackInterpSpeed=1
     FocusReleaseInterpSpeed=1
     OcclusionLowPassFilterFrequency=20000
     OcclusionVolumeAttenuation=1
     OcclusionInterpolationTime=0.1
     ReverbWetLevelMin=0.3
     ReverbWetLevelMax=0.95
     ReverbDistanceMin=400
     ReverbDistanceMax=4000
     ManualReverbSendLevel=0
     PriorityAttenuationMin=1
     PriorityAttenuationMax=1
     PriorityAttenuationDistanceMin=400
     PriorityAttenuationDistanceMax=4000
     ManualPriorityAttenuation=1
     CustomReverbSendCurve=(EditorCurveData=(Keys=,DefaultValue=340282346638528859811704183484516925440,PreInfinityExtrap=RCCE_Constant,PostInfinityExtrap=RCCE_Constant),ExternalCurve=None)
     CustomPriorityAttenuationCurve=(EditorCurveData=(Keys=,DefaultValue=340282346638528859811704183484516925440,PreInfinityExtrap=RCCE_Constant,PostInfinityExtrap=RCCE_Constant),ExternalCurve=None)
     PluginSettings=(SpatializationPluginSettingsArray=,OcclusionPluginSettingsArray=,ReverbPluginSettingsArray=,SourceDataOverridePluginSettingsArray=)
     DistanceAlgorithm=Linear
     AttenuationShape=Sphere
     FalloffMode=Continues
     dBAttenuationAtMax=-60
     AttenuationShapeExtents="400,0,0"
     ConeOffset=0
     FalloffDistance=0
     ConeSphereRadius=0
     ConeSphereFalloffDistance=0
     CustomAttenuationCurve=(EditorCurveData=(Keys=,DefaultValue=340282346638528859811704183484516925440,PreInfinityExtrap=RCCE_Constant,PostInfinityExtrap=RCCE_Constant),ExternalCurve=None)
    )) -> (
    bEnableSourceDataOverride=~
    bEnableSendToAudioLink=~
    NonSpatializedRadiusStart=~
    NonSpatializedRadiusEnd=~
    FalloffDistance=~
    ConeSphereRadius=~
    ConeSphereFalloffDistance=~
   )
   n16 = ImportAttenuationCurve(execute=@n2.else, self=@n1, Element=VectorCurve_X, Curve=@n15.CustomAttenuationCurve, MaxDistance=@n15.FalloffDistance)
  ]}
 EventGraph {nodes [
   "On Pressed (ExportOriginals_Button)"() {is=title K2Node_ComponentBoundEvent}
   > n1 = Get!(execute) -> (then, else, TargetCurveVector:Curve Vector Object Reference)
   > n2 = Cast To FortAudioCurveVector!(execute, Object=@n1) -> (then, CastFailed, AsFort Audio Curve Vector:Fort Audio Curve Vector Object Reference) {is=title K2Node_DynamicCast}
   > ExportCurves!(execute, self:Fort Audio Curve Vector Object Reference=@n2, CurveType:EVectorCurveType Enum=Original) -> (then)
   > SetActiveCurves!(execute, self:Fort Audio Curve Vector Object Reference=@n2, CurveType:EVectorCurveType Enum=Original) -> (then)
   > HighlightButton!(
    execute
    self
    Highlight Button:Button Object Reference=Get ExportOriginals_Button() -> (ExportOriginals_Button:Button Object Reference)
    Un-Highlight Button:Button Object Reference=Get ExportModified_Button() -> (ExportModified_Button:Button Object Reference)
   ) -> (then)
   "On Pressed (ExportModified_Button)"() {is=title K2Node_ComponentBoundEvent}
   > n3 = Get()
   > n4 = Cast To FortAudioCurveVector(Object=@n3)
   > ExportCurves(self=@n4, CurveType=Modified)
   > SetActiveCurves(self=@n4, CurveType=Modified)
   > HighlightButton(Highlight Button=Get ExportModified_Button(), Un-Highlight Button=Get ExportOriginals_Button())
   n5 = "On Value Changed (Distance_Slider)"!() -> (OutputDelegate, then, Value:float) {is=title K2Node_ComponentBoundEvent}
   > n6 = Set DistanceSliderValue(DistanceSliderValue:double=@n5.Value) -> (Output_Get:double)
   > OnDistanceSliderChanged!(execute, self, Value:double=@n6) -> (then)
   "On Property Changed (VectorCurveProperty)"!() -> (OutputDelegate, then, PropertyName:Name=None) {is=title K2Node_ComponentBoundEvent}
   > n7 = Get()
   > n8 = Cast To FortAudioCurveVector(Object=@n7)
   > Set MaxVectorCurveDistance(MaxVectorCurveDistance:double=GetLargestMaxDistance!(self:Fort Audio Curve Vector Object Reference=@n8) -> (ReturnValue:float)) -> (Output_Get:double)
   > n9 = SetIsEnabled(
    self=[
     Get ImportCurves_Button() -> (ImportCurves_Button:Button Object Reference)
     Get ImportGenericCurves_Button() -> (ImportGenericCurves_Button:Button Object Reference)
    ]
    bInIsEnabled=true
   )
   OnDistanceSliderChanged(execute=[@n10.then, @n9.then], Value=Get DistanceSliderValue() -> (DistanceSliderValue:double))
   Set MaxVectorCurveDistance(execute=@n7.else, MaxVectorCurveDistance=Get DefaultMaxVectorCurveDistance() -> (DefaultMaxVectorCurveDistance:double))
   > n10 = SetIsEnabled(
    self=[
     Get ExportOriginals_Button()
     Get ExportModifiedGeneric_Button() -> (ExportModifiedGeneric_Button:Button Object Reference)
     Get ExportOriginalGeneric_Button() -> (ExportOriginalGeneric_Button:Button Object Reference)
     Get ImportCurves_Button()
     Get ExportModified_Button()
     Get ImportGenericCurves_Button()
    ]
    bInIsEnabled=false
   )
   "On Pressed (ImportCurves_Button)"() {is=title K2Node_ComponentBoundEvent}
   > n11 = Get()
   > n12 = Cast To FortAudioCurveVector(Object=@n11)
   > AssignAttenuationFloatCurves!(execute, self, AudioCurveVector:Fort Audio Curve Vector Object Reference=@n12) -> (then) {is=title K2Node_CallFunction}
   > Set MaxVectorCurveDistance(MaxVectorCurveDistance=GetLargestMaxDistance(self=@n12))
   > OnDistanceSliderChanged(Value=Get DistanceSliderValue())
   > SetIsEnabled(self=[Get ExportOriginals_Button(), Get ExportModified_Button()], bInIsEnabled=true)
   > SetIsEnabled(self=[Get ExportModifiedGeneric_Button(), Get ExportOriginalGeneric_Button()], bInIsEnabled=false)
   > SetActiveCurves(self=@n12, CurveType=Modified)
   > HighlightButton(Highlight Button=Get ExportModified_Button(), Un-Highlight Button=Get ExportOriginals_Button())
   "On Pressed (ImportGenericCurves_Button)"() {is=title K2Node_ComponentBoundEvent}
   > n13 = Get()
   > n14 = Cast To FortAudioCurveVector(Object=@n13)
   > ImportFloatCurve!(
    execute
    self:Fort Audio Curve Vector Object Reference=@n14
    Element:EVectorCurveFloat Enum=VectorCurve_X
    Curve:Curve Float Object Reference=Get Curve_01() -> (Curve_01:Curve Float Object Reference)
   ) -> (then)
   > ImportFloatCurve(self=@n14, Element=VectorCurve_Y, Curve=Get Curve_02() -> (Curve_02:Curve Float Object Reference))
   > ImportFloatCurve(self=@n14, Element=VectorCurve_Z, Curve=Get Curve_03() -> (Curve_03:Curve Float Object Reference))
   > InitializeAudioCurveVector(self=@n14)
   > OnDistanceSliderChanged(Value=Get DistanceSliderValue())
   > SetIsEnabled(self=[Get ExportOriginals_Button(), Get ExportModified_Button()], bInIsEnabled=false)
   > SetIsEnabled(self=[Get ExportModifiedGeneric_Button(), Get ExportOriginalGeneric_Button()], bInIsEnabled=true)
   > SetActiveCurves(self=@n14, CurveType=Modified)
   > HighlightButton(Highlight Button=Get ExportModifiedGeneric_Button(), Un-Highlight Button=Get ExportOriginalGeneric_Button())
   "On Pressed (ExportModifiedGeneric_Button)"() {is=title K2Node_ComponentBoundEvent}
   > n15 = Get()
   > n16 = Cast To FortAudioCurveVector(Object=@n15)
   > ExportCurves(self=@n16, CurveType=Modified)
   > SetActiveCurves(self=@n16, CurveType=Modified)
   > HighlightButton(Highlight Button=Get ExportModifiedGeneric_Button(), Un-Highlight Button=Get ExportOriginalGeneric_Button())
   "On Pressed (ExportOriginalGeneric_Button)"() {is=title K2Node_ComponentBoundEvent}
   > n17 = Get()
   > n18 = Cast To FortAudioCurveVector(Object=@n17)
   > ExportCurves(self=@n18, CurveType=Original)
   > SetActiveCurves(self=@n18, CurveType=Original)
   > HighlightButton(Highlight Button=Get ExportOriginalGeneric_Button(), Un-Highlight Button=Get ExportModifiedGeneric_Button())
   n19 = UpdateCameraDistance!() -> (OutputDelegate, then, AuditionActor:Actor Object Reference, Distance:double=~) {is=title K2Node_CustomEvent}
   > n20 = Branch(Condition="Equal (Object)"(A=@n19.AuditionActor, B=Get AuditionActor() -> (AuditionActor:Actor Object Reference)) {is=title K2Node_CallFunction})
   n21 = Get AuditionActor()
   n22 = Get TargetCurveVector() -> (TargetCurveVector:Curve Vector Object Reference)
   Branch(execute=[@n20.then, @n30.then], Condition=AND Boolean(A=IsValid(Object=@n21), B=IsValid(Object=@n22)))
   > Branch(Condition="float > float"(A=Get MaxVectorCurveDistance() -> (MaxVectorCurveDistance:double), B=0) {is=title K2Node_CallFunction})
   > n23 = GetPIEWorlds!(execute, bIncludeDedicatedServer:bool=false) -> (then, ReturnValue:Array of World Object References)
   n24 = Is Valid Index!(TargetArray:Array of World Object References=@n23, IndexToTest:int) -> (ReturnValue:bool)
   > n25 = Branch(Condition=@n24)
   n26 = Get AuditionActor()
   n27 = Get Actor Location(self=GetPlayerPawn!(
     WorldContextObject:Object Reference="Get (a copy)"(Array:Array of World Object References=@n23) -> (Output:World Object Reference)
     PlayerIndex:int=0
    ) -> (ReturnValue:Pawn Object Reference))
   n28 = Get Editor Camera Info!(execute=@n25.else) -> (then, CameraLocation:Vector, ForwardDirection:Vector, FOV:float, ReturnValue:bool) {is=title K2Node_CallFunction}
   n29 = float / float(
    A="Distance (Vector)"!(V1:Vector=Select(Option 0:Vector=@n28, Option 1:Vector=@n27, Index=@n24) -> (ReturnValue:Vector), V2:Vector=Get Actor Location(self=@n26)) -> (ReturnValue:double)
    B=Get MaxVectorCurveDistance()
   ) {is=title K2Node_CallFunction}
   SetValue!(
    execute=[@n28.then, @n25.then]
    self:Slider Object Reference=Get Distance_Slider() -> (Distance_Slider:Analog Slider Object Reference)
    InValue:float=@n29
   ) -> (then)
   > OnDistanceSliderChanged(Value=@n29)
   n30 = Set AuditionActor(execute=@n20.else, AuditionActor:Actor Object Reference=@n19.AuditionActor) -> (Output_Get:Actor Object Reference)
   Event Tick()
   > UpdateCameraDistance!(execute, self, AuditionActor:Actor Object Reference=Get AuditionActor(), Distance:double=0) -> (then) {is=title K2Node_CallFunction}
   Event PreConstruct()
   > SetObject(
    self=[
     Get Curve01_Property() -> (Curve01_Property:Single Property View Object Reference)
     Get Curve02_Property() -> (Curve02_Property:Single Property View Object Reference)
     Get Curve03_Property() -> (Curve03_Property:Single Property View Object Reference)
     Get CloseAttenuationProperty() -> (CloseAttenuationProperty:Single Property View Object Reference)
     Get MediumAttenuationProperty() -> (MediumAttenuationProperty:Single Property View Object Reference)
     Get DistantAttenuationProperty() -> (DistantAttenuationProperty:Single Property View Object Reference)
     Get VectorCurveProperty() -> (VectorCurveProperty:Single Property View Object Reference)
     Get AuditionActor_Property() -> (AuditionActor_Property:Single Property View Object Reference)
    ]
    NewObject=Self-Reference()
   )
  ]}
 HighlightButton {nodes [
   n1 = HighlightButton!() -> (then, Highlight Button:Button Object Reference, Un-Highlight Button:Button Object Reference) {is=title K2Node_FunctionEntry}
   > SetBackgroundColor!(execute, self:Button Object Reference=@n1, InBackgroundColor:Linear Color Structure=(R=0.0392,G=1,B=0,A=1)) -> (then)
   > SetBackgroundColor(self=@n1.Un-Highlight Button, InBackgroundColor=(R=1,G=1,B=1,A=1))
  ]}
 OnDistanceSliderChanged {nodes [
   n1 = OnDistanceSliderChanged!() -> (then, Value:double=~) {is=title K2Node_FunctionEntry}
   > n2 = Get()
   > n3 = Cast To FortAudioCurveVector(Object=@n2)
   n4 = GetAudioCurveVectorValue!(self:Fort Audio Curve Vector Object Reference=@n3, InTime:float=@n1) -> (ReturnValue_X:double, ReturnValue_Y:double, ReturnValue_Z:double)
   > Set XValueAtTime(XValueAtTime:double=@n4) -> (Output_Get:double)
   > Set YValueAtTime(YValueAtTime:double=@n4.ReturnValue_Y) -> (Output_Get:double)
   > n5 = Set ZValueAtTime(ZValueAtTime:double=@n4.ReturnValue_Z) -> (Output_Get:double)
   n6 = Branch(execute=[@n19.then, @n5.then], Condition="float > float"(A=Get XValueAtTime() -> (XValueAtTime:double), B=0))
   n7 = Get X_Progressbar() -> (X_Progressbar:Progress Bar Object Reference)
   > SetPercent!(execute, self:Progress Bar Object Reference=@n7, InPercent:float=Get XValueAtTime()) -> (then)
   n8 = Get X_Progressbar()
   > n9 = SetFillColorAndOpacity!(execute, self:Progress Bar Object Reference=@n8, InColor:Linear Color Structure=Get X_Color() -> (X_Color:Linear Color Structure)) -> (then)
   n10 = Branch(execute=[@n9.then, @n18.then], Condition="float > float"(A=Get YValueAtTime() -> (YValueAtTime:double), B=0))
   n11 = Get Y_Progressbar() -> (Y_Progressbar:Progress Bar Object Reference)
   > SetPercent(self=@n11, InPercent=Get YValueAtTime())
   n12 = Get Y_Progressbar()
   > n13 = SetFillColorAndOpacity(self=@n12, InColor=Get Y_Color() -> (Y_Color:Linear Color Structure))
   n14 = Branch(execute=[@n13.then, @n17.then], Condition="float > float"(A=Get ZValueAtTime() -> (ZValueAtTime:double), B=0))
   n15 = Get Z_Progressbar() -> (Z_Progressbar:Progress Bar Object Reference)
   > SetPercent(self=@n15, InPercent=Get ZValueAtTime())
   n16 = Get Z_Progressbar()
   > SetFillColorAndOpacity(self=@n16, InColor=Get Z_Color() -> (Z_Color:Linear Color Structure))
   SetPercent(execute=@n14.else, self=@n15, InPercent=100)
   > SetFillColorAndOpacity(self=@n16, InColor=Get Null_Color() -> (Null_Color:Linear Color Structure))
   SetPercent(execute=@n10.else, self=@n11, InPercent=100)
   > n17 = SetFillColorAndOpacity(self=@n12, InColor=Get Null_Color())
   SetPercent(execute=@n6.else, self=@n7, InPercent=100)
   > n18 = SetFillColorAndOpacity(self=@n8, InColor=Get Null_Color())
   Set XValueAtTime(execute=@n2.else, XValueAtTime=0)
   > Set YValueAtTime(YValueAtTime=0)
   > n19 = Set ZValueAtTime(ZValueAtTime=0)
  ]}
 OnKeyDown {nodes [
   n1 = OnKeyDown!() -> (then, MyGeometry:Geometry Structure, InKeyEvent:Key Event Structure) {is=title K2Node_FunctionEntry}
   > n2 = Sequence()
   n3 = GetKey!(Input:"Key Event Structure (by ref)"=@n1.InKeyEvent) -> (ReturnValue:Key Structure)
   > n4 = Branch(Condition="Equal (Key)"!(A:Key Structure=@n3, B:Key Structure=PageUp) -> (ReturnValue:bool))
   > Adjust Volume in Decibels!(
    execute
    self:Sound Base Asset Actions Object Reference=Get AA_SoundBase() -> (AA_SoundBase:Sound Base Asset Actions Object Reference)
    Decibel Amount:double=3
    Rounded Result:bool=false
   ) -> (then)
   Branch(execute=@n4.else, Condition="Equal (Key)"(A=@n3, B=PageDown))
   > Adjust Volume in Decibels(self=Get AA_SoundBase(), Decibel Amount=-3, Rounded Result=false)
   n5 = GetInputEventFromKeyEvent!(Event:"Key Event Structure (by ref)"=@n1.InKeyEvent) -> (ReturnValue:Input Event Structure)
   n6 = GetKey(Input=@n1.InKeyEvent)
   Branch(
    execute=@n2.then_1
    Condition=AND Boolean(A="Equal (Key)"(A=@n6, B=A), B=Is Shift Down!(Input:"Input Event Structure (by ref)"=@n5) -> (ReturnValue:bool))
   )
   > PrintString(InString=Hello)
   > Return Node!(execute, ReturnValue:Event Reply Structure=Make Event Reply!() -> (EventReply:Event Reply Structure) {is=title K2Node_MakeStruct})
  ]}
]
variables [
 AA_SoundBase {type=SoundBaseAssetActions_C, default=null}
 AuditionActor {type=Actor, default=null}
 CloseAttenuation {type=SoundAttenuation, default=&Fort_Weapon_Projectile_Close_Athena_Att}
 Curve_01 {type=CurveFloat, default=/Game/Sounds/Attenuation/Curves/Fort_WeaponFire_Crossfade_Close_Athena}
 Curve_02 {type=CurveFloat, default=/Game/Sounds/Attenuation/Curves/Fort_WeaponFire_Crossfade_Medium_Athena}
 Curve_03 {type=CurveFloat, default=/Game/Sounds/Attenuation/Curves/Fort_WeaponFire_Crossfade_Distant_Athena}
 DefaultMaxVectorCurveDistance {type=number, default=10000}
 Distance_Text {type=number}
 DistanceSliderValue {type=number}
 DistantAttenuation {type=SoundAttenuation, default=&Fort_Weapon_Projectile_Distant_Mono_Athena_Att}
 MaxVectorCurveDistance {type=number, default=10000}
 MediumAttenuation {type=SoundAttenuation, default=&Fort_Weapon_Projectile_Medium_Mono_Athena_Att}
 Null_Color {type=LinearColor, default=<rgba 0.1146 0.1146 0.1146 1>}
 TargetCurveVector {type=CurveVector, default=null}
 X_Color {type=LinearColor, default=<rgba 1 0 0 1>}
 Y_Color {type=LinearColor, default=<rgba 0 0.8706 0.0549 1>}
 Z_Color {type=LinearColor, default=<rgba 0 0.102 1 1>}
]
widgets {
 info {rootWidgetClass=&CanvasPanel, widgetCount=39}
 widgets [
  CanvasPanel_0 {widgetClassPath=&CanvasPanel, widget=CanvasPanel_0}
  Border_1 {
   widgetClassPath=&Border
   parent=CanvasPanel_0
   widget=Border_1
   properties {brushColor=<rgba 0 0 0 0.1412>}
   slot {
    class=CanvasPanelSlot
    properties {bAutoSize=true, layoutData {alignment=<0 0>, anchors {maximum=<0 0>, minimum=<0 0>}, offsets=<ltrb 0 0 488.3884 397.8011>}}
   }
  }
  CanvasPanel_133 {widgetClassPath=&CanvasPanel, parent=Border_1, widget=CanvasPanel_133, slot {class=BorderSlot}}
  Border_3 {
   widgetClassPath=&Border
   parent=CanvasPanel_133
   widget=Border_3
   properties {background {
     bIsDynamicallyLoaded=false
     drawAs=Image
     imageType=NoImage
     mirroring=NoMirror
     resourceName=None
     resourceObject=null
     tiling=NoTile
     imageSize=<32 32>
     margin=<ltrb 0 0 0 0>
     outlineSettings {
      bUseBrushTransparency=false
      roundingType=HalfHeightRadius
      width=0
      color {colorUseRule=UseColor_Specified, specifiedColor=<rgba 0 0 0 0>}
      cornerRadii=<0 0 0 0>
     }
     tintColor {colorUseRule=UseColor_Specified, specifiedColor=<rgba 0.0518 0 0.2865 1>}
     uVRegion {bIsValid=false, max=<0 0>, min=<0 0>}
    }}
   slot {
    class=CanvasPanelSlot
    properties {bAutoSize=true, layoutData {alignment=<0 0>, anchors {maximum=<0 0>, minimum=<0 0>}, offsets=<ltrb 4 4 508.8088 392.0621>}}
   }
  }
  VectorCurveProperty {
   widgetClassPath=&SinglePropertyView
   parent=Border_3
   widget=VectorCurveProperty
   properties {propertyName=TargetCurveVector}
   slot {class=BorderSlot}
  }
  Border_202 {
   widgetClassPath=&Border
   parent=CanvasPanel_133
   widget=Border_202
   properties {renderOpacity=4.1055, brushColor=<rgba 1 1 1 0>, padding=<ltrb 0 0 0 0>}
   slot {
    class=CanvasPanelSlot
    properties {bAutoSize=true, layoutData {alignment=<0 0>, anchors {maximum=<0 0>, minimum=<0 0>}, offsets=<ltrb 34.8832 37.2621 100 30>}}
   }
  }
  CanvasPanel_83 {
   widgetClassPath=&CanvasPanel
   parent=CanvasPanel_133
   widget=CanvasPanel_83
   slot {
    class=CanvasPanelSlot
    properties {bAutoSize=true, layoutData {alignment=<0 0>, anchors {maximum=<0 0>, minimum=<0 0>}, offsets=<ltrb 0 108 115.015 45.4154>}}
   }
  }
  Z_Progressbar {
   widgetClassPath=/Script/UMG.ProgressBar
   bIsVariable=true
   parent=CanvasPanel_83
   widget=Z_Progressbar
   properties {
    fillColorAndOpacity=<rgba 0 0.102 1 1>
    renderTransform {angle=-90, scale=<1 1.3>, shear=<0 0>, translation=<0 0>}
    widgetStyle {
     enableFillAnimation=true
     backgroundImage {
      bIsDynamicallyLoaded=false
      drawAs=Image
      imageType=NoImage
      mirroring=NoMirror
      resourceName=None
      resourceObject=null
      tiling=NoTile
      imageSize=<0 0>
      margin=<ltrb 0 0 0 0>
      outlineSettings {
       bUseBrushTransparency=false
       roundingType=HalfHeightRadius
       width=0
       color {colorUseRule=UseColor_Specified, specifiedColor=<rgba 0 0 0 0>}
       cornerRadii=<0 0 0 0>
      }
      tintColor {colorUseRule=UseColor_Specified, specifiedColor=<rgba 0.0048 0.0048 0.0048 1>}
      uVRegion {bIsValid=false, max=<0 0>, min=<0 0>}
     }
     fillImage {
      bIsDynamicallyLoaded=false
      drawAs=Image
      imageType=FullColor
      mirroring=NoMirror
      resourceName=../../../Engine/Content/Slate/Starship/CoreWidgets/ProgressBar/ProgressMarquee.png
      resourceObject=null
      tiling=Horizontal
      imageSize=<20 12>
      margin=<ltrb 0 0 0 0>
      outlineSettings {
       bUseBrushTransparency=false
       roundingType=HalfHeightRadius
       width=0
       color {colorUseRule=UseColor_Specified, specifiedColor=<rgba 0 0 0 0>}
       cornerRadii=<0 0 0 0>
      }
      tintColor {colorUseRule=UseColor_Specified, specifiedColor=<rgba 1 1 1 1>}
      uVRegion {bIsValid=false, max=<0 0>, min=<0 0>}
     }
     marqueeImage {
      bIsDynamicallyLoaded=false
      drawAs=Image
      imageType=FullColor
      mirroring=NoMirror
      resourceName=../../../Engine/Content/Slate/Starship/CoreWidgets/ProgressBar/ProgressMarquee.png
      resourceObject=null
      tiling=Horizontal
      imageSize=<20 12>
      margin=<ltrb 0 0 0 0>
      outlineSettings {
       bUseBrushTransparency=false
       roundingType=HalfHeightRadius
       width=0
       color {colorUseRule=UseColor_Specified, specifiedColor=<rgba 0 0 0 0>}
       cornerRadii=<0 0 0 0>
      }
      tintColor {colorUseRule=UseColor_Specified, specifiedColor=<rgba 0.0044 0.2384 1 1>}
      uVRegion {bIsValid=false, max=<0 0>, min=<0 0>}
     }
    }
   }
   slot {
    class=CanvasPanelSlot
    properties {layoutData {alignment=<0 0>, anchors {maximum=<0 0>, minimum=<0 0>}, offsets=<ltrb 360 76 68.4685 38.4985>}}
   }
  }
  Y_Progressbar {
   widgetClassPath=&ProgressBar
   bIsVariable=true
   parent=CanvasPanel_83
   widget=Y_Progressbar
   properties {
    fillColorAndOpacity=<rgba 0 0.8706 0.0549 1>
    renderTransform {angle=-90, scale=<1 1.3>, shear=<0 0>, translation=<0 0>}
    widgetStyle {
     enableFillAnimation=true
     backgroundImage {
      bIsDynamicallyLoaded=false
      drawAs=Image
      imageType=NoImage
      mirroring=NoMirror
      resourceName=None
      resourceObject=null
      tiling=NoTile
      imageSize=<0 0>
      margin=<ltrb 0 0 0 0>
      outlineSettings {
       bUseBrushTransparency=false
       roundingType=HalfHeightRadius
       width=0
       color {colorUseRule=UseColor_Specified, specifiedColor=<rgba 0 0 0 0>}
       cornerRadii=<0 0 0 0>
      }
      tintColor {colorUseRule=UseColor_Specified, specifiedColor=<rgba 0.0048 0.0048 0.0048 1>}
      uVRegion {bIsValid=false, max=<0 0>, min=<0 0>}
     }
     fillImage {
      bIsDynamicallyLoaded=false
      drawAs=Image
      imageType=FullColor
      mirroring=NoMirror
      resourceName=../../../Engine/Content/Slate/Starship/CoreWidgets/ProgressBar/ProgressMarquee.png
      resourceObject=null
      tiling=Horizontal
      imageSize=<20 12>
      margin=<ltrb 0 0 0 0>
      outlineSettings {
       bUseBrushTransparency=false
       roundingType=HalfHeightRadius
       width=0
       color {colorUseRule=UseColor_Specified, specifiedColor=<rgba 0 0 0 0>}
       cornerRadii=<0 0 0 0>
      }
      tintColor {colorUseRule=UseColor_Specified, specifiedColor=<rgba 1 1 1 1>}
      uVRegion {bIsValid=false, max=<0 0>, min=<0 0>}
     }
     marqueeImage {
      bIsDynamicallyLoaded=false
      drawAs=Image
      imageType=FullColor
      mirroring=NoMirror
      resourceName=../../../Engine/Content/Slate/Starship/CoreWidgets/ProgressBar/ProgressMarquee.png
      resourceObject=null
      tiling=Horizontal
      imageSize=<20 12>
      margin=<ltrb 0 0 0 0>
      outlineSettings {
       bUseBrushTransparency=false
       roundingType=HalfHeightRadius
       width=0
       color {colorUseRule=UseColor_Specified, specifiedColor=<rgba 0 0 0 0>}
       cornerRadii=<0 0 0 0>
      }
      tintColor {colorUseRule=UseColor_Specified, specifiedColor=<rgba 0.0044 0.2384 1 1>}
      uVRegion {bIsValid=false, max=<0 0>, min=<0 0>}
     }
    }
   }
   slot {
    class=CanvasPanelSlot
    properties {layoutData {alignment=<0 0>, anchors {maximum=<0 0>, minimum=<0 0>}, offsets=<ltrb 288 76 68.4685 38.4985>}}
   }
  }
  X_Progressbar {
   widgetClassPath=&ProgressBar
   bIsVariable=true
   parent=CanvasPanel_83
   widget=X_Progressbar
   properties {
    fillColorAndOpacity=<rgba 1 0 0 1>
    renderTransform {angle=-90, scale=<1 1.3>, shear=<0 0>, translation=<0 0>}
    widgetStyle {
     enableFillAnimation=true
     backgroundImage {
      bIsDynamicallyLoaded=false
      drawAs=Image
      imageType=NoImage
      mirroring=NoMirror
      resourceName=None
      resourceObject=null
      tiling=NoTile
      imageSize=<0 0>
      margin=<ltrb 0 0 0 0>
      outlineSettings {
       bUseBrushTransparency=false
       roundingType=HalfHeightRadius
       width=0
       color {colorUseRule=UseColor_Specified, specifiedColor=<rgba 0 0 0 0>}
       cornerRadii=<0 0 0 0>
      }
      tintColor {colorUseRule=UseColor_Specified, specifiedColor=<rgba 0.0048 0.0048 0.0048 1>}
      uVRegion {bIsValid=false, max=<0 0>, min=<0 0>}
     }
     fillImage {
      bIsDynamicallyLoaded=false
      drawAs=Image
      imageType=FullColor
      mirroring=NoMirror
      resourceName=../../../Engine/Content/Slate/Starship/CoreWidgets/ProgressBar/ProgressMarquee.png
      resourceObject=null
      tiling=Horizontal
      imageSize=<20 12>
      margin=<ltrb 0 0 0 0>
      outlineSettings {
       bUseBrushTransparency=false
       roundingType=HalfHeightRadius
       width=0
       color {colorUseRule=UseColor_Specified, specifiedColor=<rgba 0 0 0 0>}
       cornerRadii=<0 0 0 0>
      }
      tintColor {colorUseRule=UseColor_Specified, specifiedColor=<rgba 1 1 1 1>}
      uVRegion {bIsValid=false, max=<0 0>, min=<0 0>}
     }
     marqueeImage {
      bIsDynamicallyLoaded=false
      drawAs=Image
      imageType=FullColor
      mirroring=NoMirror
      resourceName=../../../Engine/Content/Slate/Starship/CoreWidgets/ProgressBar/ProgressMarquee.png
      resourceObject=null
      tiling=Horizontal
      imageSize=<20 12>
      margin=<ltrb 0 0 0 0>
      outlineSettings {
       bUseBrushTransparency=false
       roundingType=HalfHeightRadius
       width=0
       color {colorUseRule=UseColor_Specified, specifiedColor=<rgba 0 0 0 0>}
       cornerRadii=<0 0 0 0>
      }
      tintColor {colorUseRule=UseColor_Specified, specifiedColor=<rgba 0.0044 0.2384 1 1>}
      uVRegion {bIsValid=false, max=<0 0>, min=<0 0>}
     }
    }
   }
   slot {
    class=CanvasPanelSlot
    properties {layoutData {alignment=<0 0>, anchors {maximum=<0 0>, minimum=<0 0>}, offsets=<ltrb 216 76 68.4685 38.4985>}}
   }
  }
  Distance_Slider {
   widgetClassPath=/Script/CommonUI.AnalogSlider
   bIsVariable=true
   parent=CanvasPanel_83
   widget=Distance_Slider
   properties {
    renderTransform {angle=0, scale=<1 2.05>, shear=<0 0>, translation=<0 0>}
    widgetStyle {
     barThickness=4
     disabledBarImage {
      bIsDynamicallyLoaded=false
      drawAs=RoundedBox
      imageType=NoImage
      mirroring=NoMirror
      resourceName=None
      resourceObject=null
      tiling=NoTile
      imageSize=<32 32>
      margin=<ltrb 0 0 0 0>
      outlineSettings {
       bUseBrushTransparency=false
       roundingType=FixedRadius
       width=1
       color {colorUseRule=UseColor_ColorTable, specifiedColor=<rgba 0 0 0 0>}
       cornerRadii=<2 2 2 2>
      }
      tintColor {colorUseRule=UseColor_Specified, specifiedColor=<rgba 1 1 1 1>}
      uVRegion {bIsValid=false, max=<0 0>, min=<0 0>}
     }
     disabledThumbImage {
      bIsDynamicallyLoaded=false
      drawAs=RoundedBox
      imageType=NoImage
      mirroring=NoMirror
      resourceName=None
      resourceObject=null
      tiling=NoTile
      imageSize=<32 32>
      margin=<ltrb 0 0 0 0>
      outlineSettings {
       bUseBrushTransparency=false
       roundingType=HalfHeightRadius
       width=0
       color {colorUseRule=UseColor_Specified, specifiedColor=<rgba 0 0 0 0>}
       cornerRadii=<0 0 0 0>
      }
      tintColor {colorUseRule=UseColor_Specified, specifiedColor=<rgba 0.5271 0.5271 0.5271 1>}
      uVRegion {bIsValid=false, max=<0 0>, min=<0 0>}
     }
     hoveredBarImage {
      bIsDynamicallyLoaded=false
      drawAs=RoundedBox
      imageType=NoImage
      mirroring=NoMirror
      resourceName=None
      resourceObject=null
      tiling=NoTile
      imageSize=<0 0>
      margin=<ltrb 0 0 0 0>
      outlineSettings {
       bUseBrushTransparency=false
       roundingType=FixedRadius
       width=1
       color {colorUseRule=UseColor_ColorTable, specifiedColor=<rgba 0 0 0 0>}
       cornerRadii=<2 2 2 2>
      }
      tintColor {colorUseRule=UseColor_Specified, specifiedColor=<rgba 0.0048 0.0048 0.0048 1>}
      uVRegion {bIsValid=false, max=<0 0>, min=<0 0>}
     }
     hoveredThumbImage {
      bIsDynamicallyLoaded=false
      drawAs=RoundedBox
      imageType=NoImage
      mirroring=NoMirror
      resourceName=None
      resourceObject=null
      tiling=NoTile
      imageSize=<8 8>
      margin=<ltrb 0 0 0 0>
      outlineSettings {
       bUseBrushTransparency=false
       roundingType=HalfHeightRadius
       width=0
       color {colorUseRule=UseColor_Specified, specifiedColor=<rgba 0 0 0 0>}
       cornerRadii=<0 0 0 0>
      }
      tintColor {colorUseRule=UseColor_Specified, specifiedColor=<rgba 1 1 1 1>}
      uVRegion {bIsValid=false, max=<0 0>, min=<0 0>}
     }
     normalBarImage {
      bIsDynamicallyLoaded=false
      drawAs=RoundedBox
      imageType=NoImage
      mirroring=NoMirror
      resourceName=None
      resourceObject=null
      tiling=NoTile
      imageSize=<0 0>
      margin=<ltrb 0 0 0 0>
      outlineSettings {
       bUseBrushTransparency=false
       roundingType=FixedRadius
       width=1
       color {colorUseRule=UseColor_ColorTable, specifiedColor=<rgba 0 0 0 0>}
       cornerRadii=<2 2 2 2>
      }
      tintColor {colorUseRule=UseColor_Specified, specifiedColor=<rgba 0.0048 0.0048 0.0048 1>}
      uVRegion {bIsValid=false, max=<0 0>, min=<0 0>}
     }
     normalThumbImage {
      bIsDynamicallyLoaded=false
      drawAs=RoundedBox
      imageType=NoImage
      mirroring=NoMirror
      resourceName=None
      resourceObject=null
      tiling=NoTile
      imageSize=<8 8>
      margin=<ltrb 0 0 0 0>
      outlineSettings {
       bUseBrushTransparency=false
       roundingType=HalfHeightRadius
       width=0
       color {colorUseRule=UseColor_Specified, specifiedColor=<rgba 0 0 0 0>}
       cornerRadii=<0 0 0 0>
      }
      tintColor {colorUseRule=UseColor_Specified, specifiedColor=<rgba 0.5271 0.5271 0.5271 1>}
      uVRegion {bIsValid=false, max=<0 0>, min=<0 0>}
     }
    }
   }
   slot {
    class=CanvasPanelSlot
    properties {layoutData {alignment=<0 0>, anchors {maximum=<0 0>, minimum=<0 0>}, offsets=<ltrb 0 96 201.5561 23.659>}}
   }
  }
  Alpha_LabelText {
   widgetClassPath=&CommonTextBlock
   parent=CanvasPanel_83
   widget=Alpha_LabelText
   properties {
    justification=Right
    style=&TextStyle-Base_C
    text=Value
    font {
     bForceMonospaced=false
     bMaterialIsStencil=false
     fontMaterial=null
     fontObject=&NotoSans
     letterSpacing=0
     monospacedWidth=1
     size=24
     skewAmount=0
     typefaceFontName=NotoSans-Regular
     outlineSettings {
      bApplyOutlineToDropShadows=false
      bMiteredCorners=false
      bSeparateFillAlpha=false
      outlineMaterial=null
      outlineSize=0
      outlineColor=<rgba 0 0 0 1>
     }
    }
    renderTransform {angle=0, scale=<0.5 0.5>, shear=<0 0>, translation=<0 0>}
    shadowOffset=<0 0>
   }
   slot {class=CanvasPanelSlot, properties {layoutData {alignment=<0 0>, anchors {maximum=<0 0>, minimum=<0 0>}, offsets=<ltrb -36 52 153 44>}}}
  }
  Distance_LabelText {
   widgetClassPath=&CommonTextBlock
   parent=CanvasPanel_83
   widget=Distance_LabelText
   properties {
    justification=Right
    style=&TextStyle-Base_C
    text=Distance
    font {
     bForceMonospaced=false
     bMaterialIsStencil=false
     fontMaterial=null
     fontObject=&NotoSans
     letterSpacing=0
     monospacedWidth=1
     size=24
     skewAmount=0
     typefaceFontName=NotoSans-Regular
     outlineSettings {
      bApplyOutlineToDropShadows=false
      bMiteredCorners=false
      bSeparateFillAlpha=false
      outlineMaterial=null
      outlineSize=0
      outlineColor=<rgba 0 0 0 1>
     }
    }
    renderTransform {angle=0, scale=<0.5 0.5>, shear=<0 0>, translation=<0 0>}
    shadowOffset=<0 0>
   }
   slot {class=CanvasPanelSlot, properties {layoutData {alignment=<0 0>, anchors {maximum=<0 0>, minimum=<0 0>}, offsets=<ltrb -36 24 153 44>}}}
  }
  CommonTextBlock_1 {
   widgetClassPath=&CommonTextBlock
   parent=CanvasPanel_83
   widget=CommonTextBlock_1
   properties {
    style=&TextStyle-Base_C
    font {
     bForceMonospaced=false
     bMaterialIsStencil=false
     fontMaterial=null
     fontObject=&NotoSans
     letterSpacing=0
     monospacedWidth=1
     size=24
     skewAmount=0
     typefaceFontName=NotoSans-Regular
     outlineSettings {
      bApplyOutlineToDropShadows=false
      bMiteredCorners=false
      bSeparateFillAlpha=false
      outlineMaterial=null
      outlineSize=0
      outlineColor=<rgba 0 0 0 1>
     }
    }
    shadowOffset=<0 0>
   }
   slot {
    class=CanvasPanelSlot
    properties {layoutData {alignment=<0 0>, anchors {maximum=<0 0>, minimum=<0 0>}, offsets=<ltrb 24 133.8378 153 6.1622>}}
   }
  }
  AuditionActor_Property {
   widgetClassPath=&SinglePropertyView
   parent=CanvasPanel_83
   widget=AuditionActor_Property
   properties {propertyName=AuditionActor}
   slot {class=CanvasPanelSlot, properties {layoutData {alignment=<0 0>, anchors {maximum=<0 0>, minimum=<0 0>}, offsets=<ltrb 4 -36 424.9249 30>}}}
  }
  Border {
   widgetClassPath=&Border
   parent=CanvasPanel_0
   widget=Border
   properties {horizontalAlignment=HAlign_Left, verticalAlignment=VAlign_Top, brushColor=<rgba 0 0 0 0.1412>, padding=<ltrb 10 10 10 10>}
   slot {
    class=CanvasPanelSlot
    properties {bAutoSize=true, layoutData {alignment=<0 0>, anchors {maximum=<0 0>, minimum=<0 0>}, offsets=<ltrb 0 260 640.941 491.7618>}}
   }
  }
  CanvasPanel {
   widgetClassPath=&CanvasPanel
   parent=Border
   widget=CanvasPanel
   slot {class=BorderSlot, properties {horizontalAlignment=HAlign_Left, verticalAlignment=VAlign_Top, padding=<ltrb 10 10 10 10>}}
  }
  ExportModified_Button {
   widgetClassPath=&Button
   parent=CanvasPanel
   widget=ExportModified_Button
   properties {
    bIsEnabled=false
    widgetStyle {
     clickedSlateSound {resourceObject=null}
     disabled {
      bIsDynamicallyLoaded=false
      drawAs=RoundedBox
      imageType=NoImage
      mirroring=NoMirror
      resourceName=None
      resourceObject=null
      tiling=NoTile
      imageSize=<0 0>
      margin=<ltrb 0 0 0 0>
      outlineSettings {
       bUseBrushTransparency=false
       roundingType=FixedRadius
       width=1
       color {colorUseRule=UseColor_ColorTable, specifiedColor=<rgba 0 0 0 0>}
       cornerRadii=<4 4 4 4>
      }
      tintColor {colorUseRule=UseColor_Specified, specifiedColor=<rgba 0.0395 0.0395 0.0395 1>}
      uVRegion {bIsValid=false, max=<0 0>, min=<0 0>}
     }
     disabledForeground {colorUseRule=UseColor_ColorTable, specifiedColor=<rgba 0 0 0 0>}
     hovered {
      bIsDynamicallyLoaded=false
      drawAs=RoundedBox
      imageType=NoImage
      mirroring=NoMirror
      resourceName=None
      resourceObject=null
      tiling=NoTile
      imageSize=<0 0>
      margin=<ltrb 0 0 0 0>
      outlineSettings {
       bUseBrushTransparency=false
       roundingType=FixedRadius
       width=1
       color {colorUseRule=UseColor_ColorTable, specifiedColor=<rgba 0 0 0 0>}
       cornerRadii=<4 4 4 4>
      }
      tintColor {colorUseRule=UseColor_Specified, specifiedColor=<rgba 0.0953 0.0953 0.0953 1>}
      uVRegion {bIsValid=false, max=<0 0>, min=<0 0>}
     }
     hoveredForeground {colorUseRule=UseColor_ColorTable, specifiedColor=<rgba 0 0 0 0>}
     hoveredSlateSound {resourceObject=null}
     normal {
      bIsDynamicallyLoaded=false
      drawAs=RoundedBox
      imageType=NoImage
      mirroring=NoMirror
      resourceName=None
      resourceObject=null
      tiling=NoTile
      imageSize=<0 0>
      margin=<ltrb 0 0 0 0>
      outlineSettings {
       bUseBrushTransparency=false
       roundingType=FixedRadius
       width=1
       color {colorUseRule=UseColor_ColorTable, specifiedColor=<rgba 0 0 0 0>}
       cornerRadii=<4 4 4 4>
      }
      tintColor {colorUseRule=UseColor_Specified, specifiedColor=<rgba 0.0395 0.0395 0.0395 1>}
      uVRegion {bIsValid=false, max=<0 0>, min=<0 0>}
     }
     normalForeground {colorUseRule=UseColor_ColorTable, specifiedColor=<rgba 0 0 0 0>}
     normalPadding=<ltrb 12 1.5 12 1.5>
     pressed {
      bIsDynamicallyLoaded=false
      drawAs=RoundedBox
      imageType=NoImage
      mirroring=NoMirror
      resourceName=None
      resourceObject=null
      tiling=NoTile
      imageSize=<0 0>
      margin=<ltrb 0 0 0 0>
      outlineSettings {
       bUseBrushTransparency=false
       roundingType=FixedRadius
       width=1
       color {colorUseRule=UseColor_ColorTable, specifiedColor=<rgba 0 0 0 0>}
       cornerRadii=<4 4 4 4>
      }
      tintColor {colorUseRule=UseColor_Specified, specifiedColor=<rgba 0.0284 0.0284 0.0284 1>}
      uVRegion {bIsValid=false, max=<0 0>, min=<0 0>}
     }
     pressedForeground {colorUseRule=UseColor_ColorTable, specifiedColor=<rgba 0 0 0 0>}
     pressedPadding=<ltrb 12 2 12 0>
     pressedSlateSound {resourceObject=null}
    }
   }
   slot {class=CanvasPanelSlot, properties {layoutData {alignment=<0 0>, anchors {maximum=<0 0>, minimum=<0 0>}, offsets=<ltrb 8 220 100 40>}}}
  }
  ExportModified_Text {
   widgetClassPath=&CommonTextBlock
   bIsVariable=true
   parent=ExportModified_Button
   widget=ExportModified_Text
   properties {
    justification=Center
    style=/FortUILibrary/Styles/Typography/Foundation/TextStyle-Base-M-B_Black.TextStyle-Base-M-B_Black_C
    text="Export\r\nModified"
    colorAndOpacity {colorUseRule=UseColor_Specified, specifiedColor=<rgba 0 0 0 1>}
    font {
     bForceMonospaced=false
     bMaterialIsStencil=false
     fontMaterial=null
     fontObject=&NotoSans
     letterSpacing=0
     monospacedWidth=1
     size=24
     skewAmount=0
     typefaceFontName=NotoSans-Bold
     outlineSettings {
      bApplyOutlineToDropShadows=false
      bMiteredCorners=false
      bSeparateFillAlpha=false
      outlineMaterial=null
      outlineSize=0
      outlineColor=<rgba 0 0 0 1>
     }
    }
    renderTransform {angle=0, scale=<0.45 0.4>, shear=<0 0>, translation=<0 -11>}
    shadowOffset=<0 0>
   }
   slot {class=ButtonSlot}
  }
  ExportOriginals_Button {
   widgetClassPath=&Button
   parent=CanvasPanel
   widget=ExportOriginals_Button
   properties {
    bIsEnabled=false
    widgetStyle {
     clickedSlateSound {resourceObject=null}
     disabled {
      bIsDynamicallyLoaded=false
      drawAs=RoundedBox
      imageType=NoImage
      mirroring=NoMirror
      resourceName=None
      resourceObject=null
      tiling=NoTile
      imageSize=<0 0>
      margin=<ltrb 0 0 0 0>
      outlineSettings {
       bUseBrushTransparency=false
       roundingType=FixedRadius
       width=1
       color {colorUseRule=UseColor_ColorTable, specifiedColor=<rgba 0 0 0 0>}
       cornerRadii=<4 4 4 4>
      }
      tintColor {colorUseRule=UseColor_Specified, specifiedColor=<rgba 0.0395 0.0395 0.0395 1>}
      uVRegion {bIsValid=false, max=<0 0>, min=<0 0>}
     }
     disabledForeground {colorUseRule=UseColor_ColorTable, specifiedColor=<rgba 0 0 0 0>}
     hovered {
      bIsDynamicallyLoaded=false
      drawAs=RoundedBox
      imageType=NoImage
      mirroring=NoMirror
      resourceName=None
      resourceObject=null
      tiling=NoTile
      imageSize=<0 0>
      margin=<ltrb 0 0 0 0>
      outlineSettings {
       bUseBrushTransparency=false
       roundingType=FixedRadius
       width=1
       color {colorUseRule=UseColor_ColorTable, specifiedColor=<rgba 0 0 0 0>}
       cornerRadii=<4 4 4 4>
      }
      tintColor {colorUseRule=UseColor_Specified, specifiedColor=<rgba 0.0953 0.0953 0.0953 1>}
      uVRegion {bIsValid=false, max=<0 0>, min=<0 0>}
     }
     hoveredForeground {colorUseRule=UseColor_ColorTable, specifiedColor=<rgba 0 0 0 0>}
     hoveredSlateSound {resourceObject=null}
     normal {
      bIsDynamicallyLoaded=false
      drawAs=RoundedBox
      imageType=NoImage
      mirroring=NoMirror
      resourceName=None
      resourceObject=null
      tiling=NoTile
      imageSize=<0 0>
      margin=<ltrb 0 0 0 0>
      outlineSettings {
       bUseBrushTransparency=false
       roundingType=FixedRadius
       width=1
       color {colorUseRule=UseColor_ColorTable, specifiedColor=<rgba 0 0 0 0>}
       cornerRadii=<4 4 4 4>
      }
      tintColor {colorUseRule=UseColor_Specified, specifiedColor=<rgba 0.0395 0.0395 0.0395 1>}
      uVRegion {bIsValid=false, max=<0 0>, min=<0 0>}
     }
     normalForeground {colorUseRule=UseColor_ColorTable, specifiedColor=<rgba 0 0 0 0>}
     normalPadding=<ltrb 12 1.5 12 1.5>
     pressed {
      bIsDynamicallyLoaded=false
      drawAs=RoundedBox
      imageType=NoImage
      mirroring=NoMirror
      resourceName=None
      resourceObject=null
      tiling=NoTile
      imageSize=<0 0>
      margin=<ltrb 0 0 0 0>
      outlineSettings {
       bUseBrushTransparency=false
       roundingType=FixedRadius
       width=1
       color {colorUseRule=UseColor_ColorTable, specifiedColor=<rgba 0 0 0 0>}
       cornerRadii=<4 4 4 4>
      }
      tintColor {colorUseRule=UseColor_Specified, specifiedColor=<rgba 0.0284 0.0284 0.0284 1>}
      uVRegion {bIsValid=false, max=<0 0>, min=<0 0>}
     }
     pressedForeground {colorUseRule=UseColor_ColorTable, specifiedColor=<rgba 0 0 0 0>}
     pressedPadding=<ltrb 12 2 12 0>
     pressedSlateSound {resourceObject=null}
    }
   }
   slot {class=CanvasPanelSlot, properties {layoutData {alignment=<0 0>, anchors {maximum=<0 0>, minimum=<0 0>}, offsets=<ltrb 112 220 100 40>}}}
  }
  ExportOriginals_Text {
   widgetClassPath=&CommonTextBlock
   bIsVariable=true
   parent=ExportOriginals_Button
   widget=ExportOriginals_Text
   properties {
    justification=Center
    style=&TextStyle-Base-M-B_Black_C
    text="Export\r\nOriginals"
    colorAndOpacity {colorUseRule=UseColor_Specified, specifiedColor=<rgba 0 0 0 1>}
    font {
     bForceMonospaced=false
     bMaterialIsStencil=false
     fontMaterial=null
     fontObject=&NotoSans
     letterSpacing=0
     monospacedWidth=1
     size=24
     skewAmount=0
     typefaceFontName=NotoSans-Bold
     outlineSettings {
      bApplyOutlineToDropShadows=false
      bMiteredCorners=false
      bSeparateFillAlpha=false
      outlineMaterial=null
      outlineSize=0
      outlineColor=<rgba 0 0 0 1>
     }
    }
    renderTransform {angle=0, scale=<0.45 0.4>, shear=<0 0>, translation=<0 -11>}
    shadowOffset=<0 0>
   }
   slot {class=ButtonSlot}
  }
  ImportCurves_Button {
   widgetClassPath=&Button
   parent=CanvasPanel
   widget=ImportCurves_Button
   properties {
    bIsEnabled=false
    widgetStyle {
     clickedSlateSound {resourceObject=null}
     disabled {
      bIsDynamicallyLoaded=false
      drawAs=RoundedBox
      imageType=NoImage
      mirroring=NoMirror
      resourceName=None
      resourceObject=null
      tiling=NoTile
      imageSize=<0 0>
      margin=<ltrb 0 0 0 0>
      outlineSettings {
       bUseBrushTransparency=false
       roundingType=FixedRadius
       width=1
       color {colorUseRule=UseColor_ColorTable, specifiedColor=<rgba 0 0 0 0>}
       cornerRadii=<4 4 4 4>
      }
      tintColor {colorUseRule=UseColor_Specified, specifiedColor=<rgba 0.0395 0.0395 0.0395 1>}
      uVRegion {bIsValid=false, max=<0 0>, min=<0 0>}
     }
     disabledForeground {colorUseRule=UseColor_ColorTable, specifiedColor=<rgba 0 0 0 0>}
     hovered {
      bIsDynamicallyLoaded=false
      drawAs=RoundedBox
      imageType=NoImage
      mirroring=NoMirror
      resourceName=None
      resourceObject=null
      tiling=NoTile
      imageSize=<0 0>
      margin=<ltrb 0 0 0 0>
      outlineSettings {
       bUseBrushTransparency=false
       roundingType=FixedRadius
       width=1
       color {colorUseRule=UseColor_ColorTable, specifiedColor=<rgba 0 0 0 0>}
       cornerRadii=<4 4 4 4>
      }
      tintColor {colorUseRule=UseColor_Specified, specifiedColor=<rgba 0.0953 0.0953 0.0953 1>}
      uVRegion {bIsValid=false, max=<0 0>, min=<0 0>}
     }
     hoveredForeground {colorUseRule=UseColor_ColorTable, specifiedColor=<rgba 0 0 0 0>}
     hoveredSlateSound {resourceObject=null}
     normal {
      bIsDynamicallyLoaded=false
      drawAs=RoundedBox
      imageType=NoImage
      mirroring=NoMirror
      resourceName=None
      resourceObject=null
      tiling=NoTile
      imageSize=<0 0>
      margin=<ltrb 0 0 0 0>
      outlineSettings {
       bUseBrushTransparency=false
       roundingType=FixedRadius
       width=1
       color {colorUseRule=UseColor_ColorTable, specifiedColor=<rgba 0 0 0 0>}
       cornerRadii=<4 4 4 4>
      }
      tintColor {colorUseRule=UseColor_Specified, specifiedColor=<rgba 0.0395 0.0395 0.0395 1>}
      uVRegion {bIsValid=false, max=<0 0>, min=<0 0>}
     }
     normalForeground {colorUseRule=UseColor_ColorTable, specifiedColor=<rgba 0 0 0 0>}
     normalPadding=<ltrb 12 1.5 12 1.5>
     pressed {
      bIsDynamicallyLoaded=false
      drawAs=RoundedBox
      imageType=NoImage
      mirroring=NoMirror
      resourceName=None
      resourceObject=null
      tiling=NoTile
      imageSize=<0 0>
      margin=<ltrb 0 0 0 0>
      outlineSettings {
       bUseBrushTransparency=false
       roundingType=FixedRadius
       width=1
       color {colorUseRule=UseColor_ColorTable, specifiedColor=<rgba 0 0 0 0>}
       cornerRadii=<4 4 4 4>
      }
      tintColor {colorUseRule=UseColor_Specified, specifiedColor=<rgba 0.0284 0.0284 0.0284 1>}
      uVRegion {bIsValid=false, max=<0 0>, min=<0 0>}
     }
     pressedForeground {colorUseRule=UseColor_ColorTable, specifiedColor=<rgba 0 0 0 0>}
     pressedPadding=<ltrb 12 2 12 0>
     pressedSlateSound {resourceObject=null}
    }
   }
   slot {class=CanvasPanelSlot, properties {layoutData {alignment=<0 0>, anchors {maximum=<0 0>, minimum=<0 0>}, offsets=<ltrb 420 0 100 40>}}}
  }
  VectorImport {
   widgetClassPath=&CommonTextBlock
   parent=ImportCurves_Button
   widget=VectorImport
   properties {
    justification=Center
    style=&TextStyle-Base-M-B_Black_C
    text="Import\r\nAttenuations"
    colorAndOpacity {colorUseRule=UseColor_Specified, specifiedColor=<rgba 0 0 0 1>}
    font {
     bForceMonospaced=false
     bMaterialIsStencil=false
     fontMaterial=null
     fontObject=&NotoSans
     letterSpacing=0
     monospacedWidth=1
     size=24
     skewAmount=0
     typefaceFontName=NotoSans-Bold
     outlineSettings {
      bApplyOutlineToDropShadows=false
      bMiteredCorners=false
      bSeparateFillAlpha=false
      outlineMaterial=null
      outlineSize=0
      outlineColor=<rgba 0 0 0 1>
     }
    }
    renderTransform {angle=0, scale=<0.4 0.5>, shear=<0 0>, translation=<0 -15>}
    shadowOffset=<0 0>
   }
   slot {class=ButtonSlot}
  }
  DistantAttenuationProperty {
   widgetClassPath=&SinglePropertyView
   parent=CanvasPanel
   widget=DistantAttenuationProperty
   properties {propertyName=DistantAttenuation}
   slot {
    class=CanvasPanelSlot
    properties {bAutoSize=true, layoutData {alignment=<0 0>, anchors {maximum=<0 0>, minimum=<0 0>}, offsets=<ltrb 8 156 292.1922 44.5045>}}
   }
  }
  MediumAttenuationProperty {
   widgetClassPath=&SinglePropertyView
   parent=CanvasPanel
   widget=MediumAttenuationProperty
   properties {propertyName=MediumAttenuation}
   slot {
    class=CanvasPanelSlot
    properties {bAutoSize=true, layoutData {alignment=<0 0>, anchors {maximum=<0 0>, minimum=<0 0>}, offsets=<ltrb 4 100 292.1922 86.5465>}}
   }
  }
  CloseAttenuationProperty {
   widgetClassPath=&SinglePropertyView
   parent=CanvasPanel
   widget=CloseAttenuationProperty
   properties {propertyName=CloseAttenuation}
   slot {
    class=CanvasPanelSlot
    properties {bAutoSize=true, layoutData {alignment=<0 0>, anchors {maximum=<0 0>, minimum=<0 0>}, offsets=<ltrb 16 44 292.1922 44.5045>}}
   }
  }
  CommonTextBlock_2 {
   widgetClassPath=&CommonTextBlock
   parent=CanvasPanel
   widget=CommonTextBlock_2
   properties {
    style=&TextStyle-Base_C
    text=Attenuation Curves
    font {
     bForceMonospaced=false
     bMaterialIsStencil=false
     fontMaterial=null
     fontObject=&NotoSans
     letterSpacing=0
     monospacedWidth=1
     size=24
     skewAmount=0
     typefaceFontName=NotoSans-Regular
     outlineSettings {
      bApplyOutlineToDropShadows=false
      bMiteredCorners=false
      bSeparateFillAlpha=false
      outlineMaterial=null
      outlineSize=0
      outlineColor=<rgba 0 0 0 1>
     }
    }
    shadowOffset=<0 0>
   }
   slot {class=CanvasPanelSlot, properties {layoutData {alignment=<0 0>, anchors {maximum=<0 0>, minimum=<0 0>}, offsets=<ltrb 0 -8 291.1381 44>}}}
  }
  Border_2 {
   widgetClassPath=&Border
   parent=CanvasPanel_0
   widget=Border_2
   properties {horizontalAlignment=HAlign_Left, verticalAlignment=VAlign_Top, brushColor=<rgba 0 0 0 0.1412>, padding=<ltrb 10 10 10 10>}
   slot {
    class=CanvasPanelSlot
    properties {bAutoSize=true, layoutData {alignment=<0 0>, anchors {maximum=<0 0>, minimum=<0 0>}, offsets=<ltrb 0 548 640.941 491.7618>}}
   }
  }
  CanvasPanel_1 {
   widgetClassPath=&CanvasPanel
   parent=Border_2
   widget=CanvasPanel_1
   slot {class=BorderSlot, properties {horizontalAlignment=HAlign_Left, verticalAlignment=VAlign_Top, padding=<ltrb 10 10 10 10>}}
  }
  Curve01_Property {
   widgetClassPath=&SinglePropertyView
   parent=CanvasPanel_1
   widget=Curve01_Property
   properties {propertyName=Curve_01}
   slot {
    class=CanvasPanelSlot
    properties {bAutoSize=true, layoutData {alignment=<0 0>, anchors {maximum=<0 0>, minimum=<0 0>}, offsets=<ltrb 0 44 292.1922 44.5045>}}
   }
  }
  Curve02_Property {
   widgetClassPath=&SinglePropertyView
   parent=CanvasPanel_1
   widget=Curve02_Property
   properties {propertyName=Curve_02}
   slot {
    class=CanvasPanelSlot
    properties {bAutoSize=true, layoutData {alignment=<0 0>, anchors {maximum=<0 0>, minimum=<0 0>}, offsets=<ltrb 0 100 292.1922 44.5045>}}
   }
  }
  Curve03_Property {
   widgetClassPath=&SinglePropertyView
   parent=CanvasPanel_1
   widget=Curve03_Property
   properties {propertyName=Curve_03}
   slot {
    class=CanvasPanelSlot
    properties {bAutoSize=true, layoutData {alignment=<0 0>, anchors {maximum=<0 0>, minimum=<0 0>}, offsets=<ltrb 0 156 292.1922 44.5045>}}
   }
  }
  ImportGenericCurves_Button {
   widgetClassPath=&Button
   parent=CanvasPanel_1
   widget=ImportGenericCurves_Button
   properties {
    bIsEnabled=false
    widgetStyle {
     clickedSlateSound {resourceObject=null}
     disabled {
      bIsDynamicallyLoaded=false
      drawAs=RoundedBox
      imageType=NoImage
      mirroring=NoMirror
      resourceName=None
      resourceObject=null
      tiling=NoTile
      imageSize=<0 0>
      margin=<ltrb 0 0 0 0>
      outlineSettings {
       bUseBrushTransparency=false
       roundingType=FixedRadius
       width=1
       color {colorUseRule=UseColor_ColorTable, specifiedColor=<rgba 0 0 0 0>}
       cornerRadii=<4 4 4 4>
      }
      tintColor {colorUseRule=UseColor_Specified, specifiedColor=<rgba 0.0395 0.0395 0.0395 1>}
      uVRegion {bIsValid=false, max=<0 0>, min=<0 0>}
     }
     disabledForeground {colorUseRule=UseColor_ColorTable, specifiedColor=<rgba 0 0 0 0>}
     hovered {
      bIsDynamicallyLoaded=false
      drawAs=RoundedBox
      imageType=NoImage
      mirroring=NoMirror
      resourceName=None
      resourceObject=null
      tiling=NoTile
      imageSize=<0 0>
      margin=<ltrb 0 0 0 0>
      outlineSettings {
       bUseBrushTransparency=false
       roundingType=FixedRadius
       width=1
       color {colorUseRule=UseColor_ColorTable, specifiedColor=<rgba 0 0 0 0>}
       cornerRadii=<4 4 4 4>
      }
      tintColor {colorUseRule=UseColor_Specified, specifiedColor=<rgba 0.0953 0.0953 0.0953 1>}
      uVRegion {bIsValid=false, max=<0 0>, min=<0 0>}
     }
     hoveredForeground {colorUseRule=UseColor_ColorTable, specifiedColor=<rgba 0 0 0 0>}
     hoveredSlateSound {resourceObject=null}
     normal {
      bIsDynamicallyLoaded=false
      drawAs=RoundedBox
      imageType=NoImage
      mirroring=NoMirror
      resourceName=None
      resourceObject=null
      tiling=NoTile
      imageSize=<0 0>
      margin=<ltrb 0 0 0 0>
      outlineSettings {
       bUseBrushTransparency=false
       roundingType=FixedRadius
       width=1
       color {colorUseRule=UseColor_ColorTable, specifiedColor=<rgba 0 0 0 0>}
       cornerRadii=<4 4 4 4>
      }
      tintColor {colorUseRule=UseColor_Specified, specifiedColor=<rgba 0.0395 0.0395 0.0395 1>}
      uVRegion {bIsValid=false, max=<0 0>, min=<0 0>}
     }
     normalForeground {colorUseRule=UseColor_ColorTable, specifiedColor=<rgba 0 0 0 0>}
     normalPadding=<ltrb 12 1.5 12 1.5>
     pressed {
      bIsDynamicallyLoaded=false
      drawAs=RoundedBox
      imageType=NoImage
      mirroring=NoMirror
      resourceName=None
      resourceObject=null
      tiling=NoTile
      imageSize=<0 0>
      margin=<ltrb 0 0 0 0>
      outlineSettings {
       bUseBrushTransparency=false
       roundingType=FixedRadius
       width=1
       color {colorUseRule=UseColor_ColorTable, specifiedColor=<rgba 0 0 0 0>}
       cornerRadii=<4 4 4 4>
      }
      tintColor {colorUseRule=UseColor_Specified, specifiedColor=<rgba 0.0284 0.0284 0.0284 1>}
      uVRegion {bIsValid=false, max=<0 0>, min=<0 0>}
     }
     pressedForeground {colorUseRule=UseColor_ColorTable, specifiedColor=<rgba 0 0 0 0>}
     pressedPadding=<ltrb 12 2 12 0>
     pressedSlateSound {resourceObject=null}
    }
   }
   slot {class=CanvasPanelSlot, properties {layoutData {alignment=<0 0>, anchors {maximum=<0 0>, minimum=<0 0>}, offsets=<ltrb 420 0 100 40>}}}
  }
  VectorImport_1 {
   widgetClassPath=&CommonTextBlock
   parent=ImportGenericCurves_Button
   widget=VectorImport_1
   properties {
    justification=Center
    style=&TextStyle-Base-M-B_Black_C
    text=Import Curves
    colorAndOpacity {colorUseRule=UseColor_Specified, specifiedColor=<rgba 0 0 0 1>}
    font {
     bForceMonospaced=false
     bMaterialIsStencil=false
     fontMaterial=null
     fontObject=&NotoSans
     letterSpacing=0
     monospacedWidth=1
     size=24
     skewAmount=0
     typefaceFontName=NotoSans-Bold
     outlineSettings {
      bApplyOutlineToDropShadows=false
      bMiteredCorners=false
      bSeparateFillAlpha=false
      outlineMaterial=null
      outlineSize=0
      outlineColor=<rgba 0 0 0 1>
     }
    }
    renderTransform {angle=0, scale=<0.4 0.7>, shear=<0 0>, translation=<0 -5>}
    shadowOffset=<0 0>
   }
   slot {class=ButtonSlot}
  }
  ExportModifiedGeneric_Button {
   widgetClassPath=&Button
   parent=CanvasPanel_1
   widget=ExportModifiedGeneric_Button
   properties {
    bIsEnabled=false
    widgetStyle {
     clickedSlateSound {resourceObject=null}
     disabled {
      bIsDynamicallyLoaded=false
      drawAs=RoundedBox
      imageType=NoImage
      mirroring=NoMirror
      resourceName=None
      resourceObject=null
      tiling=NoTile
      imageSize=<0 0>
      margin=<ltrb 0 0 0 0>
      outlineSettings {
       bUseBrushTransparency=false
       roundingType=FixedRadius
       width=1
       color {colorUseRule=UseColor_ColorTable, specifiedColor=<rgba 0 0 0 0>}
       cornerRadii=<4 4 4 4>
      }
      tintColor {colorUseRule=UseColor_Specified, specifiedColor=<rgba 0.0395 0.0395 0.0395 1>}
      uVRegion {bIsValid=false, max=<0 0>, min=<0 0>}
     }
     disabledForeground {colorUseRule=UseColor_ColorTable, specifiedColor=<rgba 0 0 0 0>}
     hovered {
      bIsDynamicallyLoaded=false
      drawAs=RoundedBox
      imageType=NoImage
      mirroring=NoMirror
      resourceName=None
      resourceObject=null
      tiling=NoTile
      imageSize=<0 0>
      margin=<ltrb 0 0 0 0>
      outlineSettings {
       bUseBrushTransparency=false
       roundingType=FixedRadius
       width=1
       color {colorUseRule=UseColor_ColorTable, specifiedColor=<rgba 0 0 0 0>}
       cornerRadii=<4 4 4 4>
      }
      tintColor {colorUseRule=UseColor_Specified, specifiedColor=<rgba 0.0953 0.0953 0.0953 1>}
      uVRegion {bIsValid=false, max=<0 0>, min=<0 0>}
     }
     hoveredForeground {colorUseRule=UseColor_ColorTable, specifiedColor=<rgba 0 0 0 0>}
     hoveredSlateSound {resourceObject=null}
     normal {
      bIsDynamicallyLoaded=false
      drawAs=RoundedBox
      imageType=NoImage
      mirroring=NoMirror
      resourceName=None
      resourceObject=null
      tiling=NoTile
      imageSize=<0 0>
      margin=<ltrb 0 0 0 0>
      outlineSettings {
       bUseBrushTransparency=false
       roundingType=FixedRadius
       width=1
       color {colorUseRule=UseColor_ColorTable, specifiedColor=<rgba 0 0 0 0>}
       cornerRadii=<4 4 4 4>
      }
      tintColor {colorUseRule=UseColor_Specified, specifiedColor=<rgba 0.0395 0.0395 0.0395 1>}
      uVRegion {bIsValid=false, max=<0 0>, min=<0 0>}
     }
     normalForeground {colorUseRule=UseColor_ColorTable, specifiedColor=<rgba 0 0 0 0>}
     normalPadding=<ltrb 12 1.5 12 1.5>
     pressed {
      bIsDynamicallyLoaded=false
      drawAs=RoundedBox
      imageType=NoImage
      mirroring=NoMirror
      resourceName=None
      resourceObject=null
      tiling=NoTile
      imageSize=<0 0>
      margin=<ltrb 0 0 0 0>
      outlineSettings {
       bUseBrushTransparency=false
       roundingType=FixedRadius
       width=1
       color {colorUseRule=UseColor_ColorTable, specifiedColor=<rgba 0 0 0 0>}
       cornerRadii=<4 4 4 4>
      }
      tintColor {colorUseRule=UseColor_Specified, specifiedColor=<rgba 0.0284 0.0284 0.0284 1>}
      uVRegion {bIsValid=false, max=<0 0>, min=<0 0>}
     }
     pressedForeground {colorUseRule=UseColor_ColorTable, specifiedColor=<rgba 0 0 0 0>}
     pressedPadding=<ltrb 12 2 12 0>
     pressedSlateSound {resourceObject=null}
    }
   }
   slot {class=CanvasPanelSlot, properties {layoutData {alignment=<0 0>, anchors {maximum=<0 0>, minimum=<0 0>}, offsets=<ltrb 4 220 100 40>}}}
  }
  ExportModifiedGeneric_Text {
   widgetClassPath=&CommonTextBlock
   bIsVariable=true
   parent=ExportModifiedGeneric_Button
   widget=ExportModifiedGeneric_Text
   properties {
    justification=Center
    style=&TextStyle-Base-M-B_Black_C
    text="Export\r\nModified"
    colorAndOpacity {colorUseRule=UseColor_Specified, specifiedColor=<rgba 0 0 0 1>}
    font {
     bForceMonospaced=false
     bMaterialIsStencil=false
     fontMaterial=null
     fontObject=&NotoSans
     letterSpacing=0
     monospacedWidth=1
     size=24
     skewAmount=0
     typefaceFontName=NotoSans-Bold
     outlineSettings {
      bApplyOutlineToDropShadows=false
      bMiteredCorners=false
      bSeparateFillAlpha=false
      outlineMaterial=null
      outlineSize=0
      outlineColor=<rgba 0 0 0 1>
     }
    }
    renderTransform {angle=0, scale=<0.45 0.4>, shear=<0 0>, translation=<0 -11>}
    shadowOffset=<0 0>
   }
   slot {class=ButtonSlot}
  }
  ExportOriginalGeneric_Button {
   widgetClassPath=&Button
   parent=CanvasPanel_1
   widget=ExportOriginalGeneric_Button
   properties {
    bIsEnabled=false
    widgetStyle {
     clickedSlateSound {resourceObject=null}
     disabled {
      bIsDynamicallyLoaded=false
      drawAs=RoundedBox
      imageType=NoImage
      mirroring=NoMirror
      resourceName=None
      resourceObject=null
      tiling=NoTile
      imageSize=<0 0>
      margin=<ltrb 0 0 0 0>
      outlineSettings {
       bUseBrushTransparency=false
       roundingType=FixedRadius
       width=1
       color {colorUseRule=UseColor_ColorTable, specifiedColor=<rgba 0 0 0 0>}
       cornerRadii=<4 4 4 4>
      }
      tintColor {colorUseRule=UseColor_Specified, specifiedColor=<rgba 0.0395 0.0395 0.0395 1>}
      uVRegion {bIsValid=false, max=<0 0>, min=<0 0>}
     }
     disabledForeground {colorUseRule=UseColor_ColorTable, specifiedColor=<rgba 0 0 0 0>}
     hovered {
      bIsDynamicallyLoaded=false
      drawAs=RoundedBox
      imageType=NoImage
      mirroring=NoMirror
      resourceName=None
      resourceObject=null
      tiling=NoTile
      imageSize=<0 0>
      margin=<ltrb 0 0 0 0>
      outlineSettings {
       bUseBrushTransparency=false
       roundingType=FixedRadius
       width=1
       color {colorUseRule=UseColor_ColorTable, specifiedColor=<rgba 0 0 0 0>}
       cornerRadii=<4 4 4 4>
      }
      tintColor {colorUseRule=UseColor_Specified, specifiedColor=<rgba 0.0953 0.0953 0.0953 1>}
      uVRegion {bIsValid=false, max=<0 0>, min=<0 0>}
     }
     hoveredForeground {colorUseRule=UseColor_ColorTable, specifiedColor=<rgba 0 0 0 0>}
     hoveredSlateSound {resourceObject=null}
     normal {
      bIsDynamicallyLoaded=false
      drawAs=RoundedBox
      imageType=NoImage
      mirroring=NoMirror
      resourceName=None
      resourceObject=null
      tiling=NoTile
      imageSize=<0 0>
      margin=<ltrb 0 0 0 0>
      outlineSettings {
       bUseBrushTransparency=false
       roundingType=FixedRadius
       width=1
       color {colorUseRule=UseColor_ColorTable, specifiedColor=<rgba 0 0 0 0>}
       cornerRadii=<4 4 4 4>
      }
      tintColor {colorUseRule=UseColor_Specified, specifiedColor=<rgba 0.0395 0.0395 0.0395 1>}
      uVRegion {bIsValid=false, max=<0 0>, min=<0 0>}
     }
     normalForeground {colorUseRule=UseColor_ColorTable, specifiedColor=<rgba 0 0 0 0>}
     normalPadding=<ltrb 12 1.5 12 1.5>
     pressed {
      bIsDynamicallyLoaded=false
      drawAs=RoundedBox
      imageType=NoImage
      mirroring=NoMirror
      resourceName=None
      resourceObject=null
      tiling=NoTile
      imageSize=<0 0>
      margin=<ltrb 0 0 0 0>
      outlineSettings {
       bUseBrushTransparency=false
       roundingType=FixedRadius
       width=1
       color {colorUseRule=UseColor_ColorTable, specifiedColor=<rgba 0 0 0 0>}
       cornerRadii=<4 4 4 4>
      }
      tintColor {colorUseRule=UseColor_Specified, specifiedColor=<rgba 0.0284 0.0284 0.0284 1>}
      uVRegion {bIsValid=false, max=<0 0>, min=<0 0>}
     }
     pressedForeground {colorUseRule=UseColor_ColorTable, specifiedColor=<rgba 0 0 0 0>}
     pressedPadding=<ltrb 12 2 12 0>
     pressedSlateSound {resourceObject=null}
    }
   }
   slot {class=CanvasPanelSlot, properties {layoutData {alignment=<0 0>, anchors {maximum=<0 0>, minimum=<0 0>}, offsets=<ltrb 108 220 100 40>}}}
  }
  ExportOriginalsGeneric_Text {
   widgetClassPath=&CommonTextBlock
   bIsVariable=true
   parent=ExportOriginalGeneric_Button
   widget=ExportOriginalsGeneric_Text
   properties {
    justification=Center
    style=&TextStyle-Base-M-B_Black_C
    text="Export\r\nOriginals"
    colorAndOpacity {colorUseRule=UseColor_Specified, specifiedColor=<rgba 0 0 0 1>}
    font {
     bForceMonospaced=false
     bMaterialIsStencil=false
     fontMaterial=null
     fontObject=&NotoSans
     letterSpacing=0
     monospacedWidth=1
     size=24
     skewAmount=0
     typefaceFontName=NotoSans-Bold
     outlineSettings {
      bApplyOutlineToDropShadows=false
      bMiteredCorners=false
      bSeparateFillAlpha=false
      outlineMaterial=null
      outlineSize=0
      outlineColor=<rgba 0 0 0 1>
     }
    }
    renderTransform {angle=0, scale=<0.45 0.4>, shear=<0 0>, translation=<0 -11>}
    shadowOffset=<0 0>
   }
   slot {class=ButtonSlot}
  }
  CommonTextBlock_3 {
   widgetClassPath=&CommonTextBlock
   parent=CanvasPanel_1
   widget=CommonTextBlock_3
   properties {
    style=&TextStyle-Base_C
    text=Generic Curves
    font {
     bForceMonospaced=false
     bMaterialIsStencil=false
     fontMaterial=null
     fontObject=&NotoSans
     letterSpacing=0
     monospacedWidth=1
     size=24
     skewAmount=0
     typefaceFontName=NotoSans-Regular
     outlineSettings {
      bApplyOutlineToDropShadows=false
      bMiteredCorners=false
      bSeparateFillAlpha=false
      outlineMaterial=null
      outlineSize=0
      outlineColor=<rgba 0 0 0 1>
     }
    }
    shadowOffset=<0 0>
   }
   slot {class=CanvasPanelSlot, properties {layoutData {alignment=<0 0>, anchors {maximum=<0 0>, minimum=<0 0>}, offsets=<ltrb 0 -8 291.1381 44>}}}
  }
 ]
}
