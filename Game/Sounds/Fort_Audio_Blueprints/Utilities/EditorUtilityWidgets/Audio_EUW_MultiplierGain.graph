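# Editor utility widget with a dB gain knob that sets the volume multiplier on SoundCue and SoundWave assets.
# Knob values pass through a linear-to-log curve, then write to SingleAffectedSound, or all selected assets if it is unset, without change notifications.
# At unity or above, a text box accepts positive dB, converted by 10^(dB/20).
# Tick polls selection, refreshing knob and readout through the inverse curve when the first asset changes.
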EditorUtilityWidgetBlueprint /Game/Sounds/Fort_Audio_Blueprints/Utilities/EditorUtilityWidgets/Audio_EUW_MultiplierGain
graphs [
 EventGraph {nodes [
   n1 = "On Value Changed (SynthKnob_0)"!() -> (OutputDelegate, then, Value:float) {is=title K2Node_ComponentBoundEvent}
   > n2 = Set AuditionAudioMultiplier(AuditionAudioMultiplier:double=@n1.Value) -> (Output_Get:double)
   > SetEnableEditableMultiplier!(execute, self, Enable:bool="float >= float"(A=@n2, B=1) {is=title K2Node_CallFunction}) -> (then)
   > n3 = Set ModifiedAudioMultiplier(ModifiedAudioMultiplier:double=GetFloatValue!(
     self:Curve Float Object Reference=Get AudioCurve_LinToLog_MultiplierKnob() -> (AudioCurve_LinToLog_MultiplierKnob:Curve Float Object Reference)
     InTime:float=Get AuditionAudioMultiplier() -> (AuditionAudioMultiplier:double)
    ) -> (ReturnValue:float)) -> (Output_Get:double)
   > n4 = MultiplierToDecibelsText!(execute, self, InputPin:double=@n3) -> (then, ReturnValue:Text)
   > n5 = "SetText (Text)"(self=Get MultiplierValue_TextBox() -> (MultiplierValue_TextBox:Common Text Object Reference), InText=@n4)
   n6 = Get!(execute:Exec=[@n5.then, @n13.then]) -> (then:Exec, else:Exec, SingleAffectedSound:Sound Base Object Reference)
   > SetSoundBaseVolume!(
    execute
    self
    Object:Object Reference=Get SingleAffectedSound() -> (SingleAffectedSound:Sound Base Object Reference)
    Multiplier:double=Get ModifiedAudioMultiplier() -> (ModifiedAudioMultiplier:double)
   ) -> (then)
   n7 = GetSelectedAssets(execute=@n6.else)
   > n8 = For Each Loop(Array:Array of Object References=@n7) -> (Array Element:Object Reference)
   > SetSoundBaseVolume(Object=@n8, Multiplier=Get ModifiedAudioMultiplier())
   n9 = "On Text Committed (MultiplierInput_TextBox)"!() -> (OutputDelegate, then, Text:"Text (by ref)", CommitMethod:ETextCommit Enum=Default) {is=title K2Node_ComponentBoundEvent}
   > SetPositiveMultiplierValue!(execute, self, Value:double=String To Float(InString="To String (Text)"(InText=@n9.Text))) -> (then)
   n10 = SetPositiveMultiplierValue!() -> (OutputDelegate, then, Value:double=~) {is=title K2Node_CustomEvent}
   > Branch(Condition="float >= float"(A=@n10.Value, B=0))
   > Set ModifiedAudioMultiplier(ModifiedAudioMultiplier=Power!(Base:double=10, Exp:double=float / float(A=@n10.Value, B=20) {is=title K2Node_CallFunction}) -> (ReturnValue:double) {is=title K2Node_CallFunction})
   n11 = "To Text (Float)"(Value=@n10.Value, MinimumIntegralDigits=1, MaximumIntegralDigits=2, MinimumFractionalDigits=2, MaximumFractionalDigits=2)
   n12 = "To Text (String)"(InString=Append!(A=+, B="To String (Text)"(InText=@n11), C:String=" dB") -> (ReturnValue))
   > "SetText (Text)"(self=Get MultiplierValue_TextBox(), InText=@n12)
   > n13 = "SetText (Text Box)"!(
    execute
    self:Text Box Object Reference=Get MultiplierInput_TextBox() -> (MultiplierInput_TextBox:Text Box Object Reference)
    InText:Text=@n12
   ) -> (then)
   Event Tick()
   > n14 = Get()
   n15 = GetSelectedAssets(execute=@n14.else)
   > Branch(Condition=Is Valid Index!(TargetArray:Array of Object References=@n15, IndexToTest:int) -> (ReturnValue:bool))
   n16 = "Get (a copy)"(Array:Array of Object References=@n15) -> (Output:Object Reference)
   > Branch(Condition="Not Equal (Object)"!(A:Object Reference=Get LastSelectedAsset() -> (LastSelectedAsset:Object Reference), B:Object Reference=@n16) -> (ReturnValue:bool))
   > Set LastSelectedAsset(LastSelectedAsset:Object Reference=@n16) -> (Output_Get:Object Reference)
   > n17 = Cast To SoundCue!(execute, Object=Get LastSelectedAsset()) -> (then, CastFailed, AsSound Cue:Sound Cue Object Reference) {is=title K2Node_DynamicCast}
   > n18 = GetEditorProperty!(execute:Exec, Object:Object Reference=@n17, PropertyName:Name=VolumeMultiplier) -> (then:Exec, PropertyValue:double, ReturnValue:bool)
   > n19 = Set AuditionAudioMultiplier(AuditionAudioMultiplier=@n18)
   SetEnableEditableMultiplier(execute=[@n23.then, @n19.then], Enable="float >= float"(A=Get AuditionAudioMultiplier(), B=1))
   > Set ModifiedAudioMultiplier(ModifiedAudioMultiplier=GetFloatValue(
     self=Get AudioCurve_LogToLin_MultiplierKnob() -> (AudioCurve_LogToLin_MultiplierKnob:Curve Float Object Reference)
     InTime=Get AuditionAudioMultiplier()
    ))
   > SetValue!(
    execute
    self:Synth Knob Object Reference=Get SynthKnob_0() -> (SynthKnob_0:Synth Knob Object Reference)
    InValue:float=Get ModifiedAudioMultiplier()
   ) -> (then)
   > n20 = MultiplierToDecibelsText(InputPin=Get AuditionAudioMultiplier())
   > "SetText (Text)"(self=Get MultiplierValue_TextBox(), InText=@n20)
   > Branch(Condition="float >= float"(A=Get AuditionAudioMultiplier(), B=1))
   > "SetText (Text Box)"(self=Get MultiplierInput_TextBox(), InText=@n20)
   n21 = Cast To SoundWave!(execute=@n17.CastFailed, Object=Get LastSelectedAsset()) -> (then, CastFailed, AsSound Wave:Sound Wave Object Reference) {is=title K2Node_DynamicCast}
   > n22 = GetEditorProperty(Object=@n21, PropertyName=Volume)
   > n23 = Set AuditionAudioMultiplier(AuditionAudioMultiplier=@n22)
  ]}
 MultiplierToDecibelsText {nodes [
   n1 = MultiplierToDecibelsText!() -> (then, InputPin:double=~) {is=title K2Node_FunctionEntry}
   n2 = SelectString(A=+, bPickA="float >= float"(A=@n1, B=1))
   n3 = float * float(A=Log!(A:double=@n1, Base:double=10) -> (ReturnValue:double), B=20)
   n4 = "float <= float"!(A:double=@n1, B:double=0) -> (ReturnValue:bool) {is=title K2Node_CallFunction}
   n5 = "float <= float"(A=@n3, B=-60)
   n6 = "To String (Text)"(InText=Select(
     Option 0:Text="To Text (Float)"(Value=@n3, MinimumIntegralDigits=1, MaximumIntegralDigits=2, MinimumFractionalDigits=2, MaximumFractionalDigits=2)
     Option 1:Text="LOCTEXT(\"∞\")"
     Index=OR Boolean(A=@n4, B=@n5)
    ) -> (ReturnValue:Text))
   > Return Node!(execute, ReturnValue:Text="To Text (String)"(InString=Append(A=@n2, B=@n6, C=" dB")))
  ]}
 SetEnableEditableMultiplier {nodes [
   n1 = SetEnableEditableMultiplier!() -> (then, Enable:bool=~) {is=title K2Node_FunctionEntry}
   > SetIsEnabled(self=Get MultiplierInput_TextBox(), bInIsEnabled=@n1)
   > "SetText (Text Box)"(self=Get MultiplierInput_TextBox(), InText="LOCTEXT(\"+0.00 dB\")")
  ]}
 SetSoundBaseVolume {nodes [
   n1 = SetSoundBaseVolume!() -> (then, Object:Object Reference, Multiplier:double=~) {is=title K2Node_FunctionEntry}
   > n2 = Cast To SoundCue(Object=@n1)
   > SetEditorProperty(Object=@n2, PropertyName=VolumeMultiplier, PropertyValue:double=@n1.Multiplier, ChangeNotifyMode=Never)
   n3 = Cast To SoundWave(execute=@n2.CastFailed, Object=@n1)
   > SetEditorProperty(Object=@n3, PropertyName=Volume, PropertyValue=@n1.Multiplier, ChangeNotifyMode=Never)
  ]}
]
variables [
 AudioCurve_LinToLog_MultiplierKnob {type=CurveFloat, default=/Game/Sounds/Fort_Audio_Blueprints/Utilities/EditorUtilityWidgets/Curves/LinearToLogAudioMultiplierCurve}
 AudioCurve_LogToLin_MultiplierKnob {type=CurveFloat, default=/Game/Sounds/Fort_Audio_Blueprints/Utilities/EditorUtilityWidgets/Curves/LogToLinearAudioMultiplierCurve}
 AuditionAudioMultiplier {type=number}
 LastSelectedAsset {type=Object, default=null}
 ModifiedAudioMultiplier {type=number}
 SingleAffectedSound {type=SoundBase, default=null}
]
widgets {
 info {rootWidgetClass=&CanvasPanel, widgetCount=23}
 widgets [
  CanvasPanel_0 {widgetClassPath=&CanvasPanel, widget=CanvasPanel_0}
  CanvasPanel_154 {
   widgetClassPath=&CanvasPanel
   parent=CanvasPanel_0
   widget=CanvasPanel_154
   slot {
    class=CanvasPanelSlot
    properties {layoutData {alignment=<0 0>, anchors {maximum=<0 0>, minimum=<0 0>}, offsets=<ltrb 0 0 97.0547 117.0871>}}
   }
  }
  Knob_Canvas {
   widgetClassPath=&CanvasPanel
   parent=CanvasPanel_154
   widget=Knob_Canvas
   slot {class=CanvasPanelSlot, properties {layoutData {alignment=<0 0>, anchors {maximum=<0 0>, minimum=<0 0>}, offsets=<ltrb 0 0 96.997 72.8887>}}}
  }
  CommonTextBlock_179 {
   widgetClassPath=&CommonTextBlock
   parent=Knob_Canvas
   widget=CommonTextBlock_179
   properties {
    style=&TextStyle-Base_C
    text=-
    font {
     bForceMonospaced=false
     bMaterialIsStencil=false
     fontMaterial=null
     fontObject=&NotoSans
     letterSpacing=0
     monospacedWidth=1
     size=24
     skewAmount=0
     typefaceFontName=NotoSans-Regular
     outlineSettings {
      bApplyOutlineToDropShadows=false
      bMiteredCorners=false
      bSeparateFillAlpha=false
      outlineMaterial=null
      outlineSize=0
      outlineColor=<rgba 0 0 0 1>
     }
    }
    renderTransform {angle=0, scale=<0.7 0.7>, shear=<0 0>, translation=<-26 -30>}
    shadowOffset=<0 0>
   }
   slot {
    class=CanvasPanelSlot
    properties {layoutData {alignment=<0 0>, anchors {maximum=<0 0>, minimum=<0 0>}, offsets=<ltrb 89.202 50.1994 10.9204 44>}}
   }
  }
  CommonTextBlock_17 {
   widgetClassPath=&CommonTextBlock
   parent=Knob_Canvas
   widget=CommonTextBlock_17
   properties {
    style=&TextStyle-Base_C
    text=-
    font {
     bForceMonospaced=false
     bMaterialIsStencil=false
     fontMaterial=null
     fontObject=&NotoSans
     letterSpacing=0
     monospacedWidth=1
     size=24
     skewAmount=0
     typefaceFontName=NotoSans-Regular
     outlineSettings {
      bApplyOutlineToDropShadows=false
      bMiteredCorners=false
      bSeparateFillAlpha=false
      outlineMaterial=null
      outlineSize=0
      outlineColor=<rgba 0 0 0 1>
     }
    }
    renderTransform {angle=30, scale=<0.5 0.5>, shear=<0 0>, translation=<-26 -30>}
    shadowOffset=<0 0>
   }
   slot {
    class=CanvasPanelSlot
    properties {layoutData {alignment=<0 0>, anchors {maximum=<0 0>, minimum=<0 0>}, offsets=<ltrb 87.5414 60.0604 10.9204 44>}}
   }
  }
  CommonTextBlock_23 {
   widgetClassPath=&CommonTextBlock
   parent=Knob_Canvas
   widget=CommonTextBlock_23
   properties {
    style=&TextStyle-Base_C
    text=-
    font {
     bForceMonospaced=false
     bMaterialIsStencil=false
     fontMaterial=null
     fontObject=&NotoSans
     letterSpacing=0
     monospacedWidth=1
     size=24
     skewAmount=0
     typefaceFontName=NotoSans-Regular
     outlineSettings {
      bApplyOutlineToDropShadows=false
      bMiteredCorners=false
      bSeparateFillAlpha=false
      outlineMaterial=null
      outlineSize=0
      outlineColor=<rgba 0 0 0 1>
     }
    }
    renderTransform {angle=145, scale=<0.5 0.5>, shear=<0 0>, translation=<-26 -30>}
    shadowOffset=<0 0>
   }
   slot {
    class=CanvasPanelSlot
    properties {layoutData {alignment=<0 0>, anchors {maximum=<0 0>, minimum=<0 0>}, offsets=<ltrb 53.7712 63.1459 10.9204 44>}}
   }
  }
  CommonTextBlock_20 {
   widgetClassPath=&CommonTextBlock
   parent=Knob_Canvas
   widget=CommonTextBlock_20
   properties {
    style=&TextStyle-Base_C
    text=-
    font {
     bForceMonospaced=false
     bMaterialIsStencil=false
     fontMaterial=null
     fontObject=&NotoSans
     letterSpacing=0
     monospacedWidth=1
     size=24
     skewAmount=0
     typefaceFontName=NotoSans-Regular
     outlineSettings {
      bApplyOutlineToDropShadows=false
      bMiteredCorners=false
      bSeparateFillAlpha=false
      outlineMaterial=null
      outlineSize=0
      outlineColor=<rgba 0 0 0 1>
     }
    }
    renderTransform {angle=-30, scale=<0.5 0.5>, shear=<0 0>, translation=<-26 -30>}
    shadowOffset=<0 0>
   }
   slot {
    class=CanvasPanelSlot
    properties {layoutData {alignment=<0 0>, anchors {maximum=<0 0>, minimum=<0 0>}, offsets=<ltrb 84.8266 41.8028 10.9204 44>}}
   }
  }
  CommonTextBlock_21 {
   widgetClassPath=&CommonTextBlock
   parent=Knob_Canvas
   widget=CommonTextBlock_21
   properties {
    style=&TextStyle-Base_C
    text=-
    font {
     bForceMonospaced=false
     bMaterialIsStencil=false
     fontMaterial=null
     fontObject=&NotoSans
     letterSpacing=0
     monospacedWidth=1
     size=24
     skewAmount=0
     typefaceFontName=NotoSans-Regular
     outlineSettings {
      bApplyOutlineToDropShadows=false
      bMiteredCorners=false
      bSeparateFillAlpha=false
      outlineMaterial=null
      outlineSize=0
      outlineColor=<rgba 0 0 0 1>
     }
    }
    renderTransform {angle=-90, scale=<0.5 0.5>, shear=<0 0>, translation=<-26 -30>}
    shadowOffset=<0 0>
   }
   slot {
    class=CanvasPanelSlot
    properties {layoutData {alignment=<0 0>, anchors {maximum=<0 0>, minimum=<0 0>}, offsets=<ltrb 67.6394 34.1935 10.9204 44>}}
   }
  }
  CommonTextBlock_22 {
   widgetClassPath=&CommonTextBlock
   parent=Knob_Canvas
   widget=CommonTextBlock_22
   properties {
    style=&TextStyle-Base_C
    text=-
    font {
     bForceMonospaced=false
     bMaterialIsStencil=false
     fontMaterial=null
     fontObject=&NotoSans
     letterSpacing=0
     monospacedWidth=1
     size=24
     skewAmount=0
     typefaceFontName=NotoSans-Regular
     outlineSettings {
      bApplyOutlineToDropShadows=false
      bMiteredCorners=false
      bSeparateFillAlpha=false
      outlineMaterial=null
      outlineSize=0
      outlineColor=<rgba 0 0 0 1>
     }
    }
    renderTransform {angle=-145, scale=<0.5 0.5>, shear=<0 0>, translation=<-26 -30>}
    shadowOffset=<0 0>
   }
   slot {
    class=CanvasPanelSlot
    properties {layoutData {alignment=<0 0>, anchors {maximum=<0 0>, minimum=<0 0>}, offsets=<ltrb 53.9382 41.4513 10.0075 48.7355>}}
   }
  }
  CommonTextBlock_18 {
   widgetClassPath=&CommonTextBlock
   parent=Knob_Canvas
   widget=CommonTextBlock_18
   properties {
    style=&TextStyle-Base_C
    text=-
    font {
     bForceMonospaced=false
     bMaterialIsStencil=false
     fontMaterial=null
     fontObject=&NotoSans
     letterSpacing=0
     monospacedWidth=1
     size=24
     skewAmount=0
     typefaceFontName=NotoSans-Regular
     outlineSettings {
      bApplyOutlineToDropShadows=false
      bMiteredCorners=false
      bSeparateFillAlpha=false
      outlineMaterial=null
      outlineSize=0
      outlineColor=<rgba 0 0 0 1>
     }
    }
    renderTransform {angle=50, scale=<0.7 0.7>, shear=<0 0>, translation=<-26 -30>}
    shadowOffset=<0 0>
   }
   slot {
    class=CanvasPanelSlot
    properties {layoutData {alignment=<0 0>, anchors {maximum=<0 0>, minimum=<0 0>}, offsets=<ltrb 83.7736 66.9294 10.9204 44>}}
   }
  }
  CommonTextBlock_19 {
   widgetClassPath=&CommonTextBlock
   parent=Knob_Canvas
   widget=CommonTextBlock_19
   properties {
    style=&TextStyle-Base_C
    text=-
    font {
     bForceMonospaced=false
     bMaterialIsStencil=false
     fontMaterial=null
     fontObject=&NotoSans
     letterSpacing=0
     monospacedWidth=1
     size=24
     skewAmount=0
     typefaceFontName=NotoSans-Regular
     outlineSettings {
      bApplyOutlineToDropShadows=false
      bMiteredCorners=false
      bSeparateFillAlpha=false
      outlineMaterial=null
      outlineSize=0
      outlineColor=<rgba 0 0 0 1>
     }
    }
    renderTransform {angle=125, scale=<0.7 0.7>, shear=<0 0>, translation=<-26 -30>}
    shadowOffset=<0 0>
   }
   slot {
    class=CanvasPanelSlot
    properties {layoutData {alignment=<0 0>, anchors {maximum=<0 0>, minimum=<0 0>}, offsets=<ltrb 59.7994 70.0662 10.9204 44>}}
   }
  }
  CommonTextBlock_15 {
   widgetClassPath=&CommonTextBlock
   parent=Knob_Canvas
   widget=CommonTextBlock_15
   properties {
    style=&TextStyle-Base_C
    text=-
    font {
     bForceMonospaced=false
     bMaterialIsStencil=false
     fontMaterial=null
     fontObject=&NotoSans
     letterSpacing=0
     monospacedWidth=1
     size=24
     skewAmount=0
     typefaceFontName=NotoSans-Regular
     outlineSettings {
      bApplyOutlineToDropShadows=false
      bMiteredCorners=false
      bSeparateFillAlpha=false
      outlineMaterial=null
      outlineSize=0
      outlineColor=<rgba 0 0 0 1>
     }
    }
    renderTransform {angle=0, scale=<0.7 0.7>, shear=<0 0>, translation=<-26 -30>}
    shadowOffset=<0 0>
   }
   slot {
    class=CanvasPanelSlot
    properties {layoutData {alignment=<0 0>, anchors {maximum=<0 0>, minimum=<0 0>}, offsets=<ltrb 50.5796 49.8205 10.1099 44>}}
   }
  }
  CommonTextBlock_11 {
   widgetClassPath=&CommonTextBlock
   parent=Knob_Canvas
   widget=CommonTextBlock_11
   properties {
    style=&TextStyle-Base_C
    text=-12
    font {
     bForceMonospaced=false
     bMaterialIsStencil=false
     fontMaterial=null
     fontObject=&NotoSans
     letterSpacing=0
     monospacedWidth=1
     size=24
     skewAmount=0
     typefaceFontName=NotoSans-Regular
     outlineSettings {
      bApplyOutlineToDropShadows=false
      bMiteredCorners=false
      bSeparateFillAlpha=false
      outlineMaterial=null
      outlineSize=0
      outlineColor=<rgba 0 0 0 1>
     }
    }
    renderTransform {angle=0, scale=<0.3 0.3>, shear=<0 0>, translation=<-26 -30>}
    shadowOffset=<0 0>
   }
   slot {
    class=CanvasPanelSlot
    properties {layoutData {alignment=<0 0>, anchors {maximum=<0 0>, minimum=<0 0>}, offsets=<ltrb 65.805 25.7917 48.5384 44>}}
   }
  }
  CommonTextBlock_14 {
   widgetClassPath=&CommonTextBlock
   parent=Knob_Canvas
   widget=CommonTextBlock_14
   properties {
    style=&TextStyle-Base_C
    text=-6
    font {
     bForceMonospaced=false
     bMaterialIsStencil=false
     fontMaterial=null
     fontObject=&NotoSans
     letterSpacing=0
     monospacedWidth=1
     size=24
     skewAmount=0
     typefaceFontName=NotoSans-Regular
     outlineSettings {
      bApplyOutlineToDropShadows=false
      bMiteredCorners=false
      bSeparateFillAlpha=false
      outlineMaterial=null
      outlineSize=0
      outlineColor=<rgba 0 0 0 1>
     }
    }
    renderTransform {angle=0, scale=<0.3 0.3>, shear=<0 0>, translation=<-26 -30>}
    shadowOffset=<0 0>
   }
   slot {
    class=CanvasPanelSlot
    properties {layoutData {alignment=<0 0>, anchors {maximum=<0 0>, minimum=<0 0>}, offsets=<ltrb 83.6515 51.4351 48.5384 44>}}
   }
  }
  CommonTextBlock_24 {
   widgetClassPath=&CommonTextBlock
   parent=Knob_Canvas
   widget=CommonTextBlock_24
   properties {
    style=&TextStyle-Base_C
    text=0
    font {
     bForceMonospaced=false
     bMaterialIsStencil=false
     fontMaterial=null
     fontObject=&NotoSans
     letterSpacing=0
     monospacedWidth=1
     size=24
     skewAmount=0
     typefaceFontName=NotoSans-Regular
     outlineSettings {
      bApplyOutlineToDropShadows=false
      bMiteredCorners=false
      bSeparateFillAlpha=false
      outlineMaterial=null
      outlineSize=0
      outlineColor=<rgba 0 0 0 1>
     }
    }
    renderTransform {angle=0, scale=<0.3 0.3>, shear=<0 0>, translation=<-26 -30>}
    shadowOffset=<0 0>
   }
   slot {
    class=CanvasPanelSlot
    properties {layoutData {alignment=<0 0>, anchors {maximum=<0 0>, minimum=<0 0>}, offsets=<ltrb 72.8397 75.2904 48.5384 44>}}
   }
  }
  CommonTextBlock_25 {
   widgetClassPath=&CommonTextBlock
   parent=Knob_Canvas
   widget=CommonTextBlock_25
   properties {
    style=&TextStyle-Base_C
    text="∞"
    font {
     bForceMonospaced=false
     bMaterialIsStencil=false
     fontMaterial=null
     fontObject=&NotoSans
     letterSpacing=0
     monospacedWidth=1
     size=24
     skewAmount=0
     typefaceFontName=NotoSans-Regular
     outlineSettings {
      bApplyOutlineToDropShadows=false
      bMiteredCorners=false
      bSeparateFillAlpha=false
      outlineMaterial=null
      outlineSize=0
      outlineColor=<rgba 0 0 0 1>
     }
    }
    renderTransform {angle=0, scale=<0.3 0.3>, shear=<0 0>, translation=<-26 -30>}
    shadowOffset=<0 0>
   }
   slot {
    class=CanvasPanelSlot
    properties {layoutData {alignment=<0 0>, anchors {maximum=<0 0>, minimum=<0 0>}, offsets=<ltrb 36.3415 75.2243 48.5384 44>}}
   }
  }
  CommonTextBlock_16 {
   widgetClassPath=&CommonTextBlock
   parent=Knob_Canvas
   widget=CommonTextBlock_16
   properties {
    style=&TextStyle-Base_C
    text=-24
    font {
     bForceMonospaced=false
     bMaterialIsStencil=false
     fontMaterial=null
     fontObject=&NotoSans
     letterSpacing=0
     monospacedWidth=1
     size=24
     skewAmount=0
     typefaceFontName=NotoSans-Regular
     outlineSettings {
      bApplyOutlineToDropShadows=false
      bMiteredCorners=false
      bSeparateFillAlpha=false
      outlineMaterial=null
      outlineSize=0
      outlineColor=<rgba 0 0 0 1>
     }
    }
    renderTransform {angle=0, scale=<0.3 0.3>, shear=<0 0>, translation=<-26 -30>}
    shadowOffset=<0 0>
   }
   slot {
    class=CanvasPanelSlot
    properties {layoutData {alignment=<0 0>, anchors {maximum=<0 0>, minimum=<0 0>}, offsets=<ltrb 18.3121 51.3847 48.5384 44>}}
   }
  }
  CommonTextBlock_13 {
   widgetClassPath=&CommonTextBlock
   parent=Knob_Canvas
   widget=CommonTextBlock_13
   properties {
    style=&TextStyle-Base_C
    text=-18
    font {
     bForceMonospaced=false
     bMaterialIsStencil=false
     fontMaterial=null
     fontObject=&NotoSans
     letterSpacing=0
     monospacedWidth=1
     size=24
     skewAmount=0
     typefaceFontName=NotoSans-Regular
     outlineSettings {
      bApplyOutlineToDropShadows=false
      bMiteredCorners=false
      bSeparateFillAlpha=false
      outlineMaterial=null
      outlineSize=0
      outlineColor=<rgba 0 0 0 1>
     }
    }
    renderTransform {angle=0, scale=<0.3 0.3>, shear=<0 0>, translation=<-26 -30>}
    shadowOffset=<0 0>
   }
   slot {
    class=CanvasPanelSlot
    properties {layoutData {alignment=<0 0>, anchors {maximum=<0 0>, minimum=<0 0>}, offsets=<ltrb 35.0012 25.9428 47.8949 44>}}
   }
  }
  CommonTextBlock_9 {
   widgetClassPath=&CommonTextBlock
   parent=Knob_Canvas
   widget=CommonTextBlock_9
   properties {
    style=&TextStyle-Base_C
    text=-
    font {
     bForceMonospaced=false
     bMaterialIsStencil=false
     fontMaterial=null
     fontObject=&NotoSans
     letterSpacing=0
     monospacedWidth=1
     size=24
     skewAmount=0
     typefaceFontName=NotoSans-Regular
     outlineSettings {
      bApplyOutlineToDropShadows=false
      bMiteredCorners=false
      bSeparateFillAlpha=false
      outlineMaterial=null
      outlineSize=0
      outlineColor=<rgba 0 0 0 1>
     }
    }
    renderTransform {angle=-60, scale=<0.7 0.7>, shear=<0 0>, translation=<-26 -30>}
    shadowOffset=<0 0>
   }
   slot {
    class=CanvasPanelSlot
    properties {layoutData {alignment=<0 0>, anchors {maximum=<0 0>, minimum=<0 0>}, offsets=<ltrb 77.0784 34.4003 12.5024 44>}}
   }
  }
  CommonTextBlock_12 {
   widgetClassPath=&CommonTextBlock
   parent=Knob_Canvas
   widget=CommonTextBlock_12
   properties {
    style=&TextStyle-Base_C
    text=-
    font {
     bForceMonospaced=false
     bMaterialIsStencil=false
     fontMaterial=null
     fontObject=&NotoSans
     letterSpacing=0
     monospacedWidth=1
     size=24
     skewAmount=0
     typefaceFontName=NotoSans-Regular
     outlineSettings {
      bApplyOutlineToDropShadows=false
      bMiteredCorners=false
      bSeparateFillAlpha=false
      outlineMaterial=null
      outlineSize=0
      outlineColor=<rgba 0 0 0 1>
     }
    }
    renderTransform {angle=-120, scale=<0.7 0.7>, shear=<0 0>, translation=<-26 -30>}
    shadowOffset=<0 0>
   }
   slot {
    class=CanvasPanelSlot
    properties {layoutData {alignment=<0 0>, anchors {maximum=<0 0>, minimum=<0 0>}, offsets=<ltrb 56.5276 36.629 12.5024 44>}}
   }
  }
  SynthKnob_0 {
   widgetClassPath=/Script/Synthesis.SynthKnob
   bIsVariable=true
   parent=Knob_Canvas
   widget=SynthKnob_0
   properties {
    value=0.81
    renderTransform {angle=0, scale=<0.2 0.2>, shear=<0 0>, translation=<-26 -30>}
    widgetStyle {
     knobSize=Large
     maxValueAngle=0.4
     minValueAngle=-0.4
     largeKnob {
      bIsDynamicallyLoaded=true
      drawAs=Image
      imageType=FullColor
      mirroring=NoMirror
      resourceName=../../../Engine/Plugins/Runtime/Synthesis/Content/UI/SynthKnobLarge.png
      resourceObject=null
      tiling=NoTile
      imageSize=<150 150>
      margin=<ltrb 0 0 0 0>
      outlineSettings {
       bUseBrushTransparency=false
       roundingType=HalfHeightRadius
       width=0
       color {colorUseRule=UseColor_Specified, specifiedColor=<rgba 0 0 0 0>}
       cornerRadii=<0 0 0 0>
      }
      tintColor {colorUseRule=UseColor_Specified, specifiedColor=<rgba 1 1 1 1>}
      uVRegion {bIsValid=false, max=<0 0>, min=<0 0>}
     }
     largeKnobOverlay {
      bIsDynamicallyLoaded=true
      drawAs=Image
      imageType=FullColor
      mirroring=NoMirror
      resourceName=../../../Engine/Plugins/Runtime/Synthesis/Content/UI/SynthKnobLargeOverlay.png
      resourceObject=null
      tiling=NoTile
      imageSize=<150 150>
      margin=<ltrb 0 0 0 0>
      outlineSettings {
       bUseBrushTransparency=false
       roundingType=HalfHeightRadius
       width=0
       color {colorUseRule=UseColor_Specified, specifiedColor=<rgba 0 0 0 0>}
       cornerRadii=<0 0 0 0>
      }
      tintColor {colorUseRule=UseColor_Specified, specifiedColor=<rgba 1 1 1 1>}
      uVRegion {bIsValid=false, max=<0 0>, min=<0 0>}
     }
     mediumKnob {
      bIsDynamicallyLoaded=true
      drawAs=Image
      imageType=FullColor
      mirroring=NoMirror
      resourceName=../../../Engine/Plugins/Runtime/Synthesis/Content/UI/SynthKnobMedium.png
      resourceObject=null
      tiling=NoTile
      imageSize=<150 150>
      margin=<ltrb 0 0 0 0>
      outlineSettings {
       bUseBrushTransparency=false
       roundingType=HalfHeightRadius
       width=0
       color {colorUseRule=UseColor_Specified, specifiedColor=<rgba 0 0 0 0>}
       cornerRadii=<0 0 0 0>
      }
      tintColor {colorUseRule=UseColor_Specified, specifiedColor=<rgba 1 1 1 1>}
      uVRegion {bIsValid=false, max=<0 0>, min=<0 0>}
     }
     mediumKnobOverlay {
      bIsDynamicallyLoaded=true
      drawAs=Image
      imageType=FullColor
      mirroring=NoMirror
      resourceName=../../../Engine/Plugins/Runtime/Synthesis/Content/UI/SynthKnobMediumOverlay.png
      resourceObject=null
      tiling=NoTile
      imageSize=<150 150>
      margin=<ltrb 0 0 0 0>
      outlineSettings {
       bUseBrushTransparency=false
       roundingType=HalfHeightRadius
       width=0
       color {colorUseRule=UseColor_Specified, specifiedColor=<rgba 0 0 0 0>}
       cornerRadii=<0 0 0 0>
      }
      tintColor {colorUseRule=UseColor_Specified, specifiedColor=<rgba 1 1 1 1>}
      uVRegion {bIsValid=false, max=<0 0>, min=<0 0>}
     }
    }
   }
   slot {
    class=CanvasPanelSlot
    properties {bAutoSize=true, layoutData {alignment=<0 0>, anchors {maximum=<0 0>, minimum=<0 0>}, offsets=<ltrb 0 0 150 150>}}
   }
  }
  MultiplierInput_TextBox {
   widgetClassPath=&EditableTextBox
   bIsVariable=true
   parent=Knob_Canvas
   widget=MultiplierInput_TextBox
   properties {
    bIsEnabled=false
    justification=Right
    text=+0.00 db
    widgetStyle {
     backgroundColor {colorUseRule=UseColor_Specified, specifiedColor=<rgba 1 1 1 1>}
     backgroundImageFocused {
      bIsDynamicallyLoaded=false
      drawAs=RoundedBox
      imageType=NoImage
      mirroring=NoMirror
      resourceName=None
      resourceObject=null
      tiling=NoTile
      imageSize=<0 0>
      margin=<ltrb 0 0 0 0>
      outlineSettings {
       bUseBrushTransparency=false
       roundingType=FixedRadius
       width=1
       color {colorUseRule=UseColor_ColorTable, specifiedColor=<rgba 0 0 0 0>}
       cornerRadii=<4 4 4 4>
      }
      tintColor {colorUseRule=UseColor_Specified, specifiedColor=<rgba 0.0048 0.0048 0.0048 1>}
      uVRegion {bIsValid=false, max=<0 0>, min=<0 0>}
     }
     backgroundImageHovered {
      bIsDynamicallyLoaded=false
      drawAs=RoundedBox
      imageType=NoImage
      mirroring=NoMirror
      resourceName=None
      resourceObject=null
      tiling=NoTile
      imageSize=<0 0>
      margin=<ltrb 0 0 0 0>
      outlineSettings {
       bUseBrushTransparency=false
       roundingType=FixedRadius
       width=1
       color {colorUseRule=UseColor_ColorTable, specifiedColor=<rgba 0 0 0 0>}
       cornerRadii=<4 4 4 4>
      }
      tintColor {colorUseRule=UseColor_Specified, specifiedColor=<rgba 0.0048 0.0048 0.0048 1>}
      uVRegion {bIsValid=false, max=<0 0>, min=<0 0>}
     }
     backgroundImageNormal {
      bIsDynamicallyLoaded=false
      drawAs=RoundedBox
      imageType=NoImage
      mirroring=NoMirror
      resourceName=None
      resourceObject=null
      tiling=NoTile
      imageSize=<0 0>
      margin=<ltrb 0 0 0 0>
      outlineSettings {
       bUseBrushTransparency=false
       roundingType=FixedRadius
       width=1
       color {colorUseRule=UseColor_ColorTable, specifiedColor=<rgba 0 0 0 0>}
       cornerRadii=<4 4 4 4>
      }
      tintColor {colorUseRule=UseColor_Specified, specifiedColor=<rgba 0.0048 0.0048 0.0048 1>}
      uVRegion {bIsValid=false, max=<0 0>, min=<0 0>}
     }
     backgroundImageReadOnly {
      bIsDynamicallyLoaded=false
      drawAs=RoundedBox
      imageType=NoImage
      mirroring=NoMirror
      resourceName=None
      resourceObject=null
      tiling=NoTile
      imageSize=<0 0>
      margin=<ltrb 0 0 0 0>
      outlineSettings {
       bUseBrushTransparency=false
       roundingType=FixedRadius
       width=1
       color {colorUseRule=UseColor_ColorTable, specifiedColor=<rgba 0 0 0 0>}
       cornerRadii=<4 4 4 4>
      }
      tintColor {colorUseRule=UseColor_Specified, specifiedColor=<rgba 0.0284 0.0284 0.0284 1>}
      uVRegion {bIsValid=false, max=<0 0>, min=<0 0>}
     }
     focusedForegroundColor {colorUseRule=UseColor_ColorTable, specifiedColor=<rgba 0 0 0 0>}
     foregroundColor {colorUseRule=UseColor_Specified, specifiedColor=<rgba 0.5271 0.5271 0.5271 1>}
     hScrollBarPadding=<ltrb 0 0 0 0>
     padding=<ltrb 8 3 8 4>
     readOnlyForegroundColor {colorUseRule=UseColor_Foreground, specifiedColor=<rgba 1 0 1 1>}
     scrollBarStyle {
      thickness=8
      draggedThumbImage {
       bIsDynamicallyLoaded=false
       drawAs=RoundedBox
       imageType=NoImage
       mirroring=NoMirror
       resourceName=None
       resourceObject=null
       tiling=NoTile
       imageSize=<0 0>
       margin=<ltrb 0 0 0 0>
       outlineSettings {
        bUseBrushTransparency=false
        roundingType=FixedRadius
        width=0
        color {colorUseRule=UseColor_Specified, specifiedColor=<rgba 0 0 0 0>}
        cornerRadii=<4 4 4 4>
       }
       tintColor {colorUseRule=UseColor_Specified, specifiedColor=<rgba 0.2159 0.2159 0.2159 1>}
       uVRegion {bIsValid=false, max=<0 0>, min=<0 0>}
      }
      horizontalBackgroundImage {
       bIsDynamicallyLoaded=false
       drawAs=NoDrawType
       imageType=NoImage
       mirroring=NoMirror
       resourceName=None
       resourceObject=null
       tiling=NoTile
       imageSize=<0 0>
       margin=<ltrb 0 0 0 0>
       outlineSettings {
        bUseBrushTransparency=false
        roundingType=HalfHeightRadius
        width=0
        color {colorUseRule=UseColor_Specified, specifiedColor=<rgba 0 0 0 0>}
        cornerRadii=<0 0 0 0>
       }
       tintColor {colorUseRule=UseColor_Specified, specifiedColor=<rgba 1 1 1 1>}
       uVRegion {bIsValid=false, max=<0 0>, min=<0 0>}
      }
      horizontalBottomSlotImage {
       bIsDynamicallyLoaded=false
       drawAs=NoDrawType
       imageType=NoImage
       mirroring=NoMirror
       resourceName=None
       resourceObject=null
       tiling=NoTile
       imageSize=<0 0>
       margin=<ltrb 0 0 0 0>
       outlineSettings {
        bUseBrushTransparency=false
        roundingType=HalfHeightRadius
        width=0
        color {colorUseRule=UseColor_Specified, specifiedColor=<rgba 0 0 0 0>}
        cornerRadii=<0 0 0 0>
       }
       tintColor {colorUseRule=UseColor_Specified, specifiedColor=<rgba 1 1 1 1>}
       uVRegion {bIsValid=false, max=<0 0>, min=<0 0>}
      }
      horizontalTopSlotImage {
       bIsDynamicallyLoaded=false
       drawAs=NoDrawType
       imageType=NoImage
       mirroring=NoMirror
       resourceName=None
       resourceObject=null
       tiling=NoTile
       imageSize=<0 0>
       margin=<ltrb 0 0 0 0>
       outlineSettings {
        bUseBrushTransparency=false
        roundingType=HalfHeightRadius
        width=0
        color {colorUseRule=UseColor_Specified, specifiedColor=<rgba 0 0 0 0>}
        cornerRadii=<0 0 0 0>
       }
       tintColor {colorUseRule=UseColor_Specified, specifiedColor=<rgba 1 1 1 1>}
       uVRegion {bIsValid=false, max=<0 0>, min=<0 0>}
      }
      hoveredThumbImage {
       bIsDynamicallyLoaded=false
       drawAs=RoundedBox
       imageType=NoImage
       mirroring=NoMirror
       resourceName=None
       resourceObject=null
       tiling=NoTile
       imageSize=<0 0>
       margin=<ltrb 0 0 0 0>
       outlineSettings {
        bUseBrushTransparency=false
        roundingType=FixedRadius
        width=0
        color {colorUseRule=UseColor_Specified, specifiedColor=<rgba 0 0 0 0>}
        cornerRadii=<4 4 4 4>
       }
       tintColor {colorUseRule=UseColor_Specified, specifiedColor=<rgba 0.2159 0.2159 0.2159 1>}
       uVRegion {bIsValid=false, max=<0 0>, min=<0 0>}
      }
      normalThumbImage {
       bIsDynamicallyLoaded=false
       drawAs=RoundedBox
       imageType=NoImage
       mirroring=NoMirror
       resourceName=None
       resourceObject=null
       tiling=NoTile
       imageSize=<0 0>
       margin=<ltrb 0 0 0 0>
       outlineSettings {
        bUseBrushTransparency=false
        roundingType=FixedRadius
        width=0
        color {colorUseRule=UseColor_Specified, specifiedColor=<rgba 0 0 0 0>}
        cornerRadii=<4 4 4 4>
       }
       tintColor {colorUseRule=UseColor_Specified, specifiedColor=<rgba 0.0953 0.0953 0.0953 1>}
       uVRegion {bIsValid=false, max=<0 0>, min=<0 0>}
      }
      verticalBackgroundImage {
       bIsDynamicallyLoaded=false
       drawAs=NoDrawType
       imageType=NoImage
       mirroring=NoMirror
       resourceName=None
       resourceObject=null
       tiling=NoTile
       imageSize=<0 0>
       margin=<ltrb 0 0 0 0>
       outlineSettings {
        bUseBrushTransparency=false
        roundingType=HalfHeightRadius
        width=0
        color {colorUseRule=UseColor_Specified, specifiedColor=<rgba 0 0 0 0>}
        cornerRadii=<0 0 0 0>
       }
       tintColor {colorUseRule=UseColor_Specified, specifiedColor=<rgba 1 1 1 1>}
       uVRegion {bIsValid=false, max=<0 0>, min=<0 0>}
      }
      verticalBottomSlotImage {
       bIsDynamicallyLoaded=false
       drawAs=NoDrawType
       imageType=NoImage
       mirroring=NoMirror
       resourceName=None
       resourceObject=null
       tiling=NoTile
       imageSize=<0 0>
       margin=<ltrb 0 0 0 0>
       outlineSettings {
        bUseBrushTransparency=false
        roundingType=HalfHeightRadius
        width=0
        color {colorUseRule=UseColor_Specified, specifiedColor=<rgba 0 0 0 0>}
        cornerRadii=<0 0 0 0>
       }
       tintColor {colorUseRule=UseColor_Specified, specifiedColor=<rgba 1 1 1 1>}
       uVRegion {bIsValid=false, max=<0 0>, min=<0 0>}
      }
      verticalTopSlotImage {
       bIsDynamicallyLoaded=false
       drawAs=NoDrawType
       imageType=NoImage
       mirroring=NoMirror
       resourceName=None
       resourceObject=null
       tiling=NoTile
       imageSize=<0 0>
       margin=<ltrb 0 0 0 0>
       outlineSettings {
        bUseBrushTransparency=false
        roundingType=HalfHeightRadius
        width=0
        color {colorUseRule=UseColor_Specified, specifiedColor=<rgba 0 0 0 0>}
        cornerRadii=<0 0 0 0>
       }
       tintColor {colorUseRule=UseColor_Specified, specifiedColor=<rgba 1 1 1 1>}
       uVRegion {bIsValid=false, max=<0 0>, min=<0 0>}
      }
     }
     textStyle {
      overflowPolicy=Clip
      transformPolicy=None
      colorAndOpacity {colorUseRule=UseColor_Foreground, specifiedColor=<rgba 1 0 1 1>}
      font {
       bForceMonospaced=false
       bMaterialIsStencil=false
       fontMaterial=null
       fontObject=null
       letterSpacing=0
       monospacedWidth=1
       size=10
       skewAmount=0
       typefaceFontName=Regular
       outlineSettings {
        bApplyOutlineToDropShadows=false
        bMiteredCorners=false
        bSeparateFillAlpha=false
        outlineMaterial=null
        outlineSize=0
        outlineColor=<rgba 0 0 0 1>
       }
      }
      highlightColor {colorUseRule=UseColor_ColorTable, specifiedColor=<rgba 0 0 0 0>}
      highlightShape {
       bIsDynamicallyLoaded=false
       drawAs=Image
       imageType=NoImage
       mirroring=NoMirror
       resourceName=None
       resourceObject=null
       tiling=NoTile
       imageSize=<0 0>
       margin=<ltrb 0 0 0 0>
       outlineSettings {
        bUseBrushTransparency=false
        roundingType=HalfHeightRadius
        width=0
        color {colorUseRule=UseColor_Specified, specifiedColor=<rgba 0 0 0 0>}
        cornerRadii=<0 0 0 0>
       }
       tintColor {colorUseRule=UseColor_ColorTable, specifiedColor=<rgba 0 0 0 0>}
       uVRegion {bIsValid=false, max=<0 0>, min=<0 0>}
      }
      selectedBackgroundColor {colorUseRule=UseColor_ColorTable, specifiedColor=<rgba 0 0 0 0>}
      shadowColorAndOpacity=<rgba 0 0 0 1>
      shadowOffset=<0 0>
      strikeBrush {
       bIsDynamicallyLoaded=false
       drawAs=Image
       imageType=NoImage
       mirroring=NoMirror
       resourceName=None
       resourceObject=null
       tiling=NoTile
       imageSize=<32 32>
       margin=<ltrb 0 0 0 0>
       outlineSettings {
        bUseBrushTransparency=false
        roundingType=HalfHeightRadius
        width=0
        color {colorUseRule=UseColor_Specified, specifiedColor=<rgba 0 0 0 0>}
        cornerRadii=<0 0 0 0>
       }
       tintColor {colorUseRule=UseColor_Specified, specifiedColor=<rgba 1 1 1 1>}
       uVRegion {bIsValid=false, max=<0 0>, min=<0 0>}
      }
      underlineBrush {
       bIsDynamicallyLoaded=false
       drawAs=Image
       imageType=NoImage
       mirroring=NoMirror
       resourceName=None
       resourceObject=null
       tiling=NoTile
       imageSize=<32 32>
       margin=<ltrb 0 0 0 0>
       outlineSettings {
        bUseBrushTransparency=false
        roundingType=HalfHeightRadius
        width=0
        color {colorUseRule=UseColor_Specified, specifiedColor=<rgba 0 0 0 0>}
        cornerRadii=<0 0 0 0>
       }
       tintColor {colorUseRule=UseColor_Specified, specifiedColor=<rgba 1 1 1 1>}
       uVRegion {bIsValid=false, max=<0 0>, min=<0 0>}
      }
     }
     vScrollBarPadding=<ltrb 0 0 0 0>
    }
   }
   slot {
    class=CanvasPanelSlot
    properties {layoutData {alignment=<0 0>, anchors {maximum=<0 0>, minimum=<0 0>}, offsets=<ltrb 14.1188 77.149 68.8522 16.8952>}}
   }
  }
  MultiplierValue_TextBox {
   widgetClassPath=&CommonTextBlock
   bIsVariable=true
   parent=CanvasPanel_154
   widget=MultiplierValue_TextBox
   properties {
    justification=Right
    style=&TextStyle-Base_C
    text=+0.00 dB
    font {
     bForceMonospaced=false
     bMaterialIsStencil=false
     fontMaterial=null
     fontObject=&NotoSans
     letterSpacing=0
     monospacedWidth=1
     size=24
     skewAmount=0
     typefaceFontName=NotoSans-Regular
     outlineSettings {
      bApplyOutlineToDropShadows=false
      bMiteredCorners=false
      bSeparateFillAlpha=false
      outlineMaterial=null
      outlineSize=0
      outlineColor=<rgba 0 0 0 1>
     }
    }
    renderTransform {angle=0, scale=<0.5 0.5>, shear=<0 0>, translation=<0 0>}
    shadowOffset=<0 0>
   }
   slot {class=CanvasPanelSlot, properties {layoutData {alignment=<0 0>, anchors {maximum=<0 0>, minimum=<0 0>}, offsets=<ltrb -32 84 153 44>}}}
  }
 ]
}
