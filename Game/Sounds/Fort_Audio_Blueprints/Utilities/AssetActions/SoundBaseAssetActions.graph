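# Editor utility blueprint of asset actions for batch-editing selected SoundBase assets and generating audio helper assets.
# Adjusts volume in decibels, tags stream-safe music, adds analysis submix sends and builds dependency or referencer collections.
# Sound class switching rescales volume by the ratio of parent-chain gains, then warns when results leave logical thresholds.
# Embedded Python creates linked sidechain modulation assets and opens the listen manager widget.
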EditorUtilityBlueprint /Game/Sounds/Fort_Audio_Blueprints/Utilities/AssetActions/SoundBaseAssetActions
graphs [
 Adjust Volume in Decibels {nodes [
   n1 = Adjust Volume in Decibels!() -> (then, Decibel Amount:double=~, Rounded Result:bool=~) {is=title K2Node_FunctionEntry}
   > n2 = GetSelectedAssets()
   > n3 = For Each Loop(Array:Array of Object References=@n2) -> (Array Element:Object Reference)
   > n4 = Cast To SoundBase!(execute, Object=@n3) -> (then, CastFailed, AsSound Base:Sound Base Object Reference) {is=title K2Node_DynamicCast}
   n5 = CueOrWaveProperty!(Object:Object Reference=@n3, Wave Property:Name=Volume, Cue Property:Name=VolumeMultiplier) -> (ReturnValue:Name=None)
   > n6 = GetEditorProperty!(execute:Exec, Object:Object Reference=@n4, PropertyName:Name=@n5) -> (then:Exec, PropertyValue:double, ReturnValue:bool)
   n7 = Db to Linear!(Db:double=float + float(A=Linear to Db!(Linear:double=@n6) -> (Db:double=~) {is=title K2Node_MacroInstance}, B=@n1) {is=title K2Node_CommutativeAssociativeBinaryOperator}) -> (Linear:double=~) {is=title K2Node_MacroInstance}
   n8 = "To Text (Float)"(Value=@n7, bUseGrouping=false, MinimumIntegralDigits=1, MaximumIntegralDigits=324, MinimumFractionalDigits=0, MaximumFractionalDigits=3)
   n9 = SelectFloat(A=@n7, B=String To Float(InString="To String (Text)"(InText=@n8)), bPickA=NOT Boolean(A=@n1.Rounded Result))
   > PrintString(InString="To String (Float)"(InDouble=@n9))
   > SetEditorProperty(Object=@n4, PropertyName=@n5, PropertyValue:double=@n9)
  ]}
 CheckForVolumeThreshold {nodes [
   n1 = CheckForVolumeThreshold!() -> (then, Object:Sound Base Object Reference) {is=title K2Node_FunctionEntry}
   > n2 = GetEditorProperty(Object=@n1, PropertyName=GetVolumePropertyName!(Object:Object Reference=@n1) -> (ReturnValue:Name=None))
   n3 = Get LogicalVolumeMaxThreshold() -> (LogicalVolumeMaxThreshold:double)
   n4 = Get LogicalVolumeMinThreshold() -> (LogicalVolumeMinThreshold:double)
   > Return Node!(execute, bIsOutsideThreshold:bool=OR Boolean(A="float > float"(A=@n2, B=@n3), B="float < float"(A=@n2, B=@n4)))
  ]}
 CreateCollectionDependencies {nodes [
   n1 = CreateCollectionDependencies!() -> (then, Name:Name) {is=title K2Node_FunctionEntry}
   > n2 = Iterate Over Selected Assets()
   n3 = Get CollectionManagerScriptingSubsystem()
   n4 = CreateOrEmptyCollection!(
    execute=@n2.Completed
    self:Collection Manager Subsystem Object Reference=@n3
    Container:Collection Container Source Structure
    Collection:Name=@n1
    ShareType:ECollectionScriptingShareType Enum=Local
   ) -> (then, OutNewOrEmptyCollection:Collection Structure, ReturnValue:bool)
   > AddAssetsToCollection!(
    execute
    self:Collection Manager Subsystem Object Reference=@n3
    Collection:"Collection Structure (by ref)"=@n4
    AssetPaths:Array of Soft Object Path Structures=Get AssetPaths() -> (AssetPaths:Array of Soft Object Path Structures)
   ) -> (then, ReturnValue:bool)
   n5 = Get Asset Dependencies!(execute:Exec=@n2.Iteration, Object:Object Reference=@n2) -> (Iteration:Exec, Dependency:Object Reference, Completed:Exec) {is=title K2Node_MacroInstance}
   > Add Unique(
    TargetArray:Array of Soft Object Path Structures=Get AssetPaths()
    NewItem:"Soft Object Path Structure (by ref)"=GetSoftObjectPath(Object=@n5)
   )
  ]}
 CreateCollectionReferencers {nodes [
   n1 = CreateCollectionReferencers!() -> (then, Name:Name) {is=title K2Node_FunctionEntry}
   > n2 = Iterate Over Selected Assets()
   n3 = Get CollectionManagerScriptingSubsystem()
   n4 = CreateOrEmptyCollection(execute=@n2.Completed, self=@n3, Collection=@n1, ShareType=Local)
   > AddAssetsToCollection(self=@n3, Collection=@n4, AssetPaths=Get AssetPaths())
   n5 = Get Asset Referencers!(execute:Exec=@n2.Iteration, Object:Object Reference=@n2) -> (Iteration:Exec, Referencer:Object Reference, Completed:Exec) {is=title K2Node_MacroInstance}
   > Add Unique(TargetArray=Get AssetPaths(), NewItem=GetSoftObjectPath(Object=@n5))
  ]}
 CreateSidechainAssets {nodes [
   n1 = CreateSidechainAssets!() -> (then, AssetTypeName:String, ModulationGenerationType:EModulationAssetGenerationMode Enum=NewEnumerator0, bCreateParameterPatch?:bool=~) {is=title K2Node_FunctionEntry}
   > n2 = Branch(Condition=IsEmpty(InString=@n1))
   > Show Message Dialog(
    Title=Get Title() -> (Title:Text)
    Message="To Text (String)"(InString="An Asset Type Name was not provided, so assets will not be generated. Please provide an Asset Type name so the related sidechain assets will have a correct name. ")
    MessageType=Ok
    DefaultValue=No
    MessageCategory=Error
   )
   Execute Python Script!(
    execute=@n2.else
    PythonScript="import unreal\r\nfrom unreal import SoundModulationPatch\r\n\r\n\r\n#### BP Inputs\r\n#### asset_name : str:  of the type of sound being sidechained\r\n#### create_parameter_patch : bool  : Optional if the user would like a parameter patch created or not.\r\n#### generation_type : int: enum to int for the type of generation of assets. Primarily handles naming of assets.\r\n\r\nSIDECHAIN_SELECTED_ASSETS = 0\r\nSIDECHAIN_FROM_SELECTED = 1\r\n\r\nASSET_TOOLS = unreal.AssetToolsHelpers.get_asset_tools()\r\nEDITOR_UTIL = unreal.get_editor_subsystem(unreal.EditorUtilitySubsystem)\r\n\r\n##Create Modulation Directory\r\nCurrentContentBrowserLocation = unreal.EditorUtilityLibrary.get_current_content_browser_path()\r\nSidechainAssetDirectory = CurrentContentBrowserLocation + \"/\" \"Modulation\"\r\n\r\nif unreal.EditorAssetLibrary.does_directory_exist(SidechainAssetDirectory) is False:\r\n    unreal.EditorAssetLibrary.make_directory(SidechainAssetDirectory)\r\n#### Modulation Parameter Type\r\n\r\nTargetModulationParameter = unreal.EditorAssetLibrary.load_asset(\"/AudioModulation/Volume\") #### Hardcoded here to volume\r\n\r\nCreatedAssetArray = []\r\n\r\nprint(SidechainAssetDirectory)\r\nif generation_type == SIDECHAIN_SELECTED_ASSETS:\r\n    names = {\r\n        \"audio_bus\":       lambda n: f\"AB_{n}_Sidechain\",\r\n        \"modulation_generator\":       lambda n: f\"MG_{n}_Sidechain_EnvelopeFollower\",\r\n        \"control_bus\":  lambda n: f\"CB_{n}_Sidechain\",\r\n        \"parameter_patch\":     lambda n: f\"Patch_{n}_Volume\",\r\n    }\r\nelif generation_type == SIDECHAIN_FROM_SELECTED:\r\n    names = {\r\n        \"audio_bus\":       lambda n: f\"AB_{n}_Bus\",\r\n        \"modulation_generator\":       lambda n: f\"MG_{n}_EnvelopeFollower\",\r\n        \"control_bus\":  lambda n: f\"CB_{n}_Amplitude\",\r\n        \"parameter_patch\":     lambda n: f\"Patch_{n}_Volume\",\r\n    }\r\nelse:\r\n    raise ValueError(f\"Unknown generation_mode: {generation_type}\")\r\n\r\n\r\ndef append_object_to_string_array(input_object):\r\n    if input_object is not None:\r\n        CreatedAssetArray.append(unreal.Object.get_path_name(input_object))\r\n\r\n\r\n######## Create Assets\r\n#### Create Modulation Generator\r\n\r\nCreatedModulationGeneratorAsset = ASSET_TOOLS.create_asset_with_dialog(names[\"modulation_generator\"](asset_name), SidechainAssetDirectory, unreal.SoundModulationGenerator,unreal.SoundModulationGeneratorFactory())\r\n\r\n#### Need this, otherwise python will try to access it before it exists\r\nunreal.EditorAssetLibrary.save_asset(CreatedModulationGeneratorAsset.get_path_name())\r\nappend_object_to_string_array(CreatedModulationGeneratorAsset)\r\n\r\nif not unreal.EditorAssetLibrary.does_asset_exist(CreatedModulationGeneratorAsset.get_path_name()):\r\n    exit(1)\r\n#### create Audio Bus\r\nCreatedAudioBusAsset = ASSET_TOOLS.create_asset(names[\"audio_bus\"](asset_name), SidechainAssetDirectory, unreal.AudioBus,unreal.AudioBusFactory())\r\nappend_object_to_string_array(CreatedAudioBusAsset)\r\n\r\n\r\n##Hack, but we can't set the asset until we go through the dialogue to select the envelope follower\r\n\r\n\r\n\r\n##create Control Bus\r\nCreatedControlBusAsset = ASSET_TOOLS.create_asset(names[\"control_bus\"](asset_name), SidechainAssetDirectory, unreal.SoundControlBus, unreal.SoundControlBusFactory())\r\nappend_object_to_string_array(CreatedControlBusAsset)\r\n\r\n\r\n########   Assign assets to each other\r\nCreatedModulationGeneratorAssetEnvelopeFollower = unreal.EditorAssetLibrary.load_asset(CreatedModulationGeneratorAsset.get_path_name())\r\nModulationGeneratorSettings = unreal.EnvelopeFollowerGeneratorParams(invert=True,audio_bus=CreatedAudioBusAsset)\r\nCreatedModulationGeneratorAssetEnvelopeFollower.set_editor_property(\"params\",ModulationGeneratorSettings)\r\n\r\n\r\n\r\n#### Assign Modulation Generator to Control Bus\r\nCreatedControlBusAsset.set_editor_property(\"generators\",[CreatedModulationGeneratorAssetEnvelopeFollower])\r\n\r\n\r\n\r\n#### Create Parameter Patch\r\nif create_parameter_patch:\r\n    CreatedModulationPatchAsset = ASSET_TOOLS.create_asset(names[\"parameter_patch\"](asset_name), SidechainAssetDirectory, unreal.SoundModulationPatch, unreal.SoundModulationPatchFactory())\r\n\r\n    #### Create an object that holds settings for a parameter Patch\r\n    ControlBusPatchInputAsset = unreal.SoundControlModulationInput()\r\n    ControlBusPatchInputAsset.set_editor_property(\"bus\",CreatedControlBusAsset)\r\n\r\n    #### Setting Parameter Patch Settings\r\n    PatchSettings = unreal.SoundControlModulationPatch()\r\n    PatchSettings.set_editor_property(\"output_parameter\", TargetModulationParameter)\r\n    PatchSettings.set_editor_property(\"bypass\", False) #### defaults to true on the bypass by default\r\n    PatchSettings.set_editor_property(\"inputs\", [ControlBusPatchInputAsset])\r\n    CreatedModulationPatchAsset.set_editor_property(\"patch_settings\",PatchSettings)\r\n    append_object_to_string_array(CreatedModulationPatchAsset)\r\n\r\n##Save/Cleanup\r\nunreal.EditorAssetLibrary.save_directory(CurrentContentBrowserLocation)\r\nunreal.EditorAssetLibrary.save_directory(SidechainAssetDirectory)\r\nunreal.EditorAssetLibrary.sync_browser_to_objects(CreatedAssetArray)"
    asset_name:String=Replace(SourceString=@n1, From=" ", To=_)
    create_parameter_patch:bool=@n1.bCreateParameterPatch?
    generation_type:int="To Integer (Byte)"!(InByte:Byte=@n1.ModulationGenerationType) -> (ReturnValue:int=0)
   ) -> (then, ReturnValue) {inputs [asset_name, create_parameter_patch, generation_type]}
  ]}
 CueOrWaveProperty {nodes [
   n1 = CueOrWaveProperty!() -> (then, Object:Object Reference, Wave Property:Name, Cue Property:Name) {is=title K2Node_FunctionEntry}
   > Return Node!(
    execute
    ReturnValue:Name=Select(
     Option 0:Name=@n1.Wave Property
     Option 1:Name=@n1.Cue Property
     Index=Cast To SoundCue!(Object=@n1) -> (AsSound Cue:Sound Cue Object Reference, bSuccess:bool=~) {is=title K2Node_DynamicCast}.bSuccess
    ) -> (ReturnValue:Name)
   )
  ]}
 EventGraph {links [], nodes []}
 GetSoundClassGain {nodes [
   n1 = GetSoundClassGain!() -> (then, SoundClass:Sound Class Object Reference) {is=title K2Node_FunctionEntry}
   > n2 = Set CurrentClass(CurrentClass:Sound Class Object Reference=@n1) -> (Output_Get:Sound Class Object Reference)
   > n3 = Get Sound Class Volume!(execute:Exec, Object:Object Reference=@n2) -> (then:Exec, Volume:double=~) {is=title K2Node_MacroInstance}
   > n4 = Set GainStage(GainStage:double=@n3) -> (Output_Get:double)
   n5 = GetEditorProperty(
    execute=[@n4.then, @n7.then_1]
    Object=Get CurrentClass() -> (CurrentClass:Sound Class Object Reference)
    PropertyName=MakeLiteralName(Value=ParentClass)
   ) -> (PropertyValue:Sound Class Object Reference)
   > Set ParentClass(ParentClass:Sound Class Object Reference=@n5) -> (Output_Get:Sound Class Object Reference)
   > n6 = Branch(Condition=IsValid(Object=Get ParentClass() -> (ParentClass:Sound Class Object Reference)))
   > n7 = Sequence()
   > n8 = Set CurrentClass(CurrentClass=Get ParentClass())
   > n9 = Get Sound Class Volume(Object=@n8) -> (Volume=~)
   > Set GainStage(GainStage=float * float(A=@n9, B=Get GainStage() -> (GainStage:double)))
   Return Node!(execute=@n6.else, OutGainStage:double=Get GainStage())
  ]}
 GetSoundClassVolume {nodes [
   n1 = Inputs!() -> (execute, Object:Object Reference)
   > n2 = GetEditorProperty(Object=@n1, PropertyName=Properties) -> (PropertyValue:"Sound Class Properties Structure (by ref)")
   > Outputs!(
    then:Exec
    Volume:double=Break Sound Class Properties!(SoundClassProperties:"Sound Class Properties Structure (by ref)"=@n2) -> (Volume:float=~, ModulationSettings:Sound Modulation Default Settings Structure) {is=title K2Node_BreakStruct}
   )
  ]}
 GetVolumePropertyName {nodes [
   n1 = GetVolumePropertyName!() -> (then, Object:Object Reference) {is=title K2Node_FunctionEntry}
   > Return Node!(execute, ReturnValue:Name=Select(Option 0=Volume, Option 1=VolumeMultiplier, Index=Cast To SoundCue(Object=@n1) -> (bSuccess=~).bSuccess))
  ]}
 IsSoundbaseVolumeClipped {nodes [
   n1 = IsSoundbaseVolumeClipped!() -> (then, Soundbase:Sound Base Object Reference) {is=title K2Node_FunctionEntry}
   > n2 = GetEditorProperty(Object=@n1, PropertyName=GetVolumePropertyName(Object=@n1) -> (ReturnValue=None) {is=title K2Node_CallFunction}) -> (PropertyValue:double)
   n3 = float / float(
    A=Get LogicalVolumeMaxThreshold()
    B=GetSoundClassGain!(SoundClass:Sound Class Object Reference=Get SoundClassObject!(self:Sound Base Object Reference=@n1) -> (SoundClassObject:Sound Class Object Reference)) -> (OutGainStage:double) {is=title K2Node_CallFunction}
   )
   n4 = float - float(A=@n3, B=@n2)
   n5 = "float > float"(A=@n2, B=@n3)
   > Return Node!(execute, bIsClipped:bool=@n5, LinearVolumeClippingAmount:double=SelectFloat(A="Absolute (Float)"(A=@n4), B=0, bPickA=@n5))
  ]}
 ListenToAsset {nodes [
   ListenToAsset() {is=title K2Node_FunctionEntry}
   > n1 = GetSelectedAssets()
   n2 = "Get (a copy)"(Array:Array of Object References=@n1, Dimension 1=Last Index!(TargetArray:Array of Object References=@n1) -> (ReturnValue=0)) -> (Output:Object Reference)
   > Is Valid(InputObject=@n2)
   > n3 = Cast To SoundBase(Object=@n2)
   > Execute Python Script!(
    execute
    PythonScript="import unreal\r\n\r\nasset_path_string = \"/Game/Sounds/Fort_Audio_Blueprints/Utilities/EditorUtilityWidgets/SoundListenManager_EWidget\"\r\nasset = unreal.load_asset(asset_path_string)\r\nss = unreal.get_editor_subsystem(unreal.EditorUtilitySubsystem)\r\ntab = ss.spawn_and_register_tab(asset)\r\ntab.call_method(\"Set Audio Asset\", (asset_in,))\r\n"
    asset_in:Sound Base Object Reference=@n3
   ) -> (then, ReturnValue) {inputs [asset_in]}
  ]}
 LogicalVolumeWarning {nodes [
   n1 = LogicalVolumeWarning!() -> (then, SoundbaseAboveVolumeThreshold:Array of Sound Base Object References) {is=title K2Node_FunctionEntry}
   > n2 = Branch(Condition="integer > integer"(A=Length(TargetArray:Array of Sound Base Object References=@n1)))
   > n3 = For Each Loop(Array:Array of Sound Base Object References=@n1) -> (Array Element:Sound Base Object Reference)
   K2Node_Composite_0 = DisplayWarning(execute=@n3.Completed) {is=title K2Node_Composite}
   > n4 = GetEditorProperty(Object=@n3, PropertyName=GetVolumePropertyName(Object=@n3) -> (ReturnValue=None))
   n5 = "To Text (Float)"(Value=@n4, bUseGrouping=false, MinimumIntegralDigits=1, MaximumIntegralDigits=324, MinimumFractionalDigits=0, MaximumFractionalDigits=3)
   > Add(
    TargetArray=Get SoundsWithVolumeAppended() -> (SoundsWithVolumeAppended:Array of Strings)
    NewItem=Append!(
     A="To String (Object)"!(InObj:Object Reference=@n3) -> (ReturnValue:String)
     B="  |  "
     C:String="To String (Text)"(InText=@n5)
     D:String=" Volume "
    ) -> (ReturnValue)
   ) {is=title K2Node_CallArrayFunction}
   LogString(execute=@n2.else, InString="SoundBaseAssetAction: Cannot present warning, no sounds present in the input array.")
  ]}
 DisplayWarning {
  path=LogicalVolumeWarning.K2Node_Composite_0.DisplayWarning
  nodes [
   Inputs!() -> (execute)
   n1 = JoinStringArray(SourceArray=Get SoundsWithVolumeAppended(), Separator="\r\n")
   n2 = Append!(A="To String (Float)"(InDouble=Get LogicalVolumeMinThreshold()), B=" to ", C="To String (Float)"(InDouble=Get LogicalVolumeMaxThreshold())) -> (ReturnValue)
   > Show Message Dialog(
    Title="To Text (String)"(InString="Warning! Logical Volume Too High!")
    Message="To Text (String)"(InString=Append!(
      A="The assets below are outside "
      B="the logical volume range of  "
      C=@n2
      D:String="\r\n\r\n"
      E:String=@n1
      F:String="\r\n\r\nConsider changing volume via other methods, like metasound inputs, soundcue mixers, or normalizing your soundwaves via the waveform editor. Also consider the \"Rebalance Volume\" asset action on your target soundclass to get your soundclass volumes into a reasonable range. "
     ) -> (ReturnValue))
    MessageType=Ok
    DefaultValue=No
    MessageCategory=Warning
   )
   Outputs!()
  ]
 }
 SendToAudioAnalysisSubmix {nodes [
   SendToAudioAnalysisSubmix() {is=title K2Node_FunctionEntry}
   > n1 = Sequence()
   n2 = Get AnalysisSoundSubmixSoftRef() -> (AnalysisSoundSubmixSoftRef:Sound Submix Base Soft Object Reference)
   > LoadAsset_Blocking(Asset=@n2)
   > Set members in Sound Submix Send Info!(
    execute:Exec
    StructRef:"Sound Submix Send Info Structure (by ref)"=Get AnalysisSendInfo() -> (AnalysisSendInfo:Sound Submix Send Info Structure)
    SoundSubmix:Sound Submix Base Object Reference=Resolve Soft Reference!(Input:Sound Submix Base Soft Object Reference=@n2) -> (Output:Sound Submix Base Object Reference)
   ) -> (then:Exec, StructOut:"Sound Submix Send Info Structure (by ref)") {is=title K2Node_SetFieldsInStruct}
   n3 = Iterate Over Selected Assets(Trigger=@n1.then_1)
   > n4 = Cast To SoundBase(Object=@n3)
   > n5 = GetEditorProperty(Object=@n4, PropertyName=SoundSubmixSends) -> (PropertyValue:Array of Sound Submix Send Info Structures)
   > Add Unique(TargetArray:Array of Sound Submix Send Info Structures=@n5, NewItem:"Sound Submix Send Info Structure (by ref)"=Get AnalysisSendInfo())
   > SetEditorProperty(Object=@n4, PropertyName=SoundSubmixSends, PropertyValue:Array of Sound Submix Send Info Structures=@n5)
  ]}
 SetStreamSafeMusic {nodes [
   SetStreamSafeMusic() {is=title K2Node_FunctionEntry}
   > n1 = Iterate Over Selected Assets(AssetClass=/Script/Engine.SoundBase)
   > Is Valid(InputObject=@n1)
   > n2 = Cast To SoundBase(Object=@n1)
   n3 = Get AudioUserData() -> (AudioUserData:Asset User Data Class Reference)
   > n4 = AddAssetUserDataOfClass!(execute, self:Interface Asset User Data Interface=@n2, InUserDataClass:Asset User Data Class Reference=@n3) -> (then, ReturnValue:bool)
   > Branch(Condition=@n4)
   > n5 = GetAssetUserDataOfClass!(execute, self:Interface Asset User Data Interface=@n2, InUserDataClass:Asset User Data Class Reference=@n3) -> (then, ReturnValue:Asset User Data Object Reference)
   > n6 = Cast To AudioAssetUserData!(execute, Object=@n5) -> (then, CastFailed, AsAudio Asset User Data:Audio Asset User Data Object Reference) {is=title K2Node_DynamicCast}
   > Set MetadataTags!(
    execute
    MetadataTags:Gameplay Tag Container Structure=(GameplayTags=((TagName="Audio.Metadata.MuteForStreamSafe")))
    self:Audio Asset User Data Object Reference=@n6
   ) -> (then, Output_Get:Gameplay Tag Container Structure)
   > SaveLoadedAsset(AssetToSave=@n1, bOnlyIfIsDirty=false)
  ]}
 SwitchSoundClass {nodes [
   n1 = SwitchSoundClass!() -> (then, New Sound Class:Sound Class Object Reference) {is=title K2Node_FunctionEntry}
   > n2 = Iterate Over Selected Assets()
   n3 = Get SoundsOutsideThreshold() -> (SoundsOutsideThreshold:Array of Sound Base Object References)
   Branch(execute=@n2.Completed, Condition="integer > integer"(A=Length(TargetArray=@n3)))
   > LogicalVolumeWarning!(execute, self, SoundbaseAboveVolumeThreshold:Array of Sound Base Object References=Get SoundsOutsideThreshold()) -> (then) {is=title K2Node_CallFunction}
   n4 = Cast To SoundBase(execute=@n2.Iteration, Object=@n2)
   > n5 = SwitchSoundClassInternal!(execute, self, SoundBase:Sound Base Object Reference=@n4, SwapToClass:Sound Class Object Reference=@n1) -> (then, SoundBaseWithNewClass:Sound Base Object Reference)
   > Branch(Condition=CheckForVolumeThreshold!(Object:Sound Base Object Reference=@n5) -> (bIsOutsideThreshold:bool) {is=title K2Node_CallFunction})
   > Add(TargetArray:Array of Sound Base Object References=Get SoundsOutsideThreshold(), NewItem:"Sound Base Object Reference (by ref)"=@n5)
  ]}
 SwitchSoundClassInternal {nodes [
   n1 = SwitchSoundClassInternal!() -> (then, SoundBase:Sound Base Object Reference, SwapToClass:Sound Class Object Reference) {is=title K2Node_FunctionEntry}
   n2 = Get SoundClassObject(self=@n1)
   n3 = IsValid(Object=@n2)
   > Set SwapFromClass(SwapFromClass:Sound Class Object Reference=Select(Option 0:Sound Class Object Reference, Option 1:Sound Class Object Reference=@n2, Index=@n3) -> (ReturnValue:Sound Class Object Reference)) -> (Output_Get:Sound Class Object Reference)
   > Set bHasSoundClassDefined(bHasSoundClassDefined:bool=@n3) -> (Output_Get:bool)
   > SetEditorProperty(
    Object=Get SoundBase() -> (SoundBase:Sound Base Object Reference)
    PropertyName=MakeLiteralName(Value=SoundClassObject)
    PropertyValue:Sound Class Object Reference=Get SwapToClass() -> (SwapToClass:Sound Class Object Reference)
   )
   n4 = Get SoundBase()
   n5 = GetVolumePropertyName(Object=@n4) -> (ReturnValue=None)
   > n6 = GetEditorProperty(Object=@n4, PropertyName=@n5) -> (PropertyValue:double)
   n7 = GetSoundClassGain(SoundClass=Get SwapToClass())
   n8 = SelectFloat(
    A=GetSoundClassGain(SoundClass=Get SwapFromClass() -> (SwapFromClass:Sound Class Object Reference))
    B=1
    bPickA=Get bHasSoundClassDefined() -> (bHasSoundClassDefined:bool)
   )
   > SetEditorProperty(Object=@n4, PropertyName=@n5, PropertyValue:double=float * float(A=float / float(A=@n8, B=@n7) {is=title K2Node_CallFunction}, B=@n6))
   > Return Node!(execute, SoundBaseWithNewClass:Sound Base Object Reference=@n4)
  ]}
]
variables [
 AnalysisSendInfo {
  type=SoundSubmixSendInfo
  default {
   bEnableHPFCutoff=false
   bEnableLPFCutoff=false
   disableManualSendClamp=false
   hPFCutoff=20
   lPFCutoff=20000
   maxSendDistance=1000
   maxSendLevel=1
   minSendDistance=100
   minSendLevel=0
   sendLevel=1
   sendLevelControlMethod=Manual
   sendStage=PostDistanceAttenuation
   soundSubmix=null
   customSendLevelCurve {
    externalCurve=null
    editorCurveData {defaultValue=3.4028234663852886e+38, postInfinityExtrap=RCCE_Constant, preInfinityExtrap=RCCE_Constant, keys []}
   }
  }
 }
 AnalysisSoundSubmixSoftRef {type=SoundSubmixBase, default=&Fortnite_AudioAnalysis_Music}
 AudioUserData {type=Class, default=/Script/AudioGameplay.AudioAssetUserData}
 LogicalVolumeMaxThreshold {type=number, default=8}
 LogicalVolumeMinThreshold {type=number, default=0.005}
 StringToBecome {type=string, default=""}
 StringToReplace {type=string, default=""}
 Swap from Class {type=SoundClass, default=&SFX, class=~, graph_type=~, children=~, properties=~}
]
