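# Editor utility widget for audio tooling that, judging by its name and empty widget tree, serves as a base for panels docked in an Audio_EUW_Parent.
# It stores a reference to that parent in EUWParent.
# GetAssetActionInstance forwards a requested asset action class to the parent and returns the resulting instance, with no null check on the reference.
# The event graph holds only unconnected PreConstruct, Construct and Tick events.
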EditorUtilityWidgetBlueprint /Game/Sounds/Fort_Audio_Blueprints/Utilities/EditorUtilityWidgets/Audio_EUW_Dockable
graphs [
 EventGraph {nodes [
   Event PreConstruct()
   Event Construct()
   Event Tick()
  ]}
 GetAssetActionInstance {nodes [
   n1 = GetAssetActionInstance!() -> (then, ClassType:Audio Asset Action Base Class Reference) {is=title K2Node_FunctionEntry}
   > Return Node!(
    execute
    Instance:Audio Asset Action Base Object Reference=GetAssetActionInstance!(
     self:Audio EUW Parent Object Reference=Get EUWParent() -> (EUWParent:Audio EUW Parent Object Reference)
     ClassType:Audio Asset Action Base Class Reference=@n1
    ) -> (Value:Audio Asset Action Base Object Reference) {is=title K2Node_CallFunction}
   )
  ]}
]
variables [EUWParent {type=Audio_EUW_Parent_C, default=null}]
widgets {info {rootWidgetClass=null, widgetCount=0}, widgets []}
