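# Editor utility widget showing a single-channel dB audio meter with numeric readout, driven by calls to its functions.
# ReadVolumes sums envelope values, converts to dB via 20·log10, updates the AudioMeter and prints "-inf" at or below -96.
# PeakValue updates after the meter push whenever level exceeds the previous reading, so it is no true maximum hold.
# SetMeterColorGradient tints the bar by lerping between three colors according to volume.
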EditorUtilityWidgetBlueprint /Game/Sounds/Fort_Audio_Blueprints/Utilities/EditorUtilityWidgets/Audio_EUW_Meter
graphs [
 EventGraph {nodes [
   Event PreConstruct()
   Event Construct()
  ]}
 ReadVolumes {nodes [
   n1 = ReadVolumes!() -> (then, Envelope:"Array of Float (double-precision)s") {is=title K2Node_FunctionEntry}
   > Set SummedVolume(SummedVolume:double=0) -> (Output_Get:double)
   > n2 = For Each Loop(Array:"Array of Float (double-precision)s"=@n1) -> (Array Element:double)
   n3 = Get SummedVolume() -> (SummedVolume:double)
   Set MeterValue(execute=@n2.Completed, MeterValue:double=float * float(A=Log!(A:double=@n3, Base:double=10) -> (ReturnValue:double), B=20)) -> (Output_Get:double)
   n4 = Get PeakValue() -> (PeakValue:double)
   n5 = Get MeterValue() -> (MeterValue:double)
   > SetMeterChannelInfo!(
    execute
    self:Audio Meter Object Reference=Get AudioMeter_46() -> (AudioMeter_46:Audio Meter Object Reference)
    InMeterChannelInfo:Array of Meter Channel Info Structures=Make Array("[0]":Meter Channel Info Structure=Make Meter Channel Info!(MeterValue:float=@n5, PeakValue:float=@n4, ClippingValue:float=0) -> (MeterChannelInfo:Meter Channel Info Structure) {is=title K2Node_MakeStruct}) -> (Array:Array of Meter Channel Info Structures)
   ) -> (then)
   > n6 = Branch(Condition="float > float"(A=Get MeterValue(), B=Get LastVolumeRead() -> (LastVolumeRead:double)) {is=title K2Node_CallFunction})
   > Set PeakValue(PeakValue:double=Get MeterValue()) -> (Output_Get:double)
   n7 = "float <= float"!(A:double=Get MeterValue(), B:double=-96) -> (ReturnValue:bool) {is=title K2Node_CallFunction}
   n8 = "To String (Text)"(InText="To Text (Float)"(Value=Get MeterValue(), MinimumIntegralDigits=2, MaximumIntegralDigits=2, MinimumFractionalDigits=2, MaximumFractionalDigits=2))
   > n9 = "SetText (Text)"(
    self=Get CurveValue_Text() -> (CurveValue_Text:Common Text Object Reference)
    InText="To Text (String)"(InString=SelectString(A=- inf, B=@n8, bPickA=@n7))
   )
   Set LastVolumeRead(execute=[@n9.then, @n6.else], LastVolumeRead:double=Get MeterValue()) -> (Output_Get:double)
   Set SummedVolume(execute=@n2.LoopBody, SummedVolume=float + float(A=Get SummedVolume(), B=@n2))
  ]}
 SetMeterColorGradient {nodes [
   n1 = SetMeterColorGradient!() -> (
    then
    Volume:double=~
    BarReference:Object Reference
    Color_01:Linear Color Structure
    Color_02:Linear Color Structure
    Color_02Range_X:double
    Color_02Range_Y:double
    Color_03:Linear Color Structure
   ) {is=title K2Node_FunctionEntry}
   n2 = MapRangeUnclamped!(Value:double=@n1, InRangeA:double=@n1.Color_02Range_X, InRangeB:double=@n1.Color_02Range_Y, OutRangeA:double=-60, OutRangeB:double=6) -> (ReturnValue:double)
   n3 = MapRangeUnclamped(Value=@n1, InRangeA=@n1.Color_02Range_X, InRangeB=6, OutRangeA=-60, OutRangeB=6)
   > SetMeterValueColor!(
    execute
    self:Audio Meter Object Reference=Get AudioMeter_46()
    InValue:Linear Color Structure=SelectColor(
     A="Lerp (LinearColor)"(A=@n1.Color_02, B=@n1.Color_03, Alpha=@n3)
     B="Lerp (LinearColor)"(A=@n1.Color_01, B=@n1.Color_02, Alpha=@n2)
     bPickA="float > float"(A=@n1, B=@n1.Color_02Range_Y)
    )
   ) -> (then)
  ]}
]
variables [
 LastVolumeRead {type=number}
 MeterValue {type=number}
 PeakValue {type=number}
 SummedVolume {type=number}
]
widgets {
 info {rootWidgetClass=&CanvasPanel, widgetCount=7}
 widgets [
  CanvasPanel_0 {widgetClassPath=&CanvasPanel, widget=CanvasPanel_0}
  ScaleBox_508 {
   widgetClassPath=&ScaleBox
   parent=CanvasPanel_0
   widget=ScaleBox_508
   slot {
    class=CanvasPanelSlot
    properties {bAutoSize=true, layoutData {alignment=<0 0>, anchors {maximum=<1 1>, minimum=<0 0>}, offsets=<ltrb 0 0 0 0>}}
   }
  }
  CanvasPanel_600 {widgetClassPath=&CanvasPanel, parent=ScaleBox_508, widget=CanvasPanel_600, slot {class=ScaleBoxSlot}}
  Meter_Canvas {
   widgetClassPath=&CanvasPanel
   parent=CanvasPanel_600
   widget=Meter_Canvas
   slot {class=CanvasPanelSlot, properties {layoutData {alignment=<0 0>, anchors {maximum=<0 0>, minimum=<0 0>}, offsets=<ltrb 0 0 45 136.6733>}}}
  }
  CurveValue_Text {
   widgetClassPath=&CommonTextBlock
   bIsVariable=true
   parent=Meter_Canvas
   widget=CurveValue_Text
   properties {
    justification=Right
    style=&TextStyle-Base_C
    text=-inf
    font {
     bForceMonospaced=false
     bMaterialIsStencil=false
     fontMaterial=null
     fontObject=&NotoSans
     letterSpacing=0
     monospacedWidth=1
     size=24
     skewAmount=0
     typefaceFontName=NotoSans-Regular
     outlineSettings {
      bApplyOutlineToDropShadows=false
      bMiteredCorners=false
      bSeparateFillAlpha=false
      outlineMaterial=null
      outlineSize=0
      outlineColor=<rgba 0 0 0 1>
     }
    }
    renderTransform {angle=0, scale=<0.3 0.3>, shear=<0 0>, translation=<-34 -7>}
    shadowOffset=<0 0>
   }
   slot {
    class=CanvasPanelSlot
    properties {layoutData {alignment=<0 0>, anchors {maximum=<0 0>, minimum=<0 0>}, offsets=<ltrb 7.7039 143 93.0956 38.7447>}}
   }
  }
  CommonTextBlock_0 {
   widgetClassPath=&CommonTextBlock
   parent=Meter_Canvas
   widget=CommonTextBlock_0
   properties {
    style=&TextStyle-Base_C
    text=dB
    font {
     bForceMonospaced=false
     bMaterialIsStencil=false
     fontMaterial=null
     fontObject=&NotoSans
     letterSpacing=0
     monospacedWidth=1
     size=24
     skewAmount=0
     typefaceFontName=NotoSans-Regular
     outlineSettings {
      bApplyOutlineToDropShadows=false
      bMiteredCorners=false
      bSeparateFillAlpha=false
      outlineMaterial=null
      outlineSize=0
      outlineColor=<rgba 0 0 0 1>
     }
    }
    renderTransform {angle=0, scale=<0.2 0.2>, shear=<0 0>, translation=<0 0>}
    shadowOffset=<0 0>
   }
   slot {
    class=CanvasPanelSlot
    properties {layoutData {alignment=<0 0>, anchors {maximum=<0 0>, minimum=<0 0>}, offsets=<ltrb 19.1126 135 43.3904 44>}}
   }
  }
  AudioMeter_46 {
   widgetClassPath=/Script/AudioWidgets.AudioMeter
   bIsVariable=true
   parent=Meter_Canvas
   widget=AudioMeter_46
   properties {widgetStyle {
     bScaleSide=true
     bShowScale=true
     clippingValueWidth=2
     decibelsPerHash=6
     meterValuePadding=0
     peakValueWidth=1
     scaleHashHeight=5
     scaleHashOffset=5
     scaleHashWidth=1
     backgroundImage {
      bIsDynamicallyLoaded=false
      drawAs=Image
      imageType=NoImage
      mirroring=NoMirror
      resourceName=None
      resourceObject=null
      tiling=NoTile
      imageSize=<32 32>
      margin=<ltrb 0 0 0 0>
      outlineSettings {
       bUseBrushTransparency=false
       roundingType=HalfHeightRadius
       width=0
       color {colorUseRule=UseColor_Specified, specifiedColor=<rgba 0 0 0 0>}
       cornerRadii=<0 0 0 0>
      }
      tintColor {colorUseRule=UseColor_Specified, specifiedColor=<rgba 1 1 1 1>}
      uVRegion {bIsValid=false, max=<0 0>, min=<0 0>}
     }
     font {
      bForceMonospaced=false
      bMaterialIsStencil=false
      fontMaterial=null
      fontObject=&NotoSans
      letterSpacing=0
      monospacedWidth=1
      size=5
      skewAmount=0
      typefaceFontName=NotoSans-Regular
      outlineSettings {
       bApplyOutlineToDropShadows=false
       bMiteredCorners=false
       bSeparateFillAlpha=false
       outlineMaterial=null
       outlineSize=0
       outlineColor=<rgba 0 0 0 1>
      }
     }
     meterBackgroundImage {
      bIsDynamicallyLoaded=false
      drawAs=Image
      imageType=NoImage
      mirroring=NoMirror
      resourceName=None
      resourceObject=null
      tiling=NoTile
      imageSize=<32 32>
      margin=<ltrb 0 0 0 0>
      outlineSettings {
       bUseBrushTransparency=false
       roundingType=HalfHeightRadius
       width=0
       color {colorUseRule=UseColor_Specified, specifiedColor=<rgba 0 0 0 0>}
       cornerRadii=<0 0 0 0>
      }
      tintColor {colorUseRule=UseColor_Specified, specifiedColor=<rgba 1 1 1 1>}
      uVRegion {bIsValid=false, max=<0 0>, min=<0 0>}
     }
     meterClippingImage {
      bIsDynamicallyLoaded=false
      drawAs=Image
      imageType=NoImage
      mirroring=NoMirror
      resourceName=None
      resourceObject=null
      tiling=NoTile
      imageSize=<32 32>
      margin=<ltrb 0 0 0 0>
      outlineSettings {
       bUseBrushTransparency=false
       roundingType=HalfHeightRadius
       width=0
       color {colorUseRule=UseColor_Specified, specifiedColor=<rgba 0 0 0 0>}
       cornerRadii=<0 0 0 0>
      }
      tintColor {colorUseRule=UseColor_Specified, specifiedColor=<rgba 1 1 1 1>}
      uVRegion {bIsValid=false, max=<0 0>, min=<0 0>}
     }
     meterPadding=<5 5>
     meterPeakImage {
      bIsDynamicallyLoaded=false
      drawAs=Image
      imageType=NoImage
      mirroring=NoMirror
      resourceName=None
      resourceObject=null
      tiling=NoTile
      imageSize=<32 32>
      margin=<ltrb 0 0 0 0>
      outlineSettings {
       bUseBrushTransparency=false
       roundingType=HalfHeightRadius
       width=0
       color {colorUseRule=UseColor_Specified, specifiedColor=<rgba 0 0 0 0>}
       cornerRadii=<0 0 0 0>
      }
      tintColor {colorUseRule=UseColor_Specified, specifiedColor=<rgba 1 1 1 1>}
      uVRegion {bIsValid=false, max=<0 0>, min=<0 0>}
     }
     meterSize=<250 17>
     meterValueBackgroundImage {
      bIsDynamicallyLoaded=false
      drawAs=Image
      imageType=NoImage
      mirroring=NoMirror
      resourceName=None
      resourceObject=null
      tiling=NoTile
      imageSize=<32 32>
      margin=<ltrb 0 0 0 0>
      outlineSettings {
       bUseBrushTransparency=false
       roundingType=HalfHeightRadius
       width=0
       color {colorUseRule=UseColor_Specified, specifiedColor=<rgba 0 0 0 0>}
       cornerRadii=<0 0 0 0>
      }
      tintColor {colorUseRule=UseColor_Specified, specifiedColor=<rgba 1 1 1 1>}
      uVRegion {bIsValid=false, max=<0 0>, min=<0 0>}
     }
     meterValueImage {
      bIsDynamicallyLoaded=false
      drawAs=Image
      imageType=NoImage
      mirroring=NoMirror
      resourceName=None
      resourceObject=null
      tiling=NoTile
      imageSize=<32 32>
      margin=<ltrb 0 0 0 0>
      outlineSettings {
       bUseBrushTransparency=false
       roundingType=HalfHeightRadius
       width=0
       color {colorUseRule=UseColor_Specified, specifiedColor=<rgba 0 0 0 0>}
       cornerRadii=<0 0 0 0>
      }
      tintColor {colorUseRule=UseColor_Specified, specifiedColor=<rgba 1 1 1 1>}
      uVRegion {bIsValid=false, max=<0 0>, min=<0 0>}
     }
     valueRangeDb=<-60 6>
    }}
   slot {class=CanvasPanelSlot, properties {layoutData {alignment=<0 0>, anchors {maximum=<0 0>, minimum=<0 0>}, offsets=<ltrb 0 0 45 150>}}}
  }
 ]
}
