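# Editor utility widget that displays an eight-channel audio level meter labeled for 7.1 (L, R, C, LFE, Ls, Rs, Lr, Rr).
# ReadVolumes receives an envelope array from outside, stores it in Volumes and converts each value to decibels via 20·log10.
# SetChannelVolume writes each meter value, updating the peak only when the dB value exceeds that channel's stored linear volume.
# Results collect in MeterInfo, which is pushed to the AudioMeter after the loop.
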EditorUtilityWidgetBlueprint /Game/Sounds/Fort_Audio_Blueprints/Utilities/EditorUtilityWidgets/Audio_EUW_Multichannel_Meter
graphs [
 EventGraph {nodes [Event Construct()]}
 ReadVolumes {nodes [
   n1 = ReadVolumes!() -> (then, Envelope:"Array of Float (double-precision)s") {is=title K2Node_FunctionEntry}
   > n2 = Set Volumes(Volumes:"Array of Float (double-precision)s"=@n1) -> (Output_Get:"Array of Float (double-precision)s")
   > n3 = For Each Loop(Array:"Array of Float (double-precision)s"=@n2) -> (Array Element:double)
   SetMeterChannelInfo!(
    execute=@n3.Completed
    self:Audio Meter Object Reference=Get ChannelMeters() -> (ChannelMeters:Audio Meter Object Reference)
    InMeterChannelInfo:Array of Meter Channel Info Structures=Get MeterInfo() -> (MeterInfo:Array of Meter Channel Info Structures)
   ) -> (then)
   SetChannelVolume!(
    execute=@n3.LoopBody
    self
    ChannelVolume:double=float * float(A=Log!(A:double=@n3, Base:double=10) -> (ReturnValue:double), B=20)
    Index:int=@n3.Array Index
   ) -> (then)
  ]}
 SetChannelVolume {nodes [
   n1 = SetChannelVolume!() -> (then, ChannelVolume:double=~, Index:int) {is=title K2Node_FunctionEntry}
   n2 = Get ChannelMeters()
   > n3 = Set members in Meter Channel Info!(
    execute:Exec
    StructRef:"Meter Channel Info Structure (by ref)"="Get (a ref)"(
     Array:Array of Meter Channel Info Structures=GetMeterChannelInfo!(self:Audio Meter Object Reference=@n2) -> (ReturnValue:Array of Meter Channel Info Structures)
     Dimension 1=@n1.Index
    ) -> (Output:"Meter Channel Info Structure (by ref)")
    MeterValue:float=@n1
   ) -> (then:Exec, StructOut:"Meter Channel Info Structure (by ref)") {is=title K2Node_SetFieldsInStruct}
   n4 = Get Volumes() -> (Volumes:"Array of Float (double-precision)s")
   > n5 = Branch(Condition="float > float"(
     A=@n1
     B="Get (a ref)"(Array:"Array of Float (double-precision)s"=@n4, Dimension 1=@n1.Index) -> (Output:"Float (double-precision) (by ref)")
    ) {is=title K2Node_CallFunction})
   > n6 = Set members in Meter Channel Info!(execute, StructRef=@n3, PeakValue:float=@n1) -> (then, StructOut)
   Set Array Elem!(
    execute=[@n6.then, @n5.else]
    TargetArray:Array of Meter Channel Info Structures=Get MeterInfo()
    Index:int=@n1.Index
    Item:"Meter Channel Info Structure (by ref)"=@n6
    bSizeToFit:bool=true
   ) -> (then)
  ]}
]
variables [
 MeterInfo {type="array<MeterChannelInfo>", default []}
 Volumes {type="array<number>", default []}
]
widgets {
 info {rootWidgetClass=&CanvasPanel, widgetCount=14}
 widgets [
  CanvasPanel_0 {widgetClassPath=&CanvasPanel, widget=CanvasPanel_0}
  ScaleBox_508 {
   widgetClassPath=&ScaleBox
   parent=CanvasPanel_0
   widget=ScaleBox_508
   slot {
    class=CanvasPanelSlot
    properties {bAutoSize=true, layoutData {alignment=<0 0>, anchors {maximum=<1 1>, minimum=<0 0>}, offsets=<ltrb 0 0 0 0>}}
   }
  }
  CanvasPanel_600 {widgetClassPath=&CanvasPanel, parent=ScaleBox_508, widget=CanvasPanel_600, slot {class=ScaleBoxSlot}}
  Meter_Canvas {
   widgetClassPath=&CanvasPanel
   parent=CanvasPanel_600
   widget=Meter_Canvas
   slot {
    class=CanvasPanelSlot
    properties {layoutData {alignment=<0 0>, anchors {maximum=<0 0>, minimum=<0 0>}, offsets=<ltrb 35.5911 0 217.8475 140.7488>}}
   }
  }
  Numbers_Canvas {
   widgetClassPath=&CanvasPanel
   parent=Meter_Canvas
   widget=Numbers_Canvas
   slot {
    class=CanvasPanelSlot
    properties {layoutData {alignment=<0 0>, anchors {maximum=<0 0>, minimum=<0 0>}, offsets=<ltrb -36.1988 10.7371 365.3368 252.8657>}}
   }
  }
  CommonTextBlock_2 {
   widgetClassPath=&CommonTextBlock
   parent=Numbers_Canvas
   widget=CommonTextBlock_2
   properties {
    style=&TextStyle-Base_C
    text=L
    font {
     bForceMonospaced=false
     bMaterialIsStencil=false
     fontMaterial=null
     fontObject=&NotoSans
     letterSpacing=0
     monospacedWidth=1
     size=24
     skewAmount=0
     typefaceFontName=NotoSans-Regular
     outlineSettings {
      bApplyOutlineToDropShadows=false
      bMiteredCorners=false
      bSeparateFillAlpha=false
      outlineMaterial=null
      outlineSize=0
      outlineColor=<rgba 0 0 0 1>
     }
    }
    renderTransform {angle=0, scale=<0.35 0.35>, shear=<0 0>, translation=<0 0>}
    shadowOffset=<0 0>
   }
   slot {
    class=CanvasPanelSlot
    properties {layoutData {alignment=<0 0>, anchors {maximum=<0 0>, minimum=<0 0>}, offsets=<ltrb 31.7934 115 24.2999 44>}}
   }
  }
  CommonTextBlock_3 {
   widgetClassPath=&CommonTextBlock
   parent=Numbers_Canvas
   widget=CommonTextBlock_3
   properties {
    style=&TextStyle-Base_C
    text=C
    font {
     bForceMonospaced=false
     bMaterialIsStencil=false
     fontMaterial=null
     fontObject=&NotoSans
     letterSpacing=0
     monospacedWidth=1
     size=24
     skewAmount=0
     typefaceFontName=NotoSans-Regular
     outlineSettings {
      bApplyOutlineToDropShadows=false
      bMiteredCorners=false
      bSeparateFillAlpha=false
      outlineMaterial=null
      outlineSize=0
      outlineColor=<rgba 0 0 0 1>
     }
    }
    renderTransform {angle=0, scale=<0.35 0.35>, shear=<0 0>, translation=<0 0>}
    shadowOffset=<0 0>
   }
   slot {
    class=CanvasPanelSlot
    properties {layoutData {alignment=<0 0>, anchors {maximum=<0 0>, minimum=<0 0>}, offsets=<ltrb 86.9255 115 24.2999 44>}}
   }
  }
  CommonTextBlock_4 {
   widgetClassPath=&CommonTextBlock
   parent=Numbers_Canvas
   widget=CommonTextBlock_4
   properties {
    style=&TextStyle-Base_C
    text=R
    font {
     bForceMonospaced=false
     bMaterialIsStencil=false
     fontMaterial=null
     fontObject=&NotoSans
     letterSpacing=0
     monospacedWidth=1
     size=24
     skewAmount=0
     typefaceFontName=NotoSans-Regular
     outlineSettings {
      bApplyOutlineToDropShadows=false
      bMiteredCorners=false
      bSeparateFillAlpha=false
      outlineMaterial=null
      outlineSize=0
      outlineColor=<rgba 0 0 0 1>
     }
    }
    renderTransform {angle=0, scale=<0.35 0.35>, shear=<0 0>, translation=<0 0>}
    shadowOffset=<0 0>
   }
   slot {
    class=CanvasPanelSlot
    properties {layoutData {alignment=<0 0>, anchors {maximum=<0 0>, minimum=<0 0>}, offsets=<ltrb 59.0926 115 24.2999 44>}}
   }
  }
  CommonTextBlock_5 {
   widgetClassPath=&CommonTextBlock
   parent=Numbers_Canvas
   widget=CommonTextBlock_5
   properties {
    style=&TextStyle-Base_C
    text=Ls
    font {
     bForceMonospaced=false
     bMaterialIsStencil=false
     fontMaterial=null
     fontObject=&NotoSans
     letterSpacing=0
     monospacedWidth=1
     size=24
     skewAmount=0
     typefaceFontName=NotoSans-Regular
     outlineSettings {
      bApplyOutlineToDropShadows=false
      bMiteredCorners=false
      bSeparateFillAlpha=false
      outlineMaterial=null
      outlineSize=0
      outlineColor=<rgba 0 0 0 1>
     }
    }
    renderTransform {angle=0, scale=<0.35 0.35>, shear=<0 0>, translation=<0 0>}
    shadowOffset=<0 0>
   }
   slot {
    class=CanvasPanelSlot
    properties {layoutData {alignment=<0 0>, anchors {maximum=<0 0>, minimum=<0 0>}, offsets=<ltrb 142.6198 115 24.2999 44>}}
   }
  }
  CommonTextBlock_6 {
   widgetClassPath=&CommonTextBlock
   parent=Numbers_Canvas
   widget=CommonTextBlock_6
   properties {
    style=&TextStyle-Base_C
    text=Rs
    font {
     bForceMonospaced=false
     bMaterialIsStencil=false
     fontMaterial=null
     fontObject=&NotoSans
     letterSpacing=0
     monospacedWidth=1
     size=24
     skewAmount=0
     typefaceFontName=NotoSans-Regular
     outlineSettings {
      bApplyOutlineToDropShadows=false
      bMiteredCorners=false
      bSeparateFillAlpha=false
      outlineMaterial=null
      outlineSize=0
      outlineColor=<rgba 0 0 0 1>
     }
    }
    renderTransform {angle=0, scale=<0.35 0.35>, shear=<0 0>, translation=<0 0>}
    shadowOffset=<0 0>
   }
   slot {
    class=CanvasPanelSlot
    properties {layoutData {alignment=<0 0>, anchors {maximum=<0 0>, minimum=<0 0>}, offsets=<ltrb 169.7112 115 24.2999 44>}}
   }
  }
  CommonTextBlock_8 {
   widgetClassPath=&CommonTextBlock
   parent=Numbers_Canvas
   widget=CommonTextBlock_8
   properties {
    style=&TextStyle-Base_C
    text=Lr
    font {
     bForceMonospaced=false
     bMaterialIsStencil=false
     fontMaterial=null
     fontObject=&NotoSans
     letterSpacing=0
     monospacedWidth=1
     size=24
     skewAmount=0
     typefaceFontName=NotoSans-Regular
     outlineSettings {
      bApplyOutlineToDropShadows=false
      bMiteredCorners=false
      bSeparateFillAlpha=false
      outlineMaterial=null
      outlineSize=0
      outlineColor=<rgba 0 0 0 1>
     }
    }
    renderTransform {angle=0, scale=<0.35 0.35>, shear=<0 0>, translation=<0 0>}
    shadowOffset=<0 0>
   }
   slot {
    class=CanvasPanelSlot
    properties {layoutData {alignment=<0 0>, anchors {maximum=<0 0>, minimum=<0 0>}, offsets=<ltrb 198.9131 115 24.2999 44>}}
   }
  }
  CommonTextBlock_15 {
   widgetClassPath=&CommonTextBlock
   parent=Numbers_Canvas
   widget=CommonTextBlock_15
   properties {
    style=&TextStyle-Base_C
    text=Rr
    font {
     bForceMonospaced=false
     bMaterialIsStencil=false
     fontMaterial=null
     fontObject=&NotoSans
     letterSpacing=0
     monospacedWidth=1
     size=24
     skewAmount=0
     typefaceFontName=NotoSans-Regular
     outlineSettings {
      bApplyOutlineToDropShadows=false
      bMiteredCorners=false
      bSeparateFillAlpha=false
      outlineMaterial=null
      outlineSize=0
      outlineColor=<rgba 0 0 0 1>
     }
    }
    renderTransform {angle=0, scale=<0.35 0.35>, shear=<0 0>, translation=<0 0>}
    shadowOffset=<0 0>
   }
   slot {
    class=CanvasPanelSlot
    properties {layoutData {alignment=<0 0>, anchors {maximum=<0 0>, minimum=<0 0>}, offsets=<ltrb 226.6605 115 24.2999 44>}}
   }
  }
  CommonTextBlock_7 {
   widgetClassPath=&CommonTextBlock
   parent=Numbers_Canvas
   widget=CommonTextBlock_7
   properties {
    style=&TextStyle-Base_C
    text=LFE
    font {
     bForceMonospaced=false
     bMaterialIsStencil=false
     fontMaterial=null
     fontObject=&NotoSans
     letterSpacing=0
     monospacedWidth=1
     size=24
     skewAmount=0
     typefaceFontName=NotoSans-Regular
     outlineSettings {
      bApplyOutlineToDropShadows=false
      bMiteredCorners=false
      bSeparateFillAlpha=false
      outlineMaterial=null
      outlineSize=0
      outlineColor=<rgba 0 0 0 1>
     }
    }
    renderTransform {angle=0, scale=<0.35 0.35>, shear=<0 0>, translation=<0 0>}
    shadowOffset=<0 0>
   }
   slot {
    class=CanvasPanelSlot
    properties {layoutData {alignment=<0 0>, anchors {maximum=<0 0>, minimum=<0 0>}, offsets=<ltrb 109.6333 115 24.2999 44>}}
   }
  }
  ChannelMeters {
   widgetClassPath=/Script/AudioWidgets.AudioMeter
   bIsVariable=true
   parent=Meter_Canvas
   widget=ChannelMeters
   properties {
    meterChannelInfo [
     {clippingValue=0, meterValue=0, peakValue=0}
     {clippingValue=0, meterValue=0, peakValue=0}
     {clippingValue=0, meterValue=0, peakValue=0}
     {clippingValue=0, meterValue=0, peakValue=0}
     {clippingValue=0, meterValue=0, peakValue=0}
     {clippingValue=0, meterValue=0, peakValue=0}
     {clippingValue=0, meterValue=0, peakValue=0}
     {clippingValue=0, meterValue=0, peakValue=0}
    ]
    widgetStyle {
     bScaleSide=true
     bShowScale=true
     clippingValueWidth=2
     decibelsPerHash=6
     meterValuePadding=1
     peakValueWidth=1
     scaleHashHeight=10
     scaleHashOffset=5
     scaleHashWidth=1
     backgroundImage {
      bIsDynamicallyLoaded=false
      drawAs=Image
      imageType=NoImage
      mirroring=NoMirror
      resourceName=None
      resourceObject=null
      tiling=NoTile
      imageSize=<32 32>
      margin=<ltrb 0 0 0 0>
      outlineSettings {
       bUseBrushTransparency=false
       roundingType=HalfHeightRadius
       width=0
       color {colorUseRule=UseColor_Specified, specifiedColor=<rgba 0 0 0 0>}
       cornerRadii=<0 0 0 0>
      }
      tintColor {colorUseRule=UseColor_Specified, specifiedColor=<rgba 1 1 1 1>}
      uVRegion {bIsValid=false, max=<0 0>, min=<0 0>}
     }
     font {
      bForceMonospaced=false
      bMaterialIsStencil=false
      fontMaterial=null
      fontObject=&NotoSans
      letterSpacing=0
      monospacedWidth=1
      size=5
      skewAmount=0
      typefaceFontName=NotoSans-Regular
      outlineSettings {
       bApplyOutlineToDropShadows=false
       bMiteredCorners=false
       bSeparateFillAlpha=false
       outlineMaterial=null
       outlineSize=0
       outlineColor=<rgba 0 0 0 1>
      }
     }
     meterBackgroundImage {
      bIsDynamicallyLoaded=false
      drawAs=Image
      imageType=NoImage
      mirroring=NoMirror
      resourceName=None
      resourceObject=null
      tiling=NoTile
      imageSize=<32 32>
      margin=<ltrb 0 0 0 0>
      outlineSettings {
       bUseBrushTransparency=false
       roundingType=HalfHeightRadius
       width=0
       color {colorUseRule=UseColor_Specified, specifiedColor=<rgba 0 0 0 0>}
       cornerRadii=<0 0 0 0>
      }
      tintColor {colorUseRule=UseColor_Specified, specifiedColor=<rgba 1 1 1 1>}
      uVRegion {bIsValid=false, max=<0 0>, min=<0 0>}
     }
     meterClippingImage {
      bIsDynamicallyLoaded=false
      drawAs=Image
      imageType=NoImage
      mirroring=NoMirror
      resourceName=None
      resourceObject=null
      tiling=NoTile
      imageSize=<32 32>
      margin=<ltrb 0 0 0 0>
      outlineSettings {
       bUseBrushTransparency=false
       roundingType=HalfHeightRadius
       width=0
       color {colorUseRule=UseColor_Specified, specifiedColor=<rgba 0 0 0 0>}
       cornerRadii=<0 0 0 0>
      }
      tintColor {colorUseRule=UseColor_Specified, specifiedColor=<rgba 1 1 1 1>}
      uVRegion {bIsValid=false, max=<0 0>, min=<0 0>}
     }
     meterPadding=<5 7.5>
     meterPeakImage {
      bIsDynamicallyLoaded=false
      drawAs=Image
      imageType=NoImage
      mirroring=NoMirror
      resourceName=None
      resourceObject=null
      tiling=NoTile
      imageSize=<32 32>
      margin=<ltrb 0 0 0 0>
      outlineSettings {
       bUseBrushTransparency=false
       roundingType=HalfHeightRadius
       width=0
       color {colorUseRule=UseColor_Specified, specifiedColor=<rgba 0 0 0 0>}
       cornerRadii=<0 0 0 0>
      }
      tintColor {colorUseRule=UseColor_Specified, specifiedColor=<rgba 1 1 1 1>}
      uVRegion {bIsValid=false, max=<0 0>, min=<0 0>}
     }
     meterSize=<250 20>
     meterValueBackgroundImage {
      bIsDynamicallyLoaded=false
      drawAs=Image
      imageType=NoImage
      mirroring=NoMirror
      resourceName=None
      resourceObject=null
      tiling=NoTile
      imageSize=<32 32>
      margin=<ltrb 0 0 0 0>
      outlineSettings {
       bUseBrushTransparency=false
       roundingType=HalfHeightRadius
       width=0
       color {colorUseRule=UseColor_Specified, specifiedColor=<rgba 0 0 0 0>}
       cornerRadii=<0 0 0 0>
      }
      tintColor {colorUseRule=UseColor_Specified, specifiedColor=<rgba 1 1 1 1>}
      uVRegion {bIsValid=false, max=<0 0>, min=<0 0>}
     }
     meterValueImage {
      bIsDynamicallyLoaded=false
      drawAs=Image
      imageType=NoImage
      mirroring=NoMirror
      resourceName=None
      resourceObject=null
      tiling=NoTile
      imageSize=<32 32>
      margin=<ltrb 0 0 0 0>
      outlineSettings {
       bUseBrushTransparency=false
       roundingType=HalfHeightRadius
       width=0
       color {colorUseRule=UseColor_Specified, specifiedColor=<rgba 0 0 0 0>}
       cornerRadii=<0 0 0 0>
      }
      tintColor {colorUseRule=UseColor_Specified, specifiedColor=<rgba 1 1 1 1>}
      uVRegion {bIsValid=false, max=<0 0>, min=<0 0>}
     }
     valueRangeDb=<-60 6>
    }
   }
   slot {class=CanvasPanelSlot, properties {layoutData {alignment=<0 0>, anchors {maximum=<0 0>, minimum=<0 0>}, offsets=<ltrb -33.033 0 250 140>}}}
  }
 ]
}
