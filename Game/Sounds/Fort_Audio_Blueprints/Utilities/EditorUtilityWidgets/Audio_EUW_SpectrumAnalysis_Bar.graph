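# Editor utility widget for one bar of an audio spectrum analysis display, with an AudioMeter, progress bars and a text label.
# SetFrequencyAmplitude converts amplitude to decibels (20·log10) and writes it to the meter value.
# If that exceeds the stored Volume it also sets the meter's peak; otherwise it only stores it as Volume.
# SetMeterColorGradient, not called here, tints a given progress bar by blending three colors across volume ranges.
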EditorUtilityWidgetBlueprint /Game/Sounds/Fort_Audio_Blueprints/Utilities/EditorUtilityWidgets/Audio_EUW_SpectrumAnalysis_Bar
graphs [
 EventGraph {nodes [
   n1 = SetFrequencyAmplitude!() -> (OutputDelegate, then, Volume:double=~) {is=title K2Node_CustomEvent}
   n2 = Get AudioMeter_48() -> (AudioMeter_48:Audio Meter Object Reference)
   n3 = float * float(A=Log!(A:double=@n1.Volume, Base:double=10) -> (ReturnValue:double), B=20)
   > n4 = Set members in Meter Channel Info!(
    execute:Exec
    StructRef:"Meter Channel Info Structure (by ref)"="Get (a ref)"(
     Array:Array of Meter Channel Info Structures=GetMeterChannelInfo!(self:Audio Meter Object Reference=@n2) -> (ReturnValue:Array of Meter Channel Info Structures)
     Dimension 1=0
    ) -> (Output:"Meter Channel Info Structure (by ref)")
    MeterValue:float=@n3
   ) -> (then:Exec, StructOut:"Meter Channel Info Structure (by ref)") {is=title K2Node_SetFieldsInStruct}
   > SetMeterChannelInfo!(
    execute
    self:Audio Meter Object Reference=Get AudioMeter_48()
    InMeterChannelInfo:Array of Meter Channel Info Structures=Make Array("[0]":Meter Channel Info Structure=@n4) -> (Array:Array of Meter Channel Info Structures)
   ) -> (then)
   > n5 = Branch(Condition="float > float"(A=@n3, B=Get Volume() -> (Volume:double)) {is=title K2Node_CallFunction})
   > n6 = Set members in Meter Channel Info!(execute, StructRef=@n4, PeakValue:float=@n3) -> (then, StructOut)
   > SetMeterChannelInfo(self=Get AudioMeter_48(), InMeterChannelInfo=Make Array("[0]"=@n6))
   Set Volume(execute=@n5.else, Volume:double=@n3) -> (Output_Get:double)
   Event PreConstruct()
   Event Construct()
  ]}
 SetMeterColorGradient {nodes [
   n1 = SetMeterColorGradient!() -> (
    then
    Volume:double=~
    BarReference:Progress Bar Object Reference
    Color_01:Linear Color Structure
    Color_02:Linear Color Structure
    Color_02Range_X:double
    Color_02Range_Y:double
    Color_03:Linear Color Structure
   ) {is=title K2Node_FunctionEntry}
   n2 = MapRangeUnclamped!(Value:double=@n1, InRangeA:double=@n1.Color_02Range_X, InRangeB:double=@n1.Color_02Range_Y, OutRangeA:double=0, OutRangeB:double=1) -> (ReturnValue:double)
   n3 = MapRangeUnclamped(Value=@n1, InRangeA=@n1.Color_02Range_Y, InRangeB=1, OutRangeA=0, OutRangeB=1)
   > SetFillColorAndOpacity!(
    execute
    self:Progress Bar Object Reference=@n1.BarReference
    InColor:Linear Color Structure=SelectColor(
     A="Lerp (LinearColor)"(A=@n1.Color_02, B=@n1.Color_03, Alpha=@n3)
     B="Lerp (LinearColor)"(A=@n1.Color_01, B=@n1.Color_02, Alpha=@n2)
     bPickA="float > float"(A=@n1, B=@n1.Color_02Range_Y)
    )
   ) -> (then)
  ]}
]
variables [Volume {type=number}]
widgets {
 info {rootWidgetClass=&CanvasPanel, widgetCount=6}
 widgets [
  CanvasPanel_0 {widgetClassPath=&CanvasPanel, widget=CanvasPanel_0}
  ClipBar {
   widgetClassPath=/Script/UMG.ProgressBar
   bIsVariable=true
   parent=CanvasPanel_0
   widget=ClipBar
   properties {
    barFillType=BottomToTop
    percent=1
    fillColorAndOpacity=<rgba 1 0 0.0235 0>
    widgetStyle {
     enableFillAnimation=true
     backgroundImage {
      bIsDynamicallyLoaded=false
      drawAs=Image
      imageType=NoImage
      mirroring=NoMirror
      resourceName=None
      resourceObject=null
      tiling=NoTile
      imageSize=<0 0>
      margin=<ltrb 0.4167 0.4167 0.4167 0.4167>
      outlineSettings {
       bUseBrushTransparency=false
       roundingType=HalfHeightRadius
       width=0
       color {colorUseRule=UseColor_Specified, specifiedColor=<rgba 0 0 0 0>}
       cornerRadii=<0 0 0 0>
      }
      tintColor {colorUseRule=UseColor_Specified, specifiedColor=<rgba 1 1 1 0>}
      uVRegion {bIsValid=false, max=<0 0>, min=<0 0>}
     }
     fillImage {
      bIsDynamicallyLoaded=false
      drawAs=Image
      imageType=FullColor
      mirroring=NoMirror
      resourceName=../../../Engine/Content/Slate/Starship/CoreWidgets/ProgressBar/ProgressMarquee.png
      resourceObject=null
      tiling=Horizontal
      imageSize=<20 12>
      margin=<ltrb 0 0 0 0>
      outlineSettings {
       bUseBrushTransparency=false
       roundingType=HalfHeightRadius
       width=0
       color {colorUseRule=UseColor_Specified, specifiedColor=<rgba 0 0 0 0>}
       cornerRadii=<0 0 0 0>
      }
      tintColor {colorUseRule=UseColor_Specified, specifiedColor=<rgba 1 1 1 1>}
      uVRegion {bIsValid=false, max=<0 0>, min=<0 0>}
     }
     marqueeImage {
      bIsDynamicallyLoaded=false
      drawAs=Image
      imageType=FullColor
      mirroring=NoMirror
      resourceName=../../../Engine/Content/Slate/Starship/CoreWidgets/ProgressBar/ProgressMarquee.png
      resourceObject=null
      tiling=Horizontal
      imageSize=<20 12>
      margin=<ltrb 0 0 0 0>
      outlineSettings {
       bUseBrushTransparency=false
       roundingType=HalfHeightRadius
       width=0
       color {colorUseRule=UseColor_Specified, specifiedColor=<rgba 0 0 0 0>}
       cornerRadii=<0 0 0 0>
      }
      tintColor {colorUseRule=UseColor_Specified, specifiedColor=<rgba 0.0044 0.2384 1 1>}
      uVRegion {bIsValid=false, max=<0 0>, min=<0 0>}
     }
    }
   }
   slot {class=CanvasPanelSlot, properties {layoutData {alignment=<0 0>, anchors {maximum=<0 0>, minimum=<0 0>}, offsets=<ltrb 0 0 8 8.0573>}}}
  }
  ProgressBar_Sticky {
   widgetClassPath=&ProgressBar
   bIsVariable=true
   parent=CanvasPanel_0
   widget=ProgressBar_Sticky
   properties {
    barFillType=BottomToTop
    fillColorAndOpacity=<rgba 0 0.0902 0.1843 1>
    renderTransform {angle=0, scale=<1.25 1>, shear=<0 0>, translation=<0 0>}
    widgetStyle {
     enableFillAnimation=true
     backgroundImage {
      bIsDynamicallyLoaded=false
      drawAs=Image
      imageType=NoImage
      mirroring=NoMirror
      resourceName=None
      resourceObject=null
      tiling=NoTile
      imageSize=<0 0>
      margin=<ltrb 0.4167 0.4167 0.4167 0.4167>
      outlineSettings {
       bUseBrushTransparency=false
       roundingType=HalfHeightRadius
       width=0
       color {colorUseRule=UseColor_Specified, specifiedColor=<rgba 0 0 0 0>}
       cornerRadii=<0 0 0 0>
      }
      tintColor {colorUseRule=UseColor_Specified, specifiedColor=<rgba 1 1 1 0>}
      uVRegion {bIsValid=false, max=<0 0>, min=<0 0>}
     }
     fillImage {
      bIsDynamicallyLoaded=false
      drawAs=Image
      imageType=FullColor
      mirroring=NoMirror
      resourceName=../../../Engine/Content/Slate/Starship/CoreWidgets/ProgressBar/ProgressMarquee.png
      resourceObject=null
      tiling=Horizontal
      imageSize=<20 12>
      margin=<ltrb 0 0 0 0>
      outlineSettings {
       bUseBrushTransparency=false
       roundingType=HalfHeightRadius
       width=0
       color {colorUseRule=UseColor_Specified, specifiedColor=<rgba 0 0 0 0>}
       cornerRadii=<0 0 0 0>
      }
      tintColor {colorUseRule=UseColor_Specified, specifiedColor=<rgba 1 1 1 1>}
      uVRegion {bIsValid=false, max=<0 0>, min=<0 0>}
     }
     marqueeImage {
      bIsDynamicallyLoaded=false
      drawAs=Image
      imageType=FullColor
      mirroring=NoMirror
      resourceName=../../../Engine/Content/Slate/Starship/CoreWidgets/ProgressBar/ProgressMarquee.png
      resourceObject=null
      tiling=Horizontal
      imageSize=<20 12>
      margin=<ltrb 0 0 0 0>
      outlineSettings {
       bUseBrushTransparency=false
       roundingType=HalfHeightRadius
       width=0
       color {colorUseRule=UseColor_Specified, specifiedColor=<rgba 0 0 0 0>}
       cornerRadii=<0 0 0 0>
      }
      tintColor {colorUseRule=UseColor_Specified, specifiedColor=<rgba 0.0044 0.2384 1 1>}
      uVRegion {bIsValid=false, max=<0 0>, min=<0 0>}
     }
    }
   }
   slot {class=CanvasPanelSlot, properties {layoutData {alignment=<0 0>, anchors {maximum=<0 0>, minimum=<0 0>}, offsets=<ltrb 2 9.459 8 250>}}}
  }
  ProgressBar {
   widgetClassPath=&ProgressBar
   bIsVariable=true
   parent=CanvasPanel_0
   widget=ProgressBar
   properties {
    barFillType=BottomToTop
    fillColorAndOpacity=<rgba 0 0.498 1 1>
    renderTransform {angle=0, scale=<1.25 1>, shear=<0 0>, translation=<0 0>}
    widgetStyle {
     enableFillAnimation=true
     backgroundImage {
      bIsDynamicallyLoaded=false
      drawAs=Image
      imageType=NoImage
      mirroring=NoMirror
      resourceName=None
      resourceObject=null
      tiling=NoTile
      imageSize=<0 0>
      margin=<ltrb 0.4167 0.4167 0.4167 0.4167>
      outlineSettings {
       bUseBrushTransparency=false
       roundingType=HalfHeightRadius
       width=0
       color {colorUseRule=UseColor_Specified, specifiedColor=<rgba 0 0 0 0>}
       cornerRadii=<0 0 0 0>
      }
      tintColor {colorUseRule=UseColor_Specified, specifiedColor=<rgba 1 1 1 0>}
      uVRegion {bIsValid=false, max=<0 0>, min=<0 0>}
     }
     fillImage {
      bIsDynamicallyLoaded=false
      drawAs=Image
      imageType=FullColor
      mirroring=NoMirror
      resourceName=../../../Engine/Content/Slate/Starship/CoreWidgets/ProgressBar/ProgressMarquee.png
      resourceObject=null
      tiling=Horizontal
      imageSize=<20 12>
      margin=<ltrb 0 0 0 0>
      outlineSettings {
       bUseBrushTransparency=false
       roundingType=HalfHeightRadius
       width=0
       color {colorUseRule=UseColor_Specified, specifiedColor=<rgba 0 0 0 0>}
       cornerRadii=<0 0 0 0>
      }
      tintColor {colorUseRule=UseColor_Specified, specifiedColor=<rgba 1 1 1 1>}
      uVRegion {bIsValid=false, max=<0 0>, min=<0 0>}
     }
     marqueeImage {
      bIsDynamicallyLoaded=false
      drawAs=Image
      imageType=FullColor
      mirroring=NoMirror
      resourceName=../../../Engine/Content/Slate/Starship/CoreWidgets/ProgressBar/ProgressMarquee.png
      resourceObject=null
      tiling=Horizontal
      imageSize=<20 12>
      margin=<ltrb 0 0 0 0>
      outlineSettings {
       bUseBrushTransparency=false
       roundingType=HalfHeightRadius
       width=0
       color {colorUseRule=UseColor_Specified, specifiedColor=<rgba 0 0 0 0>}
       cornerRadii=<0 0 0 0>
      }
      tintColor {colorUseRule=UseColor_Specified, specifiedColor=<rgba 0.0044 0.2384 1 1>}
      uVRegion {bIsValid=false, max=<0 0>, min=<0 0>}
     }
    }
   }
   slot {class=CanvasPanelSlot, properties {layoutData {alignment=<0 0>, anchors {maximum=<0 0>, minimum=<0 0>}, offsets=<ltrb 2 9.459 8 250>}}}
  }
  CommonTextBlock_94 {
   widgetClassPath=&CommonTextBlock
   bIsVariable=true
   parent=CanvasPanel_0
   widget=CommonTextBlock_94
   properties {
    justification=Center
    style=&TextStyle-Base_C
    text=20k
    font {
     bForceMonospaced=false
     bMaterialIsStencil=false
     fontMaterial=null
     fontObject=&NotoSans
     letterSpacing=0
     monospacedWidth=1
     size=24
     skewAmount=0
     typefaceFontName=NotoSans-Regular
     outlineSettings {
      bApplyOutlineToDropShadows=false
      bMiteredCorners=false
      bSeparateFillAlpha=false
      outlineMaterial=null
      outlineSize=0
      outlineColor=<rgba 0 0 0 1>
     }
    }
    renderTransform {angle=0, scale=<0.5 0.5>, shear=<0 0>, translation=<0 0>}
    shadowOffset=<0 0>
   }
   slot {
    class=CanvasPanelSlot
    properties {layoutData {alignment=<0 0>, anchors {maximum=<0 0>, minimum=<0 0>}, offsets=<ltrb -17.869 251.8638 56.2736 35.775>}}
   }
  }
  AudioMeter_48 {
   widgetClassPath=/Script/AudioWidgets.AudioMeter
   bIsVariable=true
   parent=CanvasPanel_0
   widget=AudioMeter_48
   properties {
    meterChannelInfo [{clippingValue=0, meterValue=-6, peakValue=0}]
    widgetStyle {
     bScaleSide=true
     bShowScale=false
     clippingValueWidth=2
     decibelsPerHash=10
     meterValuePadding=0
     peakValueWidth=1
     scaleHashHeight=10
     scaleHashOffset=5
     scaleHashWidth=1
     backgroundImage {
      bIsDynamicallyLoaded=false
      drawAs=Image
      imageType=NoImage
      mirroring=NoMirror
      resourceName=None
      resourceObject=null
      tiling=NoTile
      imageSize=<32 32>
      margin=<ltrb 0 0 0 0>
      outlineSettings {
       bUseBrushTransparency=false
       roundingType=HalfHeightRadius
       width=0
       color {colorUseRule=UseColor_Specified, specifiedColor=<rgba 0 0 0 0>}
       cornerRadii=<0 0 0 0>
      }
      tintColor {colorUseRule=UseColor_Specified, specifiedColor=<rgba 1 1 1 1>}
      uVRegion {bIsValid=false, max=<0 0>, min=<0 0>}
     }
     font {
      bForceMonospaced=false
      bMaterialIsStencil=false
      fontMaterial=null
      fontObject=null
      letterSpacing=0
      monospacedWidth=1
      size=5
      skewAmount=0
      typefaceFontName=Regular
      outlineSettings {
       bApplyOutlineToDropShadows=false
       bMiteredCorners=false
       bSeparateFillAlpha=false
       outlineMaterial=null
       outlineSize=0
       outlineColor=<rgba 0 0 0 1>
      }
     }
     meterBackgroundImage {
      bIsDynamicallyLoaded=false
      drawAs=Image
      imageType=NoImage
      mirroring=NoMirror
      resourceName=None
      resourceObject=null
      tiling=NoTile
      imageSize=<32 32>
      margin=<ltrb 0 0 0 0>
      outlineSettings {
       bUseBrushTransparency=false
       roundingType=HalfHeightRadius
       width=0
       color {colorUseRule=UseColor_Specified, specifiedColor=<rgba 0 0 0 0>}
       cornerRadii=<0 0 0 0>
      }
      tintColor {colorUseRule=UseColor_Specified, specifiedColor=<rgba 1 1 1 1>}
      uVRegion {bIsValid=false, max=<0 0>, min=<0 0>}
     }
     meterClippingImage {
      bIsDynamicallyLoaded=false
      drawAs=Image
      imageType=NoImage
      mirroring=NoMirror
      resourceName=None
      resourceObject=null
      tiling=NoTile
      imageSize=<32 32>
      margin=<ltrb 0 0 0 0>
      outlineSettings {
       bUseBrushTransparency=false
       roundingType=HalfHeightRadius
       width=0
       color {colorUseRule=UseColor_Specified, specifiedColor=<rgba 0 0 0 0>}
       cornerRadii=<0 0 0 0>
      }
      tintColor {colorUseRule=UseColor_Specified, specifiedColor=<rgba 1 1 1 1>}
      uVRegion {bIsValid=false, max=<0 0>, min=<0 0>}
     }
     meterPadding=<0 0>
     meterPeakImage {
      bIsDynamicallyLoaded=false
      drawAs=Image
      imageType=NoImage
      mirroring=NoMirror
      resourceName=None
      resourceObject=null
      tiling=NoTile
      imageSize=<32 32>
      margin=<ltrb 0 0 0 0>
      outlineSettings {
       bUseBrushTransparency=false
       roundingType=HalfHeightRadius
       width=0
       color {colorUseRule=UseColor_Specified, specifiedColor=<rgba 0 0 0 0>}
       cornerRadii=<0 0 0 0>
      }
      tintColor {colorUseRule=UseColor_Specified, specifiedColor=<rgba 1 1 1 1>}
      uVRegion {bIsValid=false, max=<0 0>, min=<0 0>}
     }
     meterSize=<250 13>
     meterValueBackgroundImage {
      bIsDynamicallyLoaded=false
      drawAs=Image
      imageType=NoImage
      mirroring=NoMirror
      resourceName=None
      resourceObject=null
      tiling=NoTile
      imageSize=<32 32>
      margin=<ltrb 0 0 0 0>
      outlineSettings {
       bUseBrushTransparency=false
       roundingType=HalfHeightRadius
       width=0
       color {colorUseRule=UseColor_Specified, specifiedColor=<rgba 0 0 0 0>}
       cornerRadii=<0 0 0 0>
      }
      tintColor {colorUseRule=UseColor_Specified, specifiedColor=<rgba 1 1 1 1>}
      uVRegion {bIsValid=false, max=<0 0>, min=<0 0>}
     }
     meterValueImage {
      bIsDynamicallyLoaded=false
      drawAs=Image
      imageType=NoImage
      mirroring=NoMirror
      resourceName=None
      resourceObject=null
      tiling=NoTile
      imageSize=<32 32>
      margin=<ltrb 0 0 0 0>
      outlineSettings {
       bUseBrushTransparency=false
       roundingType=HalfHeightRadius
       width=0
       color {colorUseRule=UseColor_Specified, specifiedColor=<rgba 0 0 0 0>}
       cornerRadii=<0 0 0 0>
      }
      tintColor {colorUseRule=UseColor_Specified, specifiedColor=<rgba 1 1 1 1>}
      uVRegion {bIsValid=false, max=<0 0>, min=<0 0>}
     }
     valueRangeDb=<-60 0>
    }
   }
   slot {class=CanvasPanelSlot, properties {layoutData {alignment=<0 0>, anchors {maximum=<0 0>, minimum=<0 0>}, offsets=<ltrb 0 0 13 265>}}}
  }
 ]
}
