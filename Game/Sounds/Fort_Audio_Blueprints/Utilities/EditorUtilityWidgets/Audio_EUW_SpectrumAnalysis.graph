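# Editor utility widget that displays an audio frequency spectrum as a row of bars.
# PreConstruct fills a Dynamic Entry Box with NumEntries Audio_EUW_SpectrumAnalysis_Bar widgets, labelling them from FrequencyLabels, which names every third band from 20 to 20k.
# The ReadFrequencies event stores an incoming amplitude array and passes each value to the bar at the same index, skipping indices without a bar.
# Nothing in this graph calls ReadFrequencies or computes the spectrum.
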EditorUtilityWidgetBlueprint /Game/Sounds/Fort_Audio_Blueprints/Utilities/EditorUtilityWidgets/Audio_EUW_SpectrumAnalysis
graphs [EventGraph {nodes [
   n1 = ReadFrequencies!() -> (OutputDelegate, then, Data:"Array of Float (double-precision)s") {is=title K2Node_CustomEvent}
   > n2 = Set Data(Data:"Array of Float (double-precision)s"=@n1.Data) -> (Output_Get:"Array of Float (double-precision)s")
   > n3 = For Each Loop(Array:"Array of Float (double-precision)s"=@n2) -> (Array Element:double)
   n4 = GetAllEntries!(self:Dynamic Entry Box Base Object Reference=Get DynamicEntryBox_49() -> (DynamicEntryBox_49:Dynamic Entry Box Object Reference)) -> (ReturnValue:Array of User Widget Object References)
   > Branch(Condition=Is Valid Index!(TargetArray:Array of User Widget Object References=@n4, IndexToTest:int=@n3.Array Index) -> (ReturnValue:bool))
   > n5 = Cast To Audio_EUW_SpectrumAnalysis_Bar!(
    execute
    Object="Get (a copy)"(Array:Array of User Widget Object References=@n4, Dimension 1=@n3.Array Index) -> (Output:User Widget Object Reference)
   ) -> (then, CastFailed, AsAudio EUW Spectrum Analysis Bar:Audio EUW Spectrum Analysis Bar Object Reference) {is=title K2Node_DynamicCast}
   > SetFrequencyAmplitude!(execute, self:Audio EUW Spectrum Analysis Bar Object Reference=@n5, Volume:double=@n3) -> (then)
   Event PreConstruct()
   n6 = Get NumEntries() -> (NumEntries:int=0)
   > Branch(Condition="integer > integer"(A=@n6, B=0) {is=title K2Node_CallFunction})
   > n7 = For Loop(LastIndex=int - int(A=@n6, B=1) {is=title K2Node_CallFunction})
   > n8 = Create Entry!(execute, self:Dynamic Entry Box Object Reference=Get DynamicEntryBox_49()) -> (then, ReturnValue:User Widget Object Reference)
   > n9 = Cast To Audio_EUW_SpectrumAnalysis_Bar(Object=@n8)
   n10 = Get FrequencyLabels() -> (FrequencyLabels:Array of Strings)
   > "SetText (Text)"(
    self=Get CommonTextBlock_94!(self:Audio EUW Spectrum Analysis Bar Object Reference=@n9) -> (CommonTextBlock_94:Common Text Object Reference)
    InText="To Text (String)"(InString="Get (a copy)"(Array:Array of Strings=@n10, Dimension 1=@n7) -> (Output:String))
   )
   Event Tick()
  ]}]
variables [
 Data {type="array<number>", default []}
 FrequencyLabels {
  type="array<string>"
  default [20, "", "", 40, "", "", 80, "", "", 160, "", "", 300, "", "", 600, "", "", 1.2k, "", "", 2.4k, "", "", 5k, "", "", 10k, "", "", 20k]
 }
 NumEntries {type=integer, default=31}
]
widgets {
 info {rootWidgetClass=&CanvasPanel, widgetCount=5}
 widgets [
  CanvasPanel_0 {widgetClassPath=&CanvasPanel, widget=CanvasPanel_0}
  ScaleBox_93 {
   widgetClassPath=&ScaleBox
   parent=CanvasPanel_0
   widget=ScaleBox_93
   slot {
    class=CanvasPanelSlot
    properties {bAutoSize=true, layoutData {alignment=<0 0>, anchors {maximum=<1 1>, minimum=<0 0>}, offsets=<ltrb 0 0 -0.1031 1.9713>}}
   }
  }
  Border_133 {
   widgetClassPath=&Border
   parent=ScaleBox_93
   widget=Border_133
   properties {
    background {
     bIsDynamicallyLoaded=false
     drawAs=Image
     imageType=NoImage
     mirroring=NoMirror
     resourceName=None
     resourceObject=null
     tiling=NoTile
     imageSize=<32 32>
     margin=<ltrb 0 0 0 0>
     outlineSettings {
      bUseBrushTransparency=false
      roundingType=HalfHeightRadius
      width=0
      color {colorUseRule=UseColor_Specified, specifiedColor=<rgba 0 0 0 0>}
      cornerRadii=<0 0 0 0>
     }
     tintColor {colorUseRule=UseColor_Specified, specifiedColor=<rgba 1 1 1 0>}
     uVRegion {bIsValid=false, max=<0 0>, min=<0 0>}
    }
    desiredSizeScale=<0.8 1>
    padding=<ltrb 0 0 0 0>
   }
   slot {class=ScaleBoxSlot}
  }
  CanvasPanel_135 {
   widgetClassPath=&CanvasPanel
   parent=Border_133
   widget=CanvasPanel_135
   properties {renderTransform {angle=0, scale=<1 1>, shear=<0 0>, translation=<3 -6>}}
   slot {class=BorderSlot, properties {padding=<ltrb 0 0 0 0>}}
  }
  DynamicEntryBox_49 {
   widgetClassPath=&DynamicEntryBox
   bIsVariable=true
   parent=CanvasPanel_135
   widget=DynamicEntryBox_49
   properties {
    entryWidgetClass=&/Audio_EUW_SpectrumAnalysis_Bar.Audio_EUW_SpectrumAnalysis_Bar_C
    flowDirectionPreference=LeftToRight
    numDesignerPreviewEntries=0
    entrySpacing=<-59 0>
   }
   slot {class=CanvasPanelSlot, properties {layoutData {alignment=<0 0>, anchors {maximum=<0 0>, minimum=<0 0>}, offsets=<ltrb 7 12 454.1724 260>}}}
  }
 ]
}
