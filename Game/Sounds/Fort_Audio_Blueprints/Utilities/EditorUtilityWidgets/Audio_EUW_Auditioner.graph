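# Dockable editor utility widget for auditioning sounds and metering submixes during audio work.
# Audition plays SoundToAudition or selected assets, optionally at AuditionActor's location, with delayed looping.
# Envelope followers and spectral analysis on chosen and master submixes feed meter widgets, restarting when the editor/PIE world changes.
# Tick pushes parameter maps to active sounds and shows listener-to-actor distance; other controls stop sounds, open the test map and nudge asset volume via PageUp/PageDown.
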EditorUtilityWidgetBlueprint /Game/Sounds/Fort_Audio_Blueprints/Utilities/EditorUtilityWidgets/Audio_EUW_Auditioner
graphs [
 ApplyAudioParameters {nodes [
   ApplyAudioParameters() {is=title K2Node_FunctionEntry}
   > n1 = For Each Loop(Array:Array of Audio Component Object References=Get ActiveSounds() -> (ActiveSounds:Array of Audio Component Object References)) -> (Array Element:Audio Component Object Reference)
   n2 = Get FloatParams() -> (FloatParams:"Map of Names to Float (double-precision)s")
   > n3 = Keys!(execute, TargetMap:"Map of Names to Float (double-precision)s"=@n2) -> (then, Keys:Array of Names)
   > n4 = For Each Loop(Array:Array of Names=@n3) -> (Array Element:Name)
   n5 = Get IntParams() -> (IntParams:Map of Names to Integers)
   n6 = Keys(execute=@n4.Completed, TargetMap:Map of Names to Integers=@n5)
   > n7 = For Each Loop(Array=@n6)
   n8 = Find(TargetMap:Map of Names to Integers=@n5, Key:"Name (by ref)"=@n7) -> (Value:int)
   > Branch(Condition=@n8.ReturnValue)
   > Set Integer Parameter!(execute, self:Audio Parameter Controller Interface Interface=@n1, InName:Name=@n7, InInt:int=@n8) -> (then) {is=title K2Node_CallFunction}
   n9 = Find(TargetMap:"Map of Names to Float (double-precision)s"=@n2, Key=@n4) -> (Value:double)
   Branch(execute=@n4.LoopBody, Condition=@n9.ReturnValue)
   > Set Float Parameter!(execute, self:Audio Parameter Controller Interface Interface=@n1, InName:Name=@n4, InFloat:float=@n9) -> (then) {is=title K2Node_CallFunction}
  ]}
 Audition {nodes [
   Audition() {is=title K2Node_FunctionEntry}
   > n1 = For Each Loop(Array:Array of Audio Component Object References=Get ActiveSounds()) -> (Array Element:Audio Component Object Reference)
   > n2 = Branch(Condition=IsValid(Object=@n1))
   Remove Item!(
    execute=@n2.else
    TargetArray:Array of Audio Component Object References=Get ActiveSounds()
    Item:"Audio Component Object Reference (by ref)"=@n1
   ) -> (then, ReturnValue:bool)
   Set bSoundsStopped(execute=@n1.Completed, bSoundsStopped:bool=false) -> (Output_Get:bool)
   > n3 = Get!(execute:Exec) -> (then:Exec, else:Exec, SoundToAudition:Sound Base Object Reference)
   n4 = GetCurrentWorld!() -> (ReturnValue:World Object Reference)
   > n5 = CreateSound2D!(
    execute
    WorldContextObject:Object Reference=@n4
    Sound:Sound Base Object Reference=@n3
    VolumeMultiplier:float=1
    PitchMultiplier:float=1
    StartTime:float=0
    ConcurrencySettings:Sound Concurrency Object Reference
    bPersistAcrossLevelTransition:bool=false
    bAutoDestroy:bool=true
   ) -> (then, ReturnValue:Audio Component Object Reference)
   > n6 = Get!(execute) -> (then, else, AuditionActor:Actor Soft Object Reference)
   > Set bAllowSpatialization!(execute, bAllowSpatialization:bool=true, self:Audio Component Object Reference=@n5) -> (then, Output_Get:bool)
   > n7 = Set World Location!(
    execute
    self:Scene Component Object Reference=@n5
    NewLocation:Vector=Get Actor Location(self=Resolve Soft Reference!(Input:Actor Soft Object Reference=@n6) -> (Output:Actor Object Reference))
    bSweep:bool=false
    bTeleport:bool=false
   ) -> (then, SweepHitResult:Hit Result Structure) {is=title K2Node_CallFunction}
   Add(
    execute=[@n7.then, @n6.else]
    TargetArray:Array of Audio Component Object References=Get NewSoundsToPlay() -> (NewSoundsToPlay:Array of Audio Component Object References)
    NewItem:"Audio Component Object Reference (by ref)"=@n5
   ) {is=title K2Node_CallArrayFunction}
   > Play!(execute, self:Audio Component Object Reference=@n5, StartTime:float=0) -> (then)
   > Append Array!(
    execute
    TargetArray:Array of Audio Component Object References=Get ActiveSounds()
    SourceArray:Array of Audio Component Object References=Get NewSoundsToPlay()
   ) -> (then)
   n8 = GetSelectedAssets(execute=@n3.else)
   > n9 = For Each Loop(Array:Array of Object References=@n8) -> (Array Element:Object Reference)
   n10 = Get NewSoundsToPlay()
   n11 = For Each Loop(Exec=@n9.Completed, Array:Array of Audio Component Object References=@n10) -> (Array Element:Audio Component Object Reference)
   Append Array(execute=@n11.Completed, TargetArray=Get ActiveSounds(), SourceArray=@n10)
   Play(execute=@n11.LoopBody, self=@n11, StartTime=0)
   n12 = Cast To SoundBase!(execute=@n9.LoopBody, Object=@n9) -> (then, CastFailed, AsSound Base:Sound Base Object Reference) {is=title K2Node_DynamicCast}
   > n13 = CreateSound2D(
    WorldContextObject=@n4
    Sound=@n12
    VolumeMultiplier=1
    PitchMultiplier=1
    StartTime=0
    bPersistAcrossLevelTransition=false
    bAutoDestroy=true
   )
   > n14 = Get()
   > Set bAllowSpatialization(bAllowSpatialization=true, self=@n13)
   > n15 = Set World Location(self=@n13, NewLocation=Get Actor Location(self=Resolve Soft Reference(Input=@n14)), bSweep=false, bTeleport=false)
   Add(execute=[@n15.then, @n14.else], TargetArray=Get NewSoundsToPlay(), NewItem=@n13)
  ]}
 EventGraph {nodes [
   "On Property Changed (AnalysisSubmix_Property)"!() -> (OutputDelegate, then, PropertyName:Name=None) {is=title K2Node_ComponentBoundEvent}
   > StartAnalysis()
   "On Clicked (Audition_Button)"() {is=title K2Node_ComponentBoundEvent}
   > n1 = Audition() {is=title K2Node_CallFunction}
   Delay!(execute=[@n1.then, @n3.then], WorldContextObject:Object Reference, Duration=Get LoopDelayTime() -> (LoopDelayTime:double)) -> (then)
   n2 = Get bSoundsStopped() -> (bSoundsStopped:bool)
   > Branch(Condition=AND Boolean(
     A="Not Equal (Boolean)"!(A:bool=@n2, B:bool=true) -> (ReturnValue:bool) {is=title K2Node_CallFunction}
     B=Get bLoopAuditionSound() -> (bLoopAuditionSound:bool)
    ))
   > n3 = Audition()
   n4 = "On Check State Changed (Loop_Checkbox)"!() -> (OutputDelegate, then, bIsChecked:bool) {is=title K2Node_ComponentBoundEvent}
   > n5 = Set bLoopAuditionSound(bLoopAuditionSound:bool=@n4.bIsChecked) -> (Output_Get:bool)
   > n6 = Branch(Condition=@n5)
   n7 = GetEditorWorld(execute=@n6.else, self=Get UnrealEditorSubsystem())
   > Clear and Invalidate Timer by Handle!(execute, WorldContextObject:Object Reference=@n7, Handle=Get AuditionSoundHandle() -> (AuditionSoundHandle:Timer Handle Structure)) -> (then)
   "On Clicked (Testmap_Button)"() {is=title K2Node_ComponentBoundEvent}
   > n8 = GetCurrentLevelName!(execute, WorldContextObject:Object Reference, bRemovePrefixString:bool=true) -> (then, ReturnValue:String)
   n9 = Get LevelToLoad() -> (LevelToLoad:String)
   > n10 = Branch(Condition=Contains(SearchIn=@n9, Substring=@n8))
   "Open Level (by Name)"!(execute=@n10.else, WorldContextObject:Object Reference, LevelName:Name=String To Name(InString=@n9), bAbsolute:bool=true, Options:String) -> (then)
   "On Pressed (Audition_Button_1)"() {is=title K2Node_ComponentBoundEvent}
   > StopSounds()
   n11 = MasterSubmixEnvelopeFollower!() -> (OutputDelegate, then, Envelope:"Array of Float (single-precision)s") {is=title K2Node_CustomEvent}
   > ReadVolumes!(
    execute
    self:Audio EUW Meter Object Reference=Get Audio_EUW_Meter_0() -> (Audio_EUW_Meter_0:Audio EUW Meter Object Reference)
    Envelope:"Array of Float (double-precision)s"=@n11.Envelope
   ) -> (then)
   StartAnalysis() {is=title K2Node_CustomEvent}
   > n12 = Sequence()
   > StopEnveloper()
   InitEnveloper(execute=@n12.then_1)
   > StartEnveloper()
   n13 = Envelope Follower!() -> (OutputDelegate, then, Envelope:"Array of Float (single-precision)s") {is=title K2Node_CustomEvent}
   > ReadVolumes(self=Get Audio_EUW_Meter() -> (Audio_EUW_Meter:Audio EUW Meter Object Reference), Envelope=@n13.Envelope)
   > ReadVolumes(
    self:Audio EUW Multichannel Meter Object Reference=Get Audio_EUW_Multichannel_Meter() -> (Audio_EUW_Multichannel_Meter:Audio EUW Multichannel Meter Object Reference)
    Envelope=@n13.Envelope
   )
   n14 = SpectrumAnalysis!() -> (OutputDelegate, then, Magnitude:"Array of Float (single-precision)s") {is=title K2Node_CustomEvent}
   > ReadFrequencies!(
    execute
    self:Audio EUW Spectrum Analysis Object Reference=Get Audio_EUW_SpectrumAnalysis() -> (Audio_EUW_SpectrumAnalysis:Audio EUW Spectrum Analysis Object Reference)
    Data:"Array of Float (double-precision)s"=@n14.Magnitude
   ) -> (then)
   Event Tick()
   > n15 = Sequence!(execute) -> (then_0, then_1, then_2:Exec)
   > ApplyAudioParameters() {is=title K2Node_CallFunction}
   UpdateAnalysisWorld(execute=@n15.then_1)
   n16 = Get(execute=@n15.then_2)
   > n17 = Set AuditionDistanceActor(AuditionDistanceActor:Actor Object Reference=Resolve Soft Reference(Input=@n16)) -> (Output_Get:Actor Object Reference)
   n18 = GetPIEWorlds!(execute=[@n17.then, @n26.then], bIncludeDedicatedServer:bool=false) -> (then, ReturnValue:Array of World Object References)
   > n19 = Branch(Condition=Is Valid Index!(TargetArray:Array of World Object References=@n18, IndexToTest:int) -> (ReturnValue:bool))
   n20 = "Get (a copy)"(Array:Array of World Object References=@n18) -> (Output:World Object Reference)
   > n21 = Set ListenerLocation(ListenerLocation:Vector=Get Actor Location(self=GetPlayerPawn!(WorldContextObject:Object Reference=@n20, PlayerIndex:int=0) -> (ReturnValue:Pawn Object Reference))) -> (Output_Get:Vector)
   n22 = vector - vector!(
    A:Vector=Get ListenerLocation() -> (ListenerLocation:Vector)
    B:Vector=Get Actor Location(self=Get AuditionDistanceActor() -> (AuditionDistanceActor:Actor Object Reference))
   ) -> (ReturnValue:Vector) {is=title K2Node_CallFunction}
   "SetText (Text)"(
    execute=[@n24.then, @n21.then]
    self=Get Distance_Text() -> (Distance_Text:Common Text Object Reference)
    InText="To Text (Float)"(Value=Vector Length(A=@n22), MinimumIntegralDigits=1, MaximumIntegralDigits=324, MinimumFractionalDigits=0, MaximumFractionalDigits=0)
   )
   > SetSelectedLevelActors(
    self=Get EditorActorSubsystem()
    ActorsToSelect=Make Array("[0]":Actor Object Reference=Get AuditionDistanceActor()) -> (Array:Array of Actor Object References)
   )
   n23 = Get Editor Camera Info!(execute=@n19.else) -> (then, CameraLocation:Vector, ForwardDirection:Vector, FOV:float, ReturnValue:bool) {is=title K2Node_CallFunction}
   > n24 = Set ListenerLocation(ListenerLocation=@n23)
   n25 = Get!(execute=@n16.else) -> (then, else, AudioComponentActor:Actor Soft Object Reference)
   > n26 = Set AuditionDistanceActor(AuditionDistanceActor=Resolve Soft Reference(Input=@n25))
   Event Construct()
   > Set bIsFocusable(bIsFocusable:bool=true) -> (Output_Get:bool)
   > StartAnalysis() {is=title K2Node_CallFunction}
   Event OnMouseEnter()
   > SetKeyboardFocus(self:Widget Object Reference)
   Event PreConstruct()
   n27 = Get AnalysisSubmix_Property() -> (AnalysisSubmix_Property:Single Property View Object Reference)
   > SetObject(self=@n27, NewObject=Self-Reference())
   > SetPropertyName!(execute, self:Single Property View Object Reference=@n27, NewPropertyName:Name=SubmixToAnalyze) -> (then)
   n28 = Get AuditionActor_Property() -> (AuditionActor_Property:Single Property View Object Reference)
   n29 = Get SoundToAudition_Property() -> (SoundToAudition_Property:Single Property View Object Reference)
   > SetObject(self=[@n28, @n29], NewObject=Self-Reference())
   > SetPropertyName(self=@n29, NewPropertyName=SoundToAudition)
   > SetPropertyName(self=@n28, NewPropertyName=AuditionActor)
   "On Property Changed (SoundToAudition_Property)"!() -> (OutputDelegate, then, PropertyName:Name=None) {is=title K2Node_ComponentBoundEvent}
  ]}
 GetCurrentWorld {nodes [
   GetCurrentWorld() {is=title K2Node_FunctionEntry}
   > n1 = GetPIEWorlds(bIncludeDedicatedServer=false)
   > n2 = Branch(Condition=Is Valid Index(TargetArray=@n1))
   > Return Node!(execute, ReturnValue:World Object Reference="Get (a copy)"(Array=@n1))
   n3 = GetEditorWorld(execute=@n2.else, self=Get UnrealEditorSubsystem())
   > Is Valid(InputObject=@n3)
   > Return Node(ReturnValue=@n3)
  ]}
 InitEnveloper {nodes [
   InitEnveloper() {is=title K2Node_FunctionEntry}
   > Set PreviousSubmixToAnalyze(PreviousSubmixToAnalyze:Sound Submix Object Reference=Get SubmixToAnalyze() -> (SubmixToAnalyze:Sound Submix Object Reference)) -> (Output_Get:Sound Submix Object Reference)
   > n1 = Sequence!(execute) -> (then_0, then_1)
   n2 = GetCurrentWorld() {is=title K2Node_CallFunction}
   n3 = Get SubmixToAnalyze()
   > AddEnvelopeFollowerDelegate!(
    execute
    self:Sound Submix Object Reference=@n3
    WorldContextObject:Object Reference=@n2
    OnSubmixEnvelopeBP:"Delegate (by ref)"=Create Event()
   ) -> (then)
   > AddSpectralAnalysisDelegate!(
    execute
    self:Sound Submix Object Reference=@n3
    WorldContextObject:Object Reference=@n2
    InBandSettings:Array of Sound Submix Spectral Analysis Band Settings Structures=MakeFullSpectrumSpectralAnalysisBandSettings!(
     InNumBands:int=Get NumAnalysisBands() -> (NumAnalysisBands:int=0)
     InMinimumFrequency:float=20
     InMaximumFrequency:float=20000
     InAttackTimeMsec:int=10
     InReleaseTimeMsec:int=500
    ) -> (ReturnValue:Array of Sound Submix Spectral Analysis Band Settings Structures)
    OnSubmixSpectralAnalysisBP:"Delegate (by ref)"=Create Event()
    UpdateRate:float=30
    DecibelNoiseFloor:float=-60
    bDoNormalize:bool=true
    bDoAutoRange:bool=false
    AutoRangeAttackTime:float=0.1
    AutoRangeReleaseTime:float=60
   ) -> (then)
   AddEnvelopeFollowerDelegate(
    execute=@n1.then_1
    self=Get MasterSubmixToAnalyze() -> (MasterSubmixToAnalyze:Sound Submix Object Reference)
    WorldContextObject=@n2
    OnSubmixEnvelopeBP=Create Event()
   )
  ]}
 OnKeyDown {nodes [
   n1 = OnKeyDown!() -> (then, MyGeometry:Geometry Structure, InKeyEvent:Key Event Structure) {is=title K2Node_FunctionEntry}
   > n2 = Sequence()
   > n3 = Cast To SoundBaseAssetActions!(
    execute
    Object=GetAssetActionInstance!(ClassType:Audio Asset Action Base Class Reference=/Game/Sounds/Fort_Audio_Blueprints/Utilities/AssetActions/SoundBaseAssetActions.SoundBaseAssetActions_C) -> (Instance:Audio Asset Action Base Object Reference)
   ) -> (then, CastFailed, AsSound Base Asset Actions:Sound Base Asset Actions Object Reference) {is=title K2Node_DynamicCast}
   n4 = GetKey!(Input:"Key Event Structure (by ref)"=@n1.InKeyEvent) -> (ReturnValue:Key Structure)
   > n5 = Branch(Condition="Equal (Key)"!(A:Key Structure=@n4, B:Key Structure=PageUp) -> (ReturnValue:bool))
   > Adjust Volume in Decibels!(execute, self:Sound Base Asset Actions Object Reference=@n3, Decibel Amount:double=3, Rounded Result:bool=false) -> (then)
   Branch(execute=@n5.else, Condition="Equal (Key)"(A=@n4, B=PageDown))
   > Adjust Volume in Decibels(self=@n3, Decibel Amount=-3, Rounded Result=false)
   n6 = GetInputEventFromKeyEvent!(Event:"Key Event Structure (by ref)"=@n1.InKeyEvent) -> (ReturnValue:Input Event Structure)
   n7 = GetKey(Input=@n1.InKeyEvent)
   Branch(
    execute=@n2.then_1
    Condition=AND Boolean(A="Equal (Key)"(A=@n7, B=A), B=Is Shift Down!(Input:"Input Event Structure (by ref)"=@n6) -> (ReturnValue:bool))
   )
   > PrintString(InString=Hello)
   > Return Node(ReturnValue:Event Reply Structure=Make Event Reply!() -> (EventReply:Event Reply Structure) {is=title K2Node_MakeStruct})
  ]}
 SetSoundBaseVolume {nodes [
   n1 = SetSoundBaseVolume!() -> (then, Object:Object Reference, Multiplier:double=~) {is=title K2Node_FunctionEntry}
   > n2 = Cast To SoundCue!(execute, Object=@n1) -> (then, CastFailed, AsSound Cue:Sound Cue Object Reference) {is=title K2Node_DynamicCast}
   > SetEditorProperty(Object=@n2, PropertyName=VolumeMultiplier, PropertyValue:double=@n1.Multiplier)
   n3 = Cast To SoundWave!(execute=@n2.CastFailed, Object=@n1) -> (then, CastFailed, AsSound Wave:Sound Wave Object Reference) {is=title K2Node_DynamicCast}
   > SetEditorProperty(Object=@n3, PropertyName=Volume, PropertyValue=@n1.Multiplier)
  ]}
 SetVolumeBar {nodes [SetVolumeBar!() -> (then, Volume:double=~, VolumeBar:Progress Bar Object Reference) {is=title K2Node_FunctionEntry}]}
 StartEnveloper {nodes [
   StartEnveloper() {is=title K2Node_FunctionEntry}
   > n1 = Sequence()
   > n2 = Get!(execute) -> (then, else, PreviousSubmixToAnalyze:Sound Submix Object Reference)
   n3 = GetCurrentWorld()
   > StartEnvelopeFollowing!(execute, self:Sound Submix Object Reference=@n2, WorldContextObject:Object Reference=@n3) -> (then)
   > StartSpectralAnalysis!(
    execute
    self:Sound Submix Object Reference=@n2
    WorldContextObject:Object Reference=@n3
    FFTSize:EFFTSize Enum=DefaultSize
    InterpolationMethod:EFFTPeakInterpolationMethod Enum=Linear
    WindowType:EFFTWindowType Enum=Hann
    HopSize:float=0
    SpectrumType:EAudioSpectrumType Enum=MagnitudeSpectrum
   ) -> (then)
   n4 = Get!(execute=@n1.then_1) -> (then, else, MasterSubmixToAnalyze:Sound Submix Object Reference)
   > StartEnvelopeFollowing(self=@n4, WorldContextObject=@n3)
  ]}
 StopEnveloper {nodes [
   StopEnveloper() {is=title K2Node_FunctionEntry}
   > n1 = Sequence()
   > n2 = Get!(execute) -> (then, else, PreviousSubmixToAnalyze:Sound Submix Object Reference)
   n3 = GetCurrentWorld()
   > StopEnvelopeFollowing!(execute, self:Sound Submix Object Reference=@n2, WorldContextObject:Object Reference=@n3) -> (then)
   > StopSpectralAnalysis!(execute, self:Sound Submix Object Reference=@n2, WorldContextObject:Object Reference=@n3) -> (then)
   n4 = Get!(execute=@n1.then_1) -> (then, else, MasterSubmixToAnalyze:Sound Submix Object Reference)
   > StopEnvelopeFollowing(self=@n4, WorldContextObject=@n3)
  ]}
 StopSounds {nodes [
   StopSounds() {is=title K2Node_FunctionEntry}
   > n1 = For Each Loop(Array=Get ActiveSounds())
   Set bSoundsStopped(execute=@n1.Completed, bSoundsStopped=true)
   Stop(execute=@n1.LoopBody, self:Audio Component Object Reference=@n1)
  ]}
 UpdateAnalysisWorld {nodes [
   UpdateAnalysisWorld() {is=title K2Node_FunctionEntry}
   > Branch(Condition="Not Equal (Object)"!(A:Object Reference=GetCurrentWorld(), B:Object Reference=Get CurrentWorld() -> (CurrentWorld:Object Reference)) -> (ReturnValue:bool))
   > StopEnveloper() {is=title K2Node_CallFunction}
   > InitEnveloper() {is=title K2Node_CallFunction}
   > StartEnveloper() {is=title K2Node_CallFunction}
   > Set CurrentWorld(CurrentWorld:Object Reference=GetCurrentWorld()) -> (Output_Get:Object Reference)
  ]}
]
variables [
 ActiveSounds {type="array<AudioComponent>", default []}
 AudioComponentActor {type=Actor, default=null}
 AuditionActor {type=Actor, default=null}
 AuditionDistanceActor {type=Actor, default=null}
 AuditionSoundHandle {type=TimerHandle}
 bLoopAuditionSound {type=boolean, default=true}
 bSoundsStopped {type=boolean}
 CurrentWorld {type=Object, default=null}
 FloatParams {type="map<number>", default {notherThing=0.1, thing=0.7}}
 IntParams {type="map<integer>", default {}}
 LevelToLoad {type=string, default=World'/Game/Athena/Maps/Test/BR_Audio'}
 ListenerLocation {type=Vector}
 LoopDelayTime {type=number, default=1}
 MasterSubmixToAnalyze {type=SoundSubmix, default=/Game/Sounds/SubMixes/Fortnite_Final}
 NumAnalysisBands {type=integer, default=31}
 PreviousSubmixToAnalyze {type=SoundSubmix, default=null}
 SoundToAudition {type=SoundBase, default=null}
 SubmixToAnalyze {type=SoundSubmix, default=&Fortnite_SFX}
]
widgets {
 info {
  parentClass=&/Audio_EUW_Dockable.Audio_EUW_Dockable_C
  rootWidgetClass=&CanvasPanel
  widgetCount=29
 }
 widgets [
  CanvasPanel_0 {widgetClassPath=&CanvasPanel, bIsVariable=true, widget=CanvasPanel_0}
  ScaleBox_155 {
   widgetClassPath=&ScaleBox
   parent=CanvasPanel_0
   widget=ScaleBox_155
   properties {stretch=Fill}
   slot {
    class=CanvasPanelSlot
    properties {bAutoSize=true, layoutData {alignment=<0 0>, anchors {maximum=<1 1>, minimum=<0 0>}, offsets=<ltrb 0 0 0 0>}}
   }
  }
  VerticalBox_73 {widgetClassPath=&VerticalBox, parent=ScaleBox_155, widget=VerticalBox_73, slot {class=ScaleBoxSlot}}
  CanvasPanel_213 {
   widgetClassPath=&CanvasPanel
   parent=VerticalBox_73
   widget=CanvasPanel_213
   slot {class=VerticalBoxSlot, properties {horizontalAlignment=HAlign_Center, verticalAlignment=VAlign_Center, size {sizeRule=Fill, value=1}}}
  }
  Audition_Button {
   widgetClassPath=&Button
   parent=CanvasPanel_213
   widget=Audition_Button
   properties {
    toolTipText="Audition the currently selected USoundBase asset (don't use for looping assets yet!)"
    backgroundColor=<rgba 0.2516 1 0.3114 1>
    widgetStyle {
     clickedSlateSound {resourceObject=null}
     disabled {
      bIsDynamicallyLoaded=false
      drawAs=RoundedBox
      imageType=NoImage
      mirroring=NoMirror
      resourceName=None
      resourceObject=null
      tiling=NoTile
      imageSize=<0 0>
      margin=<ltrb 0 0 0 0>
      outlineSettings {
       bUseBrushTransparency=false
       roundingType=FixedRadius
       width=1
       color {colorUseRule=UseColor_ColorTable, specifiedColor=<rgba 0 0 0 0>}
       cornerRadii=<4 4 4 4>
      }
      tintColor {colorUseRule=UseColor_Specified, specifiedColor=<rgba 0.0395 0.0395 0.0395 1>}
      uVRegion {bIsValid=false, max=<0 0>, min=<0 0>}
     }
     disabledForeground {colorUseRule=UseColor_ColorTable, specifiedColor=<rgba 0 0 0 0>}
     hovered {
      bIsDynamicallyLoaded=false
      drawAs=RoundedBox
      imageType=NoImage
      mirroring=NoMirror
      resourceName=None
      resourceObject=null
      tiling=NoTile
      imageSize=<0 0>
      margin=<ltrb 0 0 0 0>
      outlineSettings {
       bUseBrushTransparency=false
       roundingType=FixedRadius
       width=1
       color {colorUseRule=UseColor_ColorTable, specifiedColor=<rgba 0 0 0 0>}
       cornerRadii=<4 4 4 4>
      }
      tintColor {colorUseRule=UseColor_Specified, specifiedColor=<rgba 0.7188 0.2417 0 1>}
      uVRegion {bIsValid=false, max=<0 0>, min=<0 0>}
     }
     hoveredForeground {colorUseRule=UseColor_ColorTable, specifiedColor=<rgba 0 0 0 0>}
     hoveredSlateSound {resourceObject=null}
     normal {
      bIsDynamicallyLoaded=false
      drawAs=RoundedBox
      imageType=NoImage
      mirroring=NoMirror
      resourceName=None
      resourceObject=null
      tiling=NoTile
      imageSize=<0 0>
      margin=<ltrb 0 0 0 0>
      outlineSettings {
       bUseBrushTransparency=false
       roundingType=FixedRadius
       width=1
       color {colorUseRule=UseColor_ColorTable, specifiedColor=<rgba 0 0 0 0>}
       cornerRadii=<4 4 4 4>
      }
      tintColor {colorUseRule=UseColor_Specified, specifiedColor=<rgba 0.2396 0.2396 0.2396 1>}
      uVRegion {bIsValid=false, max=<0 0>, min=<0 0>}
     }
     normalForeground {colorUseRule=UseColor_ColorTable, specifiedColor=<rgba 0 0 0 0>}
     normalPadding=<ltrb 12 1.5 12 1.5>
     pressed {
      bIsDynamicallyLoaded=false
      drawAs=RoundedBox
      imageType=NoImage
      mirroring=NoMirror
      resourceName=None
      resourceObject=null
      tiling=NoTile
      imageSize=<0 0>
      margin=<ltrb 0 0 0 0>
      outlineSettings {
       bUseBrushTransparency=false
       roundingType=FixedRadius
       width=1
       color {colorUseRule=UseColor_ColorTable, specifiedColor=<rgba 0 0 0 0>}
       cornerRadii=<4 4 4 4>
      }
      tintColor {colorUseRule=UseColor_Specified, specifiedColor=<rgba 0.0284 0.0284 0.0284 1>}
      uVRegion {bIsValid=false, max=<0 0>, min=<0 0>}
     }
     pressedForeground {colorUseRule=UseColor_ColorTable, specifiedColor=<rgba 0 0 0 0>}
     pressedPadding=<ltrb 12 2 12 0>
     pressedSlateSound {resourceObject=null}
    }
   }
   slot {
    class=CanvasPanelSlot
    properties {layoutData {alignment=<0 0>, anchors {maximum=<0 0>, minimum=<0 0>}, offsets=<ltrb 3.003 139.7898 77.8636 38.5843>}}
   }
  }
  ScaleBox_183 {widgetClassPath=&ScaleBox, parent=Audition_Button, widget=ScaleBox_183, slot {class=ButtonSlot}}
  AuditionButtonIcon {
   widgetClassPath=&Image
   bIsVariable=true
   parent=ScaleBox_183
   widget=AuditionButtonIcon
   properties {brush {
     bIsDynamicallyLoaded=false
     drawAs=Image
     imageType=NoImage
     mirroring=NoMirror
     resourceName=None
     resourceObject=/Game/UI/Foundation/Textures/Icons/Audio/T-Icon-Sound-On-M
     tiling=NoTile
     imageSize=<32 32>
     margin=<ltrb 0 0 0 0>
     outlineSettings {
      bUseBrushTransparency=false
      roundingType=HalfHeightRadius
      width=0
      color {colorUseRule=UseColor_Specified, specifiedColor=<rgba 0 0 0 0>}
      cornerRadii=<0 0 0 0>
     }
     tintColor {colorUseRule=UseColor_Specified, specifiedColor=<rgba 0.8021 0.8021 0.8021 1>}
     uVRegion {bIsValid=false, max=<0 0>, min=<0 0>}
    }}
   slot {class=ScaleBoxSlot}
  }
  Audition_Button_1 {
   widgetClassPath=&Button
   parent=CanvasPanel_213
   widget=Audition_Button_1
   properties {
    toolTipText="Audition the currently selected USoundBase asset (don't use for looping assets yet!)"
    backgroundColor=<rgba 1 0.312 0.312 1>
    widgetStyle {
     clickedSlateSound {resourceObject=null}
     disabled {
      bIsDynamicallyLoaded=false
      drawAs=RoundedBox
      imageType=NoImage
      mirroring=NoMirror
      resourceName=None
      resourceObject=null
      tiling=NoTile
      imageSize=<0 0>
      margin=<ltrb 0 0 0 0>
      outlineSettings {
       bUseBrushTransparency=false
       roundingType=FixedRadius
       width=1
       color {colorUseRule=UseColor_ColorTable, specifiedColor=<rgba 0 0 0 0>}
       cornerRadii=<4 4 4 4>
      }
      tintColor {colorUseRule=UseColor_Specified, specifiedColor=<rgba 0.0395 0.0395 0.0395 1>}
      uVRegion {bIsValid=false, max=<0 0>, min=<0 0>}
     }
     disabledForeground {colorUseRule=UseColor_ColorTable, specifiedColor=<rgba 0 0 0 0>}
     hovered {
      bIsDynamicallyLoaded=false
      drawAs=RoundedBox
      imageType=NoImage
      mirroring=NoMirror
      resourceName=None
      resourceObject=null
      tiling=NoTile
      imageSize=<0 0>
      margin=<ltrb 0 0 0 0>
      outlineSettings {
       bUseBrushTransparency=false
       roundingType=FixedRadius
       width=1
       color {colorUseRule=UseColor_ColorTable, specifiedColor=<rgba 0 0 0 0>}
       cornerRadii=<4 4 4 4>
      }
      tintColor {colorUseRule=UseColor_Specified, specifiedColor=<rgba 0.7188 0.2417 0 1>}
      uVRegion {bIsValid=false, max=<0 0>, min=<0 0>}
     }
     hoveredForeground {colorUseRule=UseColor_ColorTable, specifiedColor=<rgba 0 0 0 0>}
     hoveredSlateSound {resourceObject=null}
     normal {
      bIsDynamicallyLoaded=false
      drawAs=RoundedBox
      imageType=NoImage
      mirroring=NoMirror
      resourceName=None
      resourceObject=null
      tiling=NoTile
      imageSize=<0 0>
      margin=<ltrb 0 0 0 0>
      outlineSettings {
       bUseBrushTransparency=false
       roundingType=FixedRadius
       width=1
       color {colorUseRule=UseColor_ColorTable, specifiedColor=<rgba 0 0 0 0>}
       cornerRadii=<4 4 4 4>
      }
      tintColor {colorUseRule=UseColor_Specified, specifiedColor=<rgba 0.2396 0.2396 0.2396 1>}
      uVRegion {bIsValid=false, max=<0 0>, min=<0 0>}
     }
     normalForeground {colorUseRule=UseColor_ColorTable, specifiedColor=<rgba 0 0 0 0>}
     normalPadding=<ltrb 12 1.5 12 1.5>
     pressed {
      bIsDynamicallyLoaded=false
      drawAs=RoundedBox
      imageType=NoImage
      mirroring=NoMirror
      resourceName=None
      resourceObject=null
      tiling=NoTile
      imageSize=<0 0>
      margin=<ltrb 0 0 0 0>
      outlineSettings {
       bUseBrushTransparency=false
       roundingType=FixedRadius
       width=1
       color {colorUseRule=UseColor_ColorTable, specifiedColor=<rgba 0 0 0 0>}
       cornerRadii=<4 4 4 4>
      }
      tintColor {colorUseRule=UseColor_Specified, specifiedColor=<rgba 0.0284 0.0284 0.0284 1>}
      uVRegion {bIsValid=false, max=<0 0>, min=<0 0>}
     }
     pressedForeground {colorUseRule=UseColor_ColorTable, specifiedColor=<rgba 0 0 0 0>}
     pressedPadding=<ltrb 12 2 12 0>
     pressedSlateSound {resourceObject=null}
    }
   }
   slot {
    class=CanvasPanelSlot
    properties {layoutData {alignment=<0 0>, anchors {maximum=<0 0>, minimum=<0 0>}, offsets=<ltrb 87.003 139.7898 77.8636 38.5843>}}
   }
  }
  ScaleBox_2 {widgetClassPath=&ScaleBox, parent=Audition_Button_1, widget=ScaleBox_2, slot {class=ButtonSlot}}
  AuditionButtonIcon_2 {
   widgetClassPath=&Image
   bIsVariable=true
   parent=ScaleBox_2
   widget=AuditionButtonIcon_2
   properties {brush {
     bIsDynamicallyLoaded=false
     drawAs=Image
     imageType=NoImage
     mirroring=NoMirror
     resourceName=None
     resourceObject=/Game/UI/Foundation/Textures/Icons/Audio/T-Icon-Sound-Mute-M
     tiling=NoTile
     imageSize=<32 32>
     margin=<ltrb 0 0 0 0>
     outlineSettings {
      bUseBrushTransparency=false
      roundingType=HalfHeightRadius
      width=0
      color {colorUseRule=UseColor_Specified, specifiedColor=<rgba 0 0 0 0>}
      cornerRadii=<0 0 0 0>
     }
     tintColor {colorUseRule=UseColor_Specified, specifiedColor=<rgba 0.8021 0.8021 0.8021 1>}
     uVRegion {bIsValid=false, max=<0 0>, min=<0 0>}
    }}
   slot {class=ScaleBoxSlot}
  }
  Testmap_Button {
   widgetClassPath=&Button
   parent=CanvasPanel_213
   widget=Testmap_Button
   properties {widgetStyle {
     clickedSlateSound {resourceObject=null}
     disabled {
      bIsDynamicallyLoaded=false
      drawAs=RoundedBox
      imageType=NoImage
      mirroring=NoMirror
      resourceName=None
      resourceObject=null
      tiling=NoTile
      imageSize=<0 0>
      margin=<ltrb 0 0 0 0>
      outlineSettings {
       bUseBrushTransparency=false
       roundingType=FixedRadius
       width=1
       color {colorUseRule=UseColor_ColorTable, specifiedColor=<rgba 0 0 0 0>}
       cornerRadii=<4 4 4 4>
      }
      tintColor {colorUseRule=UseColor_Specified, specifiedColor=<rgba 0.0395 0.0395 0.0395 1>}
      uVRegion {bIsValid=false, max=<0 0>, min=<0 0>}
     }
     disabledForeground {colorUseRule=UseColor_ColorTable, specifiedColor=<rgba 0 0 0 0>}
     hovered {
      bIsDynamicallyLoaded=false
      drawAs=RoundedBox
      imageType=NoImage
      mirroring=NoMirror
      resourceName=None
      resourceObject=null
      tiling=NoTile
      imageSize=<0 0>
      margin=<ltrb 0 0 0 0>
      outlineSettings {
       bUseBrushTransparency=false
       roundingType=FixedRadius
       width=1
       color {colorUseRule=UseColor_ColorTable, specifiedColor=<rgba 0 0 0 0>}
       cornerRadii=<4 4 4 4>
      }
      tintColor {colorUseRule=UseColor_Specified, specifiedColor=<rgba 0.7188 0.2417 0 1>}
      uVRegion {bIsValid=false, max=<0 0>, min=<0 0>}
     }
     hoveredForeground {colorUseRule=UseColor_ColorTable, specifiedColor=<rgba 0 0 0 0>}
     hoveredSlateSound {resourceObject=null}
     normal {
      bIsDynamicallyLoaded=false
      drawAs=RoundedBox
      imageType=NoImage
      mirroring=NoMirror
      resourceName=None
      resourceObject=null
      tiling=NoTile
      imageSize=<0 0>
      margin=<ltrb 0 0 0 0>
      outlineSettings {
       bUseBrushTransparency=false
       roundingType=FixedRadius
       width=1
       color {colorUseRule=UseColor_ColorTable, specifiedColor=<rgba 0 0 0 0>}
       cornerRadii=<4 4 4 4>
      }
      tintColor {colorUseRule=UseColor_Specified, specifiedColor=<rgba 0.2396 0.2396 0.2396 1>}
      uVRegion {bIsValid=false, max=<0 0>, min=<0 0>}
     }
     normalForeground {colorUseRule=UseColor_ColorTable, specifiedColor=<rgba 0 0 0 0>}
     normalPadding=<ltrb 12 1.5 12 1.5>
     pressed {
      bIsDynamicallyLoaded=false
      drawAs=RoundedBox
      imageType=NoImage
      mirroring=NoMirror
      resourceName=None
      resourceObject=null
      tiling=NoTile
      imageSize=<0 0>
      margin=<ltrb 0 0 0 0>
      outlineSettings {
       bUseBrushTransparency=false
       roundingType=FixedRadius
       width=1
       color {colorUseRule=UseColor_ColorTable, specifiedColor=<rgba 0 0 0 0>}
       cornerRadii=<4 4 4 4>
      }
      tintColor {colorUseRule=UseColor_Specified, specifiedColor=<rgba 0.0284 0.0284 0.0284 1>}
      uVRegion {bIsValid=false, max=<0 0>, min=<0 0>}
     }
     pressedForeground {colorUseRule=UseColor_ColorTable, specifiedColor=<rgba 0 0 0 0>}
     pressedPadding=<ltrb 12 2 12 0>
     pressedSlateSound {resourceObject=null}
    }}
   slot {
    class=CanvasPanelSlot
    properties {layoutData {alignment=<0 0>, anchors {maximum=<0 0>, minimum=<0 0>}, offsets=<ltrb 171.003 139.7898 77.8636 38.5843>}}
   }
  }
  ScaleBox {widgetClassPath=&ScaleBox, parent=Testmap_Button, widget=ScaleBox, slot {class=ButtonSlot}}
  CommonTextBlock {
   widgetClassPath=&CommonTextBlock
   parent=ScaleBox
   widget=CommonTextBlock
   properties {
    justification=Center
    style=/FortUILibrary/Styles/Typography/Foundation/TextStyle-Base-M-B_Black.TextStyle-Base-M-B_Black_C
    text=BR Audio
    colorAndOpacity {colorUseRule=UseColor_Specified, specifiedColor=<rgba 0 0 0 1>}
    font {
     bForceMonospaced=false
     bMaterialIsStencil=false
     fontMaterial=null
     fontObject=&NotoSans
     letterSpacing=0
     monospacedWidth=1
     size=24
     skewAmount=0
     typefaceFontName=NotoSans-Bold
     outlineSettings {
      bApplyOutlineToDropShadows=false
      bMiteredCorners=false
      bSeparateFillAlpha=false
      outlineMaterial=null
      outlineSize=0
      outlineColor=<rgba 0 0 0 1>
     }
    }
    shadowOffset=<0 0>
   }
   slot {class=ScaleBoxSlot}
  }
  CanvasPanel_212 {widgetClassPath=&CanvasPanel, parent=CanvasPanel_213, widget=CanvasPanel_212, slot {class=CanvasPanelSlot, properties=~}}
  Loop_Checkbox {
   widgetClassPath=/Script/UMG.CheckBox
   bIsVariable=true
   parent=CanvasPanel_213
   widget=Loop_Checkbox
   properties {
    renderTransform {angle=0, scale=<0.8 0.8>, shear=<0 0>, translation=<0 0>}
    widgetStyle {
     checkBoxType=CheckBox
     backgroundHoveredImage {
      bIsDynamicallyLoaded=false
      drawAs=RoundedBox
      imageType=NoImage
      mirroring=NoMirror
      resourceName=None
      resourceObject=null
      tiling=NoTile
      imageSize=<18 18>
      margin=<ltrb 0 0 0 0>
      outlineSettings {
       bUseBrushTransparency=false
       roundingType=FixedRadius
       width=1
       color {colorUseRule=UseColor_ColorTable, specifiedColor=<rgba 0 0 0 0>}
       cornerRadii=<3 3 3 3>
      }
      tintColor {colorUseRule=UseColor_ColorTable, specifiedColor=<rgba 0 0 0 0>}
      uVRegion {bIsValid=false, max=<0 0>, min=<0 0>}
     }
     backgroundImage {
      bIsDynamicallyLoaded=false
      drawAs=RoundedBox
      imageType=NoImage
      mirroring=NoMirror
      resourceName=None
      resourceObject=null
      tiling=NoTile
      imageSize=<18 18>
      margin=<ltrb 0 0 0 0>
      outlineSettings {
       bUseBrushTransparency=false
       roundingType=FixedRadius
       width=1
       color {colorUseRule=UseColor_ColorTable, specifiedColor=<rgba 0 0 0 0>}
       cornerRadii=<3 3 3 3>
      }
      tintColor {colorUseRule=UseColor_ColorTable, specifiedColor=<rgba 0 0 0 0>}
      uVRegion {bIsValid=false, max=<0 0>, min=<0 0>}
     }
     backgroundPressedImage {
      bIsDynamicallyLoaded=false
      drawAs=RoundedBox
      imageType=NoImage
      mirroring=NoMirror
      resourceName=None
      resourceObject=null
      tiling=NoTile
      imageSize=<18 18>
      margin=<ltrb 0 0 0 0>
      outlineSettings {
       bUseBrushTransparency=false
       roundingType=FixedRadius
       width=1
       color {colorUseRule=UseColor_ColorTable, specifiedColor=<rgba 0 0 0 0>}
       cornerRadii=<3 3 3 3>
      }
      tintColor {colorUseRule=UseColor_ColorTable, specifiedColor=<rgba 0 0 0 0>}
      uVRegion {bIsValid=false, max=<0 0>, min=<0 0>}
     }
     borderBackgroundColor {colorUseRule=UseColor_Specified, specifiedColor=<rgba 1 1 1 1>}
     checkedForeground {colorUseRule=UseColor_ColorTable, specifiedColor=<rgba 0 0 0 0>}
     checkedHoveredForeground {colorUseRule=UseColor_ColorTable, specifiedColor=<rgba 0 0 0 0>}
     checkedHoveredImage {
      bIsDynamicallyLoaded=false
      drawAs=Image
      imageType=Vector
      mirroring=NoMirror
      resourceName=../../../Engine/Content/Slate/Starship/CoreWidgets/CheckBox/check.svg
      resourceObject=null
      tiling=NoTile
      imageSize=<16 16>
      margin=<ltrb 0 0 0 0>
      outlineSettings {
       bUseBrushTransparency=false
       roundingType=HalfHeightRadius
       width=0
       color {colorUseRule=UseColor_Specified, specifiedColor=<rgba 0 0 0 0>}
       cornerRadii=<0 0 0 0>
      }
      tintColor {colorUseRule=UseColor_Specified, specifiedColor=<rgba 0.0044 0.2384 1 1>}
      uVRegion {bIsValid=false, max=<0 0>, min=<0 0>}
     }
     checkedImage {
      bIsDynamicallyLoaded=false
      drawAs=Image
      imageType=Vector
      mirroring=NoMirror
      resourceName=../../../Engine/Content/Slate/Starship/CoreWidgets/CheckBox/check.svg
      resourceObject=null
      tiling=NoTile
      imageSize=<16 16>
      margin=<ltrb 0 0 0 0>
      outlineSettings {
       bUseBrushTransparency=false
       roundingType=HalfHeightRadius
       width=0
       color {colorUseRule=UseColor_Specified, specifiedColor=<rgba 0 0 0 0>}
       cornerRadii=<0 0 0 0>
      }
      tintColor {colorUseRule=UseColor_Specified, specifiedColor=<rgba 0 0.162 0.7454 1>}
      uVRegion {bIsValid=false, max=<0 0>, min=<0 0>}
     }
     checkedPressedForeground {colorUseRule=UseColor_ColorTable, specifiedColor=<rgba 0 0 0 0>}
     checkedPressedImage {
      bIsDynamicallyLoaded=false
      drawAs=Image
      imageType=Vector
      mirroring=NoMirror
      resourceName=../../../Engine/Content/Slate/Starship/CoreWidgets/CheckBox/check.svg
      resourceObject=null
      tiling=NoTile
      imageSize=<16 16>
      margin=<ltrb 0 0 0 0>
      outlineSettings {
       bUseBrushTransparency=false
       roundingType=HalfHeightRadius
       width=0
       color {colorUseRule=UseColor_Specified, specifiedColor=<rgba 0 0 0 0>}
       cornerRadii=<0 0 0 0>
      }
      tintColor {colorUseRule=UseColor_Specified, specifiedColor=<rgba 0 0.0802 0.3515 1>}
      uVRegion {bIsValid=false, max=<0 0>, min=<0 0>}
     }
     checkedSlateSound {resourceObject=null}
     foregroundColor {colorUseRule=UseColor_Specified, specifiedColor=<rgba 1 1 1 1>}
     hoveredForeground {colorUseRule=UseColor_ColorTable, specifiedColor=<rgba 0 0 0 0>}
     hoveredSlateSound {resourceObject=null}
     padding=<ltrb 2 0 0 0>
     pressedForeground {colorUseRule=UseColor_ColorTable, specifiedColor=<rgba 0 0 0 0>}
     uncheckedHoveredImage {
      bIsDynamicallyLoaded=false
      drawAs=NoDrawType
      imageType=NoImage
      mirroring=NoMirror
      resourceName=None
      resourceObject=null
      tiling=NoTile
      imageSize=<0 0>
      margin=<ltrb 0 0 0 0>
      outlineSettings {
       bUseBrushTransparency=false
       roundingType=HalfHeightRadius
       width=0
       color {colorUseRule=UseColor_Specified, specifiedColor=<rgba 0 0 0 0>}
       cornerRadii=<0 0 0 0>
      }
      tintColor {colorUseRule=UseColor_Specified, specifiedColor=<rgba 1 1 1 1>}
      uVRegion {bIsValid=false, max=<0 0>, min=<0 0>}
     }
     uncheckedImage {
      bIsDynamicallyLoaded=false
      drawAs=NoDrawType
      imageType=NoImage
      mirroring=NoMirror
      resourceName=None
      resourceObject=null
      tiling=NoTile
      imageSize=<0 0>
      margin=<ltrb 0 0 0 0>
      outlineSettings {
       bUseBrushTransparency=false
       roundingType=HalfHeightRadius
       width=0
       color {colorUseRule=UseColor_Specified, specifiedColor=<rgba 0 0 0 0>}
       cornerRadii=<0 0 0 0>
      }
      tintColor {colorUseRule=UseColor_Specified, specifiedColor=<rgba 1 1 1 1>}
      uVRegion {bIsValid=false, max=<0 0>, min=<0 0>}
     }
     uncheckedPressedImage {
      bIsDynamicallyLoaded=false
      drawAs=NoDrawType
      imageType=NoImage
      mirroring=NoMirror
      resourceName=None
      resourceObject=null
      tiling=NoTile
      imageSize=<0 0>
      margin=<ltrb 0 0 0 0>
      outlineSettings {
       bUseBrushTransparency=false
       roundingType=HalfHeightRadius
       width=0
       color {colorUseRule=UseColor_Specified, specifiedColor=<rgba 0 0 0 0>}
       cornerRadii=<0 0 0 0>
      }
      tintColor {colorUseRule=UseColor_Specified, specifiedColor=<rgba 1 1 1 1>}
      uVRegion {bIsValid=false, max=<0 0>, min=<0 0>}
     }
     uncheckedSlateSound {resourceObject=null}
     undeterminedForeground {colorUseRule=UseColor_ColorTable, specifiedColor=<rgba 0 0 0 0>}
     undeterminedHoveredImage {
      bIsDynamicallyLoaded=false
      drawAs=Image
      imageType=Vector
      mirroring=NoMirror
      resourceName=../../../Engine/Content/Slate/Starship/CoreWidgets/CheckBox/indeterminate.svg
      resourceObject=null
      tiling=NoTile
      imageSize=<16 16>
      margin=<ltrb 0 0 0 0>
      outlineSettings {
       bUseBrushTransparency=false
       roundingType=HalfHeightRadius
       width=0
       color {colorUseRule=UseColor_Specified, specifiedColor=<rgba 0 0 0 0>}
       cornerRadii=<0 0 0 0>
      }
      tintColor {colorUseRule=UseColor_Specified, specifiedColor=<rgba 0.0044 0.2384 1 1>}
      uVRegion {bIsValid=false, max=<0 0>, min=<0 0>}
     }
     undeterminedImage {
      bIsDynamicallyLoaded=false
      drawAs=Image
      imageType=Vector
      mirroring=NoMirror
      resourceName=../../../Engine/Content/Slate/Starship/CoreWidgets/CheckBox/indeterminate.svg
      resourceObject=null
      tiling=NoTile
      imageSize=<16 16>
      margin=<ltrb 0 0 0 0>
      outlineSettings {
       bUseBrushTransparency=false
       roundingType=HalfHeightRadius
       width=0
       color {colorUseRule=UseColor_Specified, specifiedColor=<rgba 0 0 0 0>}
       cornerRadii=<0 0 0 0>
      }
      tintColor {colorUseRule=UseColor_Specified, specifiedColor=<rgba 0 0.162 0.7454 1>}
      uVRegion {bIsValid=false, max=<0 0>, min=<0 0>}
     }
     undeterminedPressedImage {
      bIsDynamicallyLoaded=false
      drawAs=Image
      imageType=Vector
      mirroring=NoMirror
      resourceName=../../../Engine/Content/Slate/Starship/CoreWidgets/CheckBox/indeterminate.svg
      resourceObject=null
      tiling=NoTile
      imageSize=<16 16>
      margin=<ltrb 0 0 0 0>
      outlineSettings {
       bUseBrushTransparency=false
       roundingType=HalfHeightRadius
       width=0
       color {colorUseRule=UseColor_Specified, specifiedColor=<rgba 0 0 0 0>}
       cornerRadii=<0 0 0 0>
      }
      tintColor {colorUseRule=UseColor_Specified, specifiedColor=<rgba 0 0.0802 0.3515 1>}
      uVRegion {bIsValid=false, max=<0 0>, min=<0 0>}
     }
    }
   }
   slot {
    class=CanvasPanelSlot
    properties {layoutData {alignment=<0 0>, anchors {maximum=<0 0>, minimum=<0 0>}, offsets=<ltrb -0.1756 179.429 75 41.1261>}}
   }
  }
  Loop_Textbox_Text {
   widgetClassPath=&CommonTextBlock
   parent=Loop_Checkbox
   widget=Loop_Textbox_Text
   properties {
    style=&TextStyle-Base_C
    text=Loop
    font {
     bForceMonospaced=false
     bMaterialIsStencil=false
     fontMaterial=null
     fontObject=&NotoSans
     letterSpacing=0
     monospacedWidth=1
     size=24
     skewAmount=0
     typefaceFontName=NotoSans-Regular
     outlineSettings {
      bApplyOutlineToDropShadows=false
      bMiteredCorners=false
      bSeparateFillAlpha=false
      outlineMaterial=null
      outlineSize=0
      outlineColor=<rgba 0 0 0 1>
     }
    }
    renderTransform {angle=0, scale=<0.6 0.6>, shear=<0 0>, translation=<-11 0>}
    shadowOffset=<0 0>
   }
   slot {class=PanelSlot}
  }
  Distance_Text {
   widgetClassPath=&CommonTextBlock
   bIsVariable=true
   parent=CanvasPanel_213
   widget=Distance_Text
   properties {
    style=&TextStyle-Base_C
    text=0
    font {
     bForceMonospaced=false
     bMaterialIsStencil=false
     fontMaterial=null
     fontObject=&NotoSans
     letterSpacing=0
     monospacedWidth=1
     size=24
     skewAmount=0
     typefaceFontName=NotoSans-Regular
     outlineSettings {
      bApplyOutlineToDropShadows=false
      bMiteredCorners=false
      bSeparateFillAlpha=false
      outlineMaterial=null
      outlineSize=0
      outlineColor=<rgba 0 0 0 1>
     }
    }
    renderTransform {angle=0, scale=<0.5 0.5>, shear=<0 0>, translation=<0 0>}
    shadowOffset=<0 0>
   }
   slot {
    class=CanvasPanelSlot
    properties {layoutData {alignment=<0 0>, anchors {maximum=<0 0>, minimum=<0 0>}, offsets=<ltrb 112.1296 179.74 153 44>}}
   }
  }
  CommonTextBlock_10 {
   widgetClassPath=&CommonTextBlock
   parent=CanvasPanel_213
   widget=CommonTextBlock_10
   properties {
    style=&TextStyle-Base_C
    text="Distance: "
    font {
     bForceMonospaced=false
     bMaterialIsStencil=false
     fontMaterial=null
     fontObject=&NotoSans
     letterSpacing=0
     monospacedWidth=1
     size=24
     skewAmount=0
     typefaceFontName=NotoSans-Regular
     outlineSettings {
      bApplyOutlineToDropShadows=false
      bMiteredCorners=false
      bSeparateFillAlpha=false
      outlineMaterial=null
      outlineSize=0
      outlineColor=<rgba 0 0 0 1>
     }
    }
    renderTransform {angle=0, scale=<0.5 0.5>, shear=<0 0>, translation=<0 0>}
    shadowOffset=<0 0>
   }
   slot {
    class=CanvasPanelSlot
    properties {layoutData {alignment=<0 0>, anchors {maximum=<0 0>, minimum=<0 0>}, offsets=<ltrb 37.1085 178.9893 153 44>}}
   }
  }
  Audio_EUW_Meter {
   widgetClassPath=&/Audio_EUW_Meter
   bIsVariable=true
   parent=CanvasPanel_213
   widget=Audio_EUW_Meter
   slot {
    class=CanvasPanelSlot
    properties {layoutData {alignment=<0 0>, anchors {maximum=<0 0>, minimum=<0 0>}, offsets=<ltrb 372 216.4965 120 400>}}
   }
  }
  VerticalBox_36 {
   widgetClassPath=&VerticalBox
   parent=CanvasPanel_213
   widget=VerticalBox_36
   slot {class=CanvasPanelSlot, properties {layoutData {alignment=<0 0>, anchors {maximum=<1 0>, minimum=<0 0>}, offsets=<ltrb 0 0 0 135.7958>}}}
  }
  AnalysisSubmix_Property {
   widgetClassPath=&SinglePropertyView
   parent=VerticalBox_36
   widget=AnalysisSubmix_Property
   slot {class=VerticalBoxSlot, properties {size {sizeRule=Fill, value=1}}}
  }
  SoundToAudition_Property {
   widgetClassPath=&SinglePropertyView
   parent=VerticalBox_36
   widget=SoundToAudition_Property
   slot {class=VerticalBoxSlot, properties {size {sizeRule=Fill, value=1}}}
  }
  AuditionActor_Property {
   widgetClassPath=&SinglePropertyView
   parent=VerticalBox_36
   widget=AuditionActor_Property
   slot {class=VerticalBoxSlot, properties {size {sizeRule=Fill, value=1}}}
  }
  Audio_EUW_SpectrumAnalysis {
   widgetClassPath=&/Audio_EUW_SpectrumAnalysis
   bIsVariable=true
   parent=CanvasPanel_213
   widget=Audio_EUW_SpectrumAnalysis
   slot {
    class=CanvasPanelSlot
    properties {layoutData {alignment=<0 0>, anchors {maximum=<0 0>, minimum=<0 0>}, offsets=<ltrb -24 208 316.0632 209.9466>}}
   }
  }
  Audio_EUW_Multichannel_Meter {
   widgetClassPath=&/Audio_EUW_Multichannel_Meter
   bIsVariable=true
   parent=CanvasPanel_213
   widget=Audio_EUW_Multichannel_Meter
   slot {
    class=CanvasPanelSlot
    properties {layoutData {alignment=<0 0>, anchors {maximum=<0 0>, minimum=<0 0>}, offsets=<ltrb -0.6433 429.3307 356.6903 209.5128>}}
   }
  }
  CanvasPanel_266 {
   widgetClassPath=&CanvasPanel
   parent=CanvasPanel_213
   widget=CanvasPanel_266
   slot {
    class=CanvasPanelSlot
    properties {layoutData {alignment=<0 0>, anchors {maximum=<0 0>, minimum=<0 0>}, offsets=<ltrb 520.1156 270 132.6326 399.8699>}}
   }
  }
  Audio_EUW_Meter_0 {
   widgetClassPath=&Audio_EUW_Meter
   bIsVariable=true
   parent=CanvasPanel_266
   widget=Audio_EUW_Meter_0
   slot {
    class=CanvasPanelSlot
    properties {layoutData {alignment=<0 0>, anchors {maximum=<0 0>, minimum=<0 0>}, offsets=<ltrb -12 -53.5015 120 400>}}
   }
  }
  CommonTextBlock_182 {
   widgetClassPath=&CommonTextBlock
   parent=CanvasPanel_266
   widget=CommonTextBlock_182
   properties {
    style=&TextStyle-Base_C
    text=Master Submix
    font {
     bForceMonospaced=false
     bMaterialIsStencil=false
     fontMaterial=null
     fontObject=&NotoSans
     letterSpacing=0
     monospacedWidth=1
     size=24
     skewAmount=0
     typefaceFontName=NotoSans-Regular
     outlineSettings {
      bApplyOutlineToDropShadows=false
      bMiteredCorners=false
      bSeparateFillAlpha=false
      outlineMaterial=null
      outlineSize=0
      outlineColor=<rgba 0 0 0 1>
     }
    }
    renderTransform {angle=0, scale=<0.5 0.5>, shear=<0 0>, translation=<0 0>}
    shadowOffset=<0 0>
   }
   slot {
    class=CanvasPanelSlot
    properties {bAutoSize=true, layoutData {alignment=<0 0>, anchors {maximum=<0 0>, minimum=<0 0>}, offsets=<ltrb -66.257 -82.9622 184.0841 30>}}
   }
  }
  CommonTextBlock_1 {
   widgetClassPath=&CommonTextBlock
   bIsVariable=true
   parent=CanvasPanel_266
   widget=CommonTextBlock_1
   properties {
    style=&TextStyle-Base_C
    text=Submix To Analyze
    wrappingPolicy=AllowPerCharacterWrapping
    font {
     bForceMonospaced=false
     bMaterialIsStencil=false
     fontMaterial=null
     fontObject=&NotoSans
     letterSpacing=0
     monospacedWidth=1
     size=24
     skewAmount=0
     typefaceFontName=NotoSans-Regular
     outlineSettings {
      bApplyOutlineToDropShadows=false
      bMiteredCorners=false
      bSeparateFillAlpha=false
      outlineMaterial=null
      outlineSize=0
      outlineColor=<rgba 0 0 0 1>
     }
    }
    renderTransform {angle=0, scale=<0.4 0.4>, shear=<0 0>, translation=<0 0>}
    shadowOffset=<0 0>
   }
   slot {
    class=CanvasPanelSlot
    properties {layoutData {alignment=<0 0>, anchors {maximum=<1 1>, minimum=<0 1>}, offsets=<ltrb -213.0778 -482.4426 118.9837 43.2132>}}
   }
  }
 ]
}
