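# Editor utility Blueprint that batch-configures selected SoundWave assets as Emote or MusicPack cosmetic audio.
# InitCosmeticSoundWave loops over the selection, casts each to SoundWave, stores it in CurrentSoundWave and switches on the chosen type to call its init event.
# Both events currently apply identical settings through SetEditorProperty by name: compression, streaming, play-when-silent virtualization and baked FFT analysis.
# A SetProperty wrapper returning success exists but is not called here.
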EditorUtilityBlueprint /Game/Sounds/Fort_Audio_Blueprints/Utilities/AssetActions/SoundWaveAssetActions
graphs [
 EventGraph {nodes [
   Init Emote SoundWave() {is=title K2Node_CustomEvent}
   > SetEditorProperty(
    Object=Get CurrentSoundWave() -> (CurrentSoundWave:Sound Wave Object Reference)
    PropertyName=CompressionQuality
    PropertyValue:int=MakeLiteralInt!(Value:int=60) -> (ReturnValue:int=0)
   )
   > SetEditorProperty(Object=Get CurrentSoundWave(), PropertyName=FFTAnalysisReleaseTime, PropertyValue=MakeLiteralInt(Value=100) -> (ReturnValue=0))
   > SetEditorProperty(Object=Get CurrentSoundWave(), PropertyName=bStreaming, PropertyValue:bool=MakeLiteralBool(Value=true))
   > SetEditorProperty(Object=Get CurrentSoundWave(), PropertyName=bEnableBakedFFTAnalysis, PropertyValue=MakeLiteralBool(Value=true))
   > SetEditorProperty(
    Object=Get CurrentSoundWave()
    PropertyName=VirtualizationMode
    PropertyValue:EVirtualizationMode Enum=Literal enum EVirtualizationMode!(Enum:EVirtualizationMode Enum=PlayWhenSilent) -> (ReturnValue:EVirtualizationMode Enum) {is=title K2Node_EnumLiteral}
   )
   > SetEditorProperty(
    Object=Get CurrentSoundWave()
    PropertyName=FrequenciesToAnalyze
    PropertyValue:"Array of Float (double-precision)s"=Make Array!("[0]":double=100, "[1]":double=400, "[2]":double=800, "[3]":double=1200, "[4]":double=2000) -> (Array:"Array of Float (double-precision)s")
   )
   Init MusicPack SoundWave() {is=title K2Node_CustomEvent}
   > SetEditorProperty(Object=Get CurrentSoundWave(), PropertyName=CompressionQuality, PropertyValue:int=MakeLiteralInt(Value=60) -> (ReturnValue=0))
   > SetEditorProperty(Object=Get CurrentSoundWave(), PropertyName=FFTAnalysisReleaseTime, PropertyValue=MakeLiteralInt(Value=100) -> (ReturnValue=0))
   > SetEditorProperty(Object=Get CurrentSoundWave(), PropertyName=bStreaming, PropertyValue:bool=MakeLiteralBool(Value=true))
   > SetEditorProperty(Object=Get CurrentSoundWave(), PropertyName=bEnableBakedFFTAnalysis, PropertyValue=MakeLiteralBool(Value=true))
   > SetEditorProperty(
    Object=Get CurrentSoundWave()
    PropertyName=VirtualizationMode
    PropertyValue:EVirtualizationMode Enum=Literal enum EVirtualizationMode(Enum=PlayWhenSilent)
   )
   > SetEditorProperty(
    Object=Get CurrentSoundWave()
    PropertyName=FrequenciesToAnalyze
    PropertyValue:"Array of Float (double-precision)s"=Make Array("[0]"=100, "[1]"=400, "[2]"=800, "[3]"=1200, "[4]"=2000)
   )
  ]}
 InitCosmeticSoundWave {nodes [
   n1 = InitCosmeticSoundWave!() -> (then, Type:SoundWaveInitTypes Enum=NewEnumerator0) {is=title K2Node_FunctionEntry}
   > n2 = GetSelectedAssets()
   > n3 = For Each Loop(Array:Array of Object References=@n2) -> (Array Element:Object Reference)
   > n4 = Cast To SoundWave!(execute, Object=@n3) -> (then, CastFailed, AsSound Wave:Sound Wave Object Reference) {is=title K2Node_DynamicCast}
   > Set CurrentSoundWave(CurrentSoundWave:Sound Wave Object Reference=@n4) -> (Output_Get:Sound Wave Object Reference)
   > n5 = Switch on SoundWaveInitTypes!(execute, Selection:SoundWaveInitTypes Enum=@n1) -> (NewEnumerator0:Exec, NewEnumerator1:Exec) {is=title K2Node_SwitchEnum}
   > Init Emote SoundWave() {is=title K2Node_CallFunction}
   Init MusicPack SoundWave(execute=@n5.NewEnumerator1) {is=title K2Node_CallFunction}
  ]}
 SetProperty {nodes [
   n1 = Inputs!() -> (execute, PropertyName:Name, PropertyValue:Wildcard)
   > n2 = SetEditorProperty(Object=Get CurrentSoundWave(), PropertyName=@n1, PropertyValue:Wildcard=@n1.PropertyValue)
   > Outputs!(then:Exec, Success?:bool=@n2)
  ]}
]
variables [CurrentSoundWave {type=SoundWave, default=null}]
