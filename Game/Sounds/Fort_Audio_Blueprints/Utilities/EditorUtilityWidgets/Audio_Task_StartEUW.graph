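# Audio_Task_StartEUW is an editor utility blueprint that opens the Audio_EUW_Parent editor utility widget as a tab.
# On BeginExecution it gets the EditorUtilitySubsystem and calls SpawnAndRegisterTab with that widget blueprint.
# It then calls FinishExecutingTask immediately after spawning the tab.
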EditorUtilityBlueprint /Game/Sounds/Fort_Audio_Blueprints/Utilities/EditorUtilityWidgets/Audio_Task_StartEUW
graphs [EventGraph {nodes [
   Event BeginExecution()
   > SpawnAndRegisterTab(self=Get EditorUtilitySubsystem(), InBlueprint=&/Audio_EUW_Parent)
   > FinishExecutingTask()
  ]}]
variables []
