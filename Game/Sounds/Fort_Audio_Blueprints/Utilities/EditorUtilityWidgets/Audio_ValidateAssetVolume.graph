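# Editor utility widget with a Validate button that audits sound assets for near-zero volume.
# On click, it recursively lists assets under the current Content Browser path and loads each one.
# It reads Volume from SoundWaves or VolumeMultiplier from SoundCues by reflected property name.
# Values at or below a ZeroVolume threshold, or equal to zero, print the asset path to the log only.
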EditorUtilityWidgetBlueprint /Game/Sounds/Fort_Audio_Blueprints/Utilities/EditorUtilityWidgets/Audio_ValidateAssetVolume
graphs [EventGraph {nodes [
   "On Clicked (EditorUtilityButton_26)"() {is=title K2Node_ComponentBoundEvent}
   n1 = GetCurrentContentBrowserPath!() -> (OutPath:String, ReturnValue:bool)
   > Branch(Condition=@n1.ReturnValue)
   > n2 = ListAssets!(
    execute
    self:Editor Asset Subsystem Object Reference=Get EditorAssetSubsystem() -> (ReturnValue:Editor Asset Subsystem Object Reference) {is=title K2Node_GetEditorSubsystem}
    DirectoryPath:String=@n1
    bRecursive:bool=true
    bIncludeFolder:bool=true
   ) -> (then, ReturnValue:Array of Strings)
   > n3 = For Each Loop(Array=@n2)
   > n4 = ConvertFilePathToAssetPath!(execute, InPath:String=@n3) -> (then, OutPath:String, ReturnValue:bool)
   > Branch(Condition=@n4.ReturnValue)
   > n5 = LoadAsset(AssetPath=@n4)
   > n6 = Cast To SoundWave!(execute, Object=@n5) -> (then, CastFailed, AsSound Wave:Sound Wave Object Reference) {is=title K2Node_DynamicCast}
   > n7 = GetEditorProperty!(execute:Exec, Object:Object Reference=@n6, PropertyName:Name=Volume) -> (then:Exec, PropertyValue:double, ReturnValue:bool)
   > Set ThisVolume(ThisVolume:double=@n7) -> (Output_Get:double)
   n8 = Get ThisVolume() -> (ThisVolume:double)
   n9 = Get ZeroVolume() -> (ZeroVolume:double)
   > Branch(Condition=OR Boolean(A="float <= float"(A=@n8, B=@n9) -> (ReturnValue=~), B="Nearly Equal (Float)"(A=@n8, B=0, ErrorTolerance=0)))
   > PrintText(
    WorldContextObject=Self-Reference()
    InText=Format Text!(Format="LOCTEXT(\"Asset has volume ~ Zero! {path}\")", path:String=@n3) -> (Result)
    bPrintToScreen=false
   )
   n10 = Cast To SoundCue!(execute=@n6.CastFailed, Object=@n5) -> (then, CastFailed, AsSound Cue:Sound Cue Object Reference) {is=title K2Node_DynamicCast}
   > n11 = GetEditorProperty(Object=@n10, PropertyName=VolumeMultiplier)
   > Set ThisVolume(ThisVolume=@n11)
   n12 = Get ThisVolume()
   n13 = Get ZeroVolume()
   > Branch(Condition=OR Boolean(A="float <= float"(A=@n12, B=@n13) -> (ReturnValue=~), B="Nearly Equal (Float)"(A=@n12, B=0, ErrorTolerance=0)))
   > PrintText(
    WorldContextObject=Self-Reference()
    InText=Format Text(Format="LOCTEXT(\"Asset has volume ~ Zero! {path}\")", path=@n3)
    bPrintToScreen=false
   )
  ]}]
variables [
 ThisVolume {type=number}
 ZeroVolume {type=number}
]
widgets {
 info {rootWidgetClass=&EditorUtilityButton, widgetCount=2}
 widgets [
  EditorUtilityButton_26 {widgetClassPath=&EditorUtilityButton, widget=EditorUtilityButton_26, slot=~}
  Validate {widgetClassPath=&TextBlock, parent=EditorUtilityButton_26, widget=Validate, properties {text=Validate}, slot {class=ButtonSlot}}
 ]
}
